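# Editor utility list-row widget for the Level Sequence Browser, showing a sequence's name and source control status.
# On item set, it shows the asset name, optionally prefix-stripped per the batch updater's save game, then async loads the sequence and writes its path back.
# After a short retriggerable delay, a silent async file-state query drives icon, tint, text color and tooltip.
# Selection highlights the row; right-clicks forward to the browser.
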WidgetBlueprint /AnimInEdTools/LevelSequenceBrowser/LevelSequenceBrowser_ListItem
graphs [
 Async P4 Update {nodes [
   n1 = Async P4 Update!() -> (then, FileStateOut:Source Control State Structure) {is=title K2Node_FunctionEntry}
   > n2 = Sequence!(execute) -> (then_0, then_1, then_2:Exec, then_3:Exec, then_4:Exec)
   n3 = Break Source Control State!(SourceControlState:"Source Control State Structure (by ref)"=@n1) -> (
    Filename:String
    bIsValid:bool=~
    bIsUnknown:bool=~
    bCanCheckIn:bool=~
    bCanCheckOut:bool=~
    bIsCheckedOut:bool=~
    bIsCurrent:bool=~
    bIsSourceControlled:bool=~
    bIsAdded:bool=~
    bIsDeleted:bool=~
    bIsIgnored:bool=~
    bCanEdit:bool=~
    bCanDelete:bool=~
    bIsModified:bool=~
    bCanAdd:bool=~
    bIsConflicted:bool=~
    bCanRevert:bool=~
    bIsCheckedOutOther:bool=~
    CheckedOutOther:String
    bIsCheckedOutInOtherBranch:bool=~
    bIsModifiedInOtherBranch:bool=~
    PreviousUser:String
   ) {is=title K2Node_BreakStruct}
   > n4 = Branch(Condition=@n3.bIsCurrent)
   Set Style!(
    execute=@n4.else
    self
    Extra Text Color:Linear Color Structure=(R=0.7135,G=0.646,B=0,A=1)
    Extra Text Typeface:Name=Bold
    Icon:Object Reference=/AnimInEdTools/AnimationBrowser/Icons/Icon_LevelSequence_alert
    Extra Text:String
    Main Text Color:Linear Color Structure=(R=0.5255,G=0.5255,B=0.5255,A=1)
    Icon Tint:Slate Color Structure=Make Slate Color(SpecifiedColor=(R=0.7137,G=0.6471,B=0,A=1))
    Tool Tip Text:String=There is a newer version of this asset checked in
   ) -> (then) {is=title K2Node_CallFunction}
   Branch(execute=@n2.then_1, Condition=@n3.bIsAdded)
   > Set Style(
    Extra Text Color=(R=0,G=0.338,B=0.5255,A=1)
    Extra Text Typeface=Bold
    Icon=/AnimInEdTools/AnimationBrowser/Icons/Icon_LevelSequence_add
    Main Text Color=(R=0.5255,G=0.5255,B=0.5255,A=1)
    Icon Tint=Make Slate Color(SpecifiedColor=(R=0.0314,G=0.5255,B=0.0627,A=1))
    Tool Tip Text=The file is marked for add by you
   )
   n5 = Branch(execute=@n2.then_2, Condition=@n3.bIsSourceControlled)
   Set Style(
    execute=@n5.else
    Extra Text Color=(R=0.1714,G=0.1714,B=0.1714,A=1)
    Extra Text Typeface=Regular
    Icon=/AnimInEdTools/AnimationBrowser/Icons/Icon_LevelSequence_not_in_source
    Main Text Color=(R=0.1714,G=0.1714,B=0.1714,A=1)
    Icon Tint=Make Slate Color(SpecifiedColor=(R=0.7137,G=0.6471,B=0,A=1))
    Tool Tip Text=Local File
   )
   Branch(execute=@n2.then_3, Condition=@n3.bIsCheckedOut)
   > Set Style(
    Extra Text Color=(R=0.0319,G=0.5271,B=0.063,A=1)
    Extra Text Typeface=Bold
    Icon=/AnimInEdTools/AnimationBrowser/Icons/Icon_AnimationBrowser_Red_Checkmark
    Main Text Color=(R=0.5255,G=0.5255,B=0.5255,A=1)
    Icon Tint=Make Slate Color(SpecifiedColor=(R=0.0314,G=0.5255,B=0.0627,A=1))
    Tool Tip Text=The file is checked out
   )
   Branch(execute=@n2.then_4, Condition=@n3.bIsCheckedOutOther)
   n6 = Split(SourceString=@n3.CheckedOutOther, InStr=" @", SearchDir=FromStart)
   > Set Style(
    Extra Text Color=(R=0.5271,G=0,B=0,A=1)
    Extra Text Typeface=Bold
    Icon=&Icon_AnimationBrowser_Red_Checkmark
    Main Text Color=(R=0.5255,G=0.5255,B=0.5255,A=1)
    Icon Tint=Make Slate Color(SpecifiedColor=(R=0.5255,G=0,B=0,A=1))
    Tool Tip Text=Append(A="The file is currently checked out by: ", B=@n6)
   )
   Append!(A="(Checked out by: ", B=@n6, C:String=")") -> (ReturnValue)
  ]}
 Display_Selected {nodes [
   n1 = Display_Selected!() -> (then, Index:bool=~) {is=title K2Node_FunctionEntry}
   > SetBrushColor(
    self=Get Border() -> (Border:Border Object Reference)
    InBrushColor=SelectColor(A=(R=0,G=0.1608,B=0.7451,A=1), B=(R=0,G=0,B=0,A=0), bPickA=@n1)
   )
   > SetColorAndOpacity(
    self=Get TextBlock() -> (TextBlock:Text Object Reference)
    InColorAndOpacity=Select(
     Option 0:Slate Color Structure=Get Color And Opacity() -> (Color And Opacity:Slate Color Structure)
     Option 1:Slate Color Structure=Make Slate Color(SpecifiedColor=(R=0.0103,G=0.0103,B=0.0103,A=1))
     Index=@n1
    ) -> (ReturnValue:Slate Color Structure)
   )
  ]}
 Edit Text {nodes [Edit Text!() -> (then, Edit:bool=~) {is=title K2Node_FunctionEntry}]}
 EventGraph {nodes [
   Event Construct()
   > Set Color And Opacity(Color And Opacity:Slate Color Structure=Get ColorAndOpacity!(self:Text Object Reference=Get TextBlock()) -> (ColorAndOpacity:Slate Color Structure)) -> (Output_Get:Slate Color Structure)
   n1 = Event On Item Selection Changed!() -> (OutputDelegate, then, bIsSelected:bool) {is=title K2Node_Event}
   > Display_Selected!(execute, self, Index:bool=@n1.bIsSelected) -> (then) {is=title K2Node_CallFunction}
   n2 = Event OnListItemObjectSet()
   > n3 = Sequence!(execute) -> (then_0, then_1, then_2)
   > Display_Selected(Index=false)
   n4 = Cast To Object_LevelSequenceBrowser!(execute=@n3.then_1, Object=@n2.ListItemObject) -> (then, CastFailed, AsObject Level Sequence Browser:Object Level Sequence Browser Object Reference) {is=title K2Node_DynamicCast}
   > n5 = Set ItemData(ItemData:Object Level Sequence Browser Object Reference=@n4) -> (Output_Get:Object Level Sequence Browser Object Reference)
   n6 = Get Data!(self:Object Level Sequence Browser Object Reference=@n5) -> (Data:Struct Level Sequence Browser Item Structure)
   n7 = GetBaseFilename(
    InPath=To Soft Object Path!(SoftObjectReference:Soft Object Reference=Break Struct Level Sequence Browser Item!(Struct_LevelSequenceBrowser_Item:"Struct Level Sequence Browser Item Structure (by ref)"=@n6) -> (LS_58_#1:Soft Object Reference) {is=title K2Node_BreakStruct}) -> (ReturnValue_PathString:String)
    bRemovePath=true
   )
   n8 = Split(SourceString=@n7, InStr=_, SearchDir=FromStart)
   n9 = Get Filter Match Asset Name!(self:SG Level Sequence Browser Object Reference=Get Save Game!(self:EUW Level Sequence Batch Update Object Reference=Get Level Batch Updater!() -> (LS Batch:EUW Level Sequence Batch Update Object Reference) {is=title K2Node_CallFunction}) -> (Save Game:SG Level Sequence Browser Object Reference)) -> (Filter Match Asset Name:bool)
   > "SetText (Text)"(self=Get TextBlock(), InText="To Text (String)"(InString=SelectString(A=@n8.RightS, B=@n7, bPickA=@n9)))
   n10 = To Soft Object Path(SoftObjectReference=Get!(self:Object Level Sequence Browser Object Reference=Get ItemData() -> (ItemData:Object Level Sequence Browser Object Reference)) -> (Data_LS_58_#1:Soft Object Reference))
   n11 = Async Load Asset(execute=@n3.then_2, Asset=To Soft Object Reference(SoftObjectPath=MakeSoftClassPath(PathString=@n10)))
   n12 = Cast To LevelSequence!(execute=@n11.Completed, Object=@n11) -> (then, CastFailed, AsLevel Sequence:Level Sequence Object Reference) {is=title K2Node_DynamicCast}
   > Is Valid(InputObject=@n12)
   > Set!(
    execute
    Data_LS_58_#1:Soft Object Reference=To Soft Object Reference!(SoftObjectPath_PathString:String=Get Object Path String(Object=@n12)) -> (ReturnValue)
    self:Object Level Sequence Browser Object Reference=Get ItemData()
   ) -> (then, Output_Get:Struct Level Sequence Browser Item Structure)
   > SetBrush(
    self=Get Icon() -> (Icon:Image Object Reference)
    InBrush=Make Slate Brush!(
     TintColor:Slate Color Structure=(SpecifiedColor=(R=1,G=1,B=1,A=1),ColorUseRule=UseColor_Specified)
     DrawAs:ESlateBrushDrawType Enum=Image
     Tiling:ESlateBrushTileType Enum=NoTile
     Mirroring:ESlateBrushMirrorType Enum=NoMirror
     ImageSize:Vector 2D Structure=(X=14,Y=14)
     Margin:Margin Structure=(Left=0,Top=0,Right=0,Bottom=0)
     ResourceObject:Object Reference=/AnimInEdTools/AnimationBrowser/Icons/Icon_LevelSequence
     OutlineSettings:Slate Brush Outline Settings Structure=(CornerRadii=(X=0,Y=0,Z=0,W=0),Color=(SpecifiedColor=(R=0,G=0,B=0,A=0),ColorUseRule=UseColor_Specified),Width=0,RoundingType=HalfHeightRadius,bUseBrushTransparency=False)
    ) -> (SlateBrush:Slate Brush Structure) {is=title K2Node_MakeStruct}
   )
   > RetriggerableDelay!(execute, WorldContextObject:Object Reference, Duration=0.2) -> (then)
   > AsyncQueryFileState!(
    execute
    FileStateCallback:Delegate=Create Event()
    InFile:String=GetSystemPath!(Object:Object Reference=@n12) -> (ReturnValue:String)
    bSilent:bool=true
   ) -> (then)
  ]}
 Get Level Batch Updater {nodes [
   Get Level Batch Updater() {is=title K2Node_FunctionEntry}
   > n1 = FindParentWidgetOfType!(
    execute
    StartingWidget:Widget Object Reference=GetOwningListView!(UserListEntry:User List Entry Interface) -> (ReturnValue:List View Base Object Reference)
    Type:Widget Class Reference=/AnimInEdTools/LevelSequenceBatchUpdate/EUW_LevelSequenceBatchUpdate.EUW_LevelSequenceBatchUpdate_C
   ) -> (then, ReturnValue:EUW Level Sequence Batch Update Object Reference)
   > Return Node!(execute, LS Batch:EUW Level Sequence Batch Update Object Reference=@n1)
  ]}
 OnMouseButtonDown {nodes [
   n1 = OnMouseButtonDown!() -> (then, MyGeometry:Geometry Structure, MouseEvent:"Pointer Event Structure (by ref)") {is=title K2Node_FunctionEntry}
   > n2 = Branch(Condition=Is Mouse Button Down!(Input:"Pointer Event Structure (by ref)"=@n1.MouseEvent, MouseButton:Key Structure=RightMouseButton) -> (ReturnValue:bool))
   > n3 = FindParentWidgetOfType(StartingWidget=Self-Reference(), Type=&/EUW_LevelSequenceBrowser.EUW_LevelSequenceBrowser_C) -> (ReturnValue:EUW Level Sequence Browser Object Reference)
   > "On Item Right Clicked (List View)"!(
    execute
    self:EUW Level Sequence Browser Object Reference=@n3
    Item:Object Level Sequence Browser Object Reference=Get ItemData()
    Mouse Event:Pointer Event Structure=Get MouseEvent() -> (MouseEvent:"Pointer Event Structure (by ref)")
   ) -> (then)
   > Return Node!(execute, ReturnValue:Event Reply Structure=Handled())
   Return Node(execute=@n2.else, ReturnValue=Unhandled())
  ]}
 Set Style {nodes [
   n1 = Set Style!() -> (
    then
    Extra Text Color:Linear Color Structure
    Extra Text Typeface:Name=Regular
    Icon:Object Reference
    Extra Text:String
    Main Text Color:Linear Color Structure=(R=0.5255,G=0.5255,B=0.5255,A=1)
    Icon Tint:Slate Color Structure
    Tool Tip Text:String
   ) {is=title K2Node_FunctionEntry}
   > n2 = Sequence!(execute) -> (then_0, then_1, then_2, then_3:Exec, then_4:Exec)
   SetColorAndOpacity(
    execute=@n2.then_1
    self=Get TextBlock()
    InColorAndOpacity=Make Slate Color(SpecifiedColor=Get Main Text Color() -> (Main Text Color:Linear Color Structure))
   )
   SetBrush(
    execute=@n2.then_2
    self=Get Icon()
    InBrush=Make Slate Brush(
     TintColor=@n1.Icon Tint
     DrawAs=Image
     Tiling=NoTile
     Mirroring=NoMirror
     ImageSize=(X=14,Y=14)
     Margin=(Left=0,Top=0,Right=0,Bottom=0)
     ResourceObject=@n1.Icon
     OutlineSettings=(CornerRadii=(X=0,Y=0,Z=0,W=0),Color=(SpecifiedColor=(R=0,G=0,B=0,A=0),ColorUseRule=UseColor_Specified),Width=0,RoundingType=HalfHeightRadius,bUseBrushTransparency=False)
    )
   )
   Set ToolTipText!(
    execute=@n2.then_4
    ToolTipText:Text="To Text (String)"(InString=Get Tool Tip Text() -> (Tool Tip Text:String))
    self:Widget Object Reference=Get Icon()
   ) -> (then, Output_Get:Text)
  ]}
]
variables [
 Color And Opacity {type=SlateColor, default {colorUseRule=UseColor_Specified, specifiedColor=<rgba 1 1 1 1>}}
 ItemData {type=Object_LevelSequenceBrowser_C, default=null}
]
widgets {
 info {inheritedWidgetCount=0, parentClass=&EditorUtilityWidget, rootWidgetClass=&Border, widgetCount=7}
 widgets [
  Border {widgetClassPath=&Border, bIsVariable=true, widget=Border, properties {brushColor=<rgba 0 0 0 1>, padding=<ltrb 0 2 4 2>}}
  HorizontalBox_1 {
   widgetClassPath=&HorizontalBox
   parent=Border
   widget=HorizontalBox_1
   slot {class=BorderSlot, properties {padding=<ltrb 0 2 4 2>}}
  }
  Icon {
   widgetClassPath=&Image
   bIsVariable=true
   parent=HorizontalBox_1
   widget=Icon
   properties {brush {
     bIsDynamicallyLoaded=false
     drawAs=Image
     imageType=FullColor
     mirroring=NoMirror
     resourceName=None
     resourceObject=&Icon_LevelSequence
     tiling=NoTile
     imageSize=<14 14>
     margin=<ltrb 0 0 0 0>
     outlineSettings {
      bUseBrushTransparency=false
      roundingType=HalfHeightRadius
      width=0
      color {colorUseRule=UseColor_Specified, specifiedColor=<rgba 0 0 0 0>}
      cornerRadii=<0 0 0 0>
     }
     tintColor {colorUseRule=UseColor_Specified, specifiedColor=<rgba 0.0048 0.0048 0.0048 1>}
     uVRegion {bIsValid=false, max=<0 0>, min=<0 0>}
    }}
   slot {class=HorizontalBoxSlot, properties {horizontalAlignment=HAlign_Center, verticalAlignment=VAlign_Center}}
  }
  Spacer_126 {
   widgetClassPath=&Spacer
   parent=HorizontalBox_1
   widget=Spacer_126
   slot {class=HorizontalBoxSlot, properties {padding=<ltrb 5 0 0 0>}}
  }
  SizeBox_1 {
   widgetClassPath=&SizeBox
   parent=HorizontalBox_1
   widget=SizeBox_1
   properties {bOverride_MinDesiredWidth=true, minDesiredWidth=200}
   slot {class=HorizontalBoxSlot, properties {padding=<ltrb 3 0 0 0>, size {sizeRule=Fill, value=1}}}
  }
  HorizontalBox_0 {
   widgetClassPath=&HorizontalBox
   parent=SizeBox_1
   widget=HorizontalBox_0
   slot {class=SizeBoxSlot, properties {verticalAlignment=VAlign_Center}}
  }
  TextBlock {
   widgetClassPath=&TextBlock
   bIsVariable=true
   parent=HorizontalBox_0
   widget=TextBlock
   properties {
    text=LevelSequence
    colorAndOpacity {colorUseRule=UseColor_Specified, specifiedColor=<rgba 0.5255 0.5255 0.5255 1>}
    font {
     bForceMonospaced=false
     bMaterialIsStencil=false
     fontMaterial=null
     fontObject=&Roboto
     letterSpacing=0
     monospacedWidth=1
     size=10
     skewAmount=0
     typefaceFontName=Regular
     outlineSettings {
      bApplyOutlineToDropShadows=false
      bMiteredCorners=false
      bSeparateFillAlpha=false
      outlineMaterial=null
      outlineSize=0
      outlineColor=<rgba 0 0 0 1>
     }
    }
   }
   slot {class=HorizontalBoxSlot}
  }
 ]
}
