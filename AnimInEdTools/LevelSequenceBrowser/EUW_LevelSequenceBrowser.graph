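# Editor utility widget that browses level sequences for animatable creatures, filtered by asset, variant, subfolder and search text.
# Combo boxes fill from the selected project's creature list, with naming conventions resolving animation folders and Python nodes sorting options.
# Project and filter choices persist in a SaveGame, and setup waits for the Asset Registry.
# Double-click opens a sequence; right-click offers browsing to it or its linked AnimSequence via binding GUIDs.
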EditorUtilityWidgetBlueprint /AnimInEdTools/LevelSequenceBrowser/EUW_LevelSequenceBrowser
graphs [
 Animatable Creature Get Variations {nodes [
   n1 = Animatable Creature Get Variations!() -> (then, AnimatableCreatures:Array of Animatable Creature Object References, Asset Name:String, Add All Option:bool=true) {is=title K2Node_FunctionEntry}
   > Clear(TargetArray=Get Local Creatures() -> (Local Creatures:Array of Strings))
   > n2 = For Each Loop(Array:Array of Animatable Creature Object References=@n1) -> (Array Element:Animatable Creature Object Reference)
   n3 = Execute Python Script!(
    execute=@n2.Completed
    PythonScript="sorted_creatures = sorted(creatures)\r\n\r\nif 'Shared' in sorted_creatures:\r\n    sorted_creatures.insert(0, sorted_creatures.pop(sorted_creatures.index('Shared')))\r\n\r\nif len(creatures) > 1 and add_all:\r\n    sorted_creatures.insert(0, 'All')"
    creatures:Array of Strings=Get Local Creatures()
    add_all:bool=Get Add All Option() -> (Add All Option:bool)
   ) -> (then, ReturnValue, sorted_creatures:Array of Strings) {inputs [creatures,add_all], outputs [sorted_creatures]}
   > Return Node!(execute, Variations:Array of Strings=@n3.sorted_creatures)
   Branch(
    execute=@n2.LoopBody
    Condition="Equal Exactly (String)"(A=Get Asset!(self:Animatable Creature Object Reference=@n2) -> (Asset:String), B=Get Asset Name() -> (Asset Name:String))
   )
   > Add(TargetArray=Get Local Creatures(), NewItem=Get Variant!(self:Animatable Creature Object Reference=@n2) -> (Variant:String)) {is=title K2Node_CallArrayFunction}
  ]}
 AnimatableCreature Find {nodes [
   n1 = AnimatableCreature Find!() -> (then, AnimatableCreatures:Array of Animatable Creature Object References, Asset:String, Variant:String) {is=title K2Node_FunctionEntry}
   > n2 = Branch(Condition="Equal Exactly (String)"(A=@n1.Variant, B=All))
   > Set Best Fit(Best Fit:Animatable Creature Object Reference) -> (Output_Get:Animatable Creature Object Reference)
   > n3 = For Each Loop with Break!(Exec:Exec, Array:Array of Animatable Creature Object References=@n1, Break:Exec) -> (LoopBody:Exec, Array Element:Animatable Creature Object Reference, Array Index:int, Completed:Exec)
   Return Node!(
    execute=@n3.Completed
    AnimatableCreature:Animatable Creature Object Reference=Get Best Fit() -> (Best Fit:Animatable Creature Object Reference)
    Exact Match:bool=false
   )
   > Branch(Condition="Equal Exactly (String)"(A=Get Asset(self=@n3), B=Get Asset!() -> (Asset)))
   n4 = Get Variant(self=@n3)
   > n5 = Branch(Condition=OR Boolean(A="Equal, Case Insensitive (String)"(A=@n4, B=Shared), B="Equal, Case Insensitive (String)"(A=@n4, B=Core)))
   > Return Node(AnimatableCreature=@n3, Exact Match=false)
   Set Best Fit(execute=@n5.else, Best Fit=@n3)
   n6 = For Each Loop(Exec=@n2.else, Array=@n1)
   Return Node(execute=@n6.Completed, Exact Match=false)
   n7 = Get Variant(self=@n6)
   n8 = Get Variant!() -> (Variant)
   n9 = Get Asset!(self:Animatable Creature Object Reference=@n6) -> (Asset)
   n10 = Get Asset!() -> (Asset)
   > Branch(Condition=AND Boolean(A="Equal, Case Insensitive (String)"(A=@n10, B=@n9), B="Equal, Case Insensitive (String)"(A=@n7, B=@n8)))
   > Return Node(AnimatableCreature=@n6, Exact Match=true)
  ]}
 AnimatableCreature Get Assets {nodes [
   n1 = AnimatableCreature Get Assets!() -> (then, AnimatableCreatures:Array of Animatable Creature Object References) {is=title K2Node_FunctionEntry}
   > Clear(TargetArray=Get Local Categories() -> (Local Categories:Array of Strings))
   > n2 = For Each Loop(Array=@n1)
   n3 = Execute Python Script!(execute=@n2.Completed, PythonScript="sorted_creatures = sorted(creatures)", creatures=Get Local Categories()) -> (then, ReturnValue, sorted_creatures) {inputs [creatures], outputs [sorted_creatures]}
   > Return Node!(execute, Categories:Array of Strings=@n3.sorted_creatures)
   n4 = Get Asset!(self:Animatable Creature Object Reference=@n2) -> (Asset)
   n5 = Branch(execute=@n2.LoopBody, Condition=IsEmpty(InString=@n4))
   > Add Unique(TargetArray=Get Local Categories(), NewItem=Get Variant!(self:Animatable Creature Object Reference=@n2) -> (Variant))
   Add Unique(execute=@n5.else, TargetArray=Get Local Categories(), NewItem=@n4)
  ]}
 AnimatableCreature Get Categories {nodes [
   n1 = AnimatableCreature Get Categories!() -> (then, AnimatableCreatures:Array of Animatable Creature Object References) {is=title K2Node_FunctionEntry}
   > Clear(TargetArray=Get Local Categories())
   > n2 = For Each Loop(Array=@n1)
   n3 = Execute Python Script(execute=@n2.Completed, PythonScript="sorted_creatures = sorted(creatures)", creatures=Get Local Categories()) {inputs [creatures], outputs [sorted_creatures]}
   > Return Node(Categories=@n3.sorted_creatures)
   n4 = Get Asset(self=@n2)
   n5 = Branch(execute=@n2.LoopBody, Condition=IsEmpty(InString=@n4))
   > Add Unique(TargetArray=Get Local Categories(), NewItem=Get Variant(self=@n2))
   Add Unique(execute=@n5.else, TargetArray=Get Local Categories(), NewItem=@n4)
  ]}
 Browse To LS {nodes [
   Browse To LS!() -> (then, Context Item:Editor Utility Widget Object Reference, Slot Widget:Editor Utility Widget Object Reference) {is=title K2Node_FunctionEntry}
   > n1 = Cast To Object_LevelSequenceBrowser!(
    execute
    Object=Get Selected Item!(self:List View Object Reference=Get ListView() -> (ListView:Common List View Object Reference)) -> (ReturnValue:Object Reference) {is=title K2Node_CallFunction}
   ) -> (then, CastFailed, AsObject Level Sequence Browser:Object Level Sequence Browser Object Reference) {is=title K2Node_DynamicCast}
   n2 = Get!(self:Object Level Sequence Browser Object Reference=@n1) -> (Data_LS_58_#1:Soft Object Reference)
   > Branch(Condition=IsValidSoftObjectReference(SoftObjectReference=@n2))
   > SyncBrowserToObjects(AssetPaths=Make Array("[0]":String=To Soft Object Path!(SoftObjectReference:Soft Object Reference=@n2) -> (ReturnValue_PathString:String)) -> (Array:Array of Strings))
  ]}
 Browse_To_AnimSequencer {nodes [
   Browse_To_AnimSequencer!() -> (then, Context Item:Editor Utility Widget Object Reference, Slot Widget:Editor Utility Widget Object Reference) {is=title K2Node_FunctionEntry}
   > n1 = Cast To Object_LevelSequenceBrowser(Object=Get Selected Item(self=Get ListView()))
   n2 = To Soft Object Path(SoftObjectReference=Get(self=@n1))
   > n3 = LoadAsset_Blocking(Asset=To Soft Object Reference(SoftObjectPath=MakeSoftClassPath(PathString=@n2)))
   > n4 = Cast To LevelSequence!(execute, Object=@n3) -> (then, CastFailed, AsLevel Sequence:Level Sequence Object Reference) {is=title K2Node_DynamicCast}
   > n5 = Set Selected LS(Selected LS:Level Sequence Object Reference=@n4) -> (Output_Get:Level Sequence Object Reference)
   > n6 = GetBindings!(execute, Sequence:Movie Scene Sequence Object Reference=@n5) -> (then, ReturnValue:Array of Movie Scene Binding Proxy Structures)
   > n7 = GetAnimSequenceLinkFromLevelSequence!(execute, InLevelSequence:Level Sequence Object Reference=Get Selected LS() -> (Selected LS:Level Sequence Object Reference)) -> (then, ReturnValue:Level Sequence Anim Sequence Link Object Reference)
   > n8 = For Each Loop(Array:Array of Movie Scene Binding Proxy Structures=@n6) -> (Array Element:Movie Scene Binding Proxy Structure)
   > Set Anim_Movie_Binding(Anim_Movie_Binding:Movie Scene Binding Proxy Structure=@n8) -> (Output_Get:Movie Scene Binding Proxy Structure)
   > n9 = For Each Loop!(
    Exec
    Array:Array of Level Sequence Anim Sequence Link Item Structures=Get AnimSequenceLinks!(self:Level Sequence Anim Sequence Link Object Reference=@n7) -> (AnimSequenceLinks:Array of Level Sequence Anim Sequence Link Item Structures)
   ) -> (
    LoopBody
    Array Element_SkelTrackGuid:Guid Structure
    Array Element_PathToAnimSequence:Soft Object Path Structure
    Array Element_bAutoBake:bool=~
    Array Element_bExportTransforms:bool=~
    Array Element_bExportMorphTargets:bool=~
    Array Element_bExportAttributeCurves:bool=~
    Array Element_bExportMaterialCurves:bool=~
    Array Element_bSkipCurvesWithZeroValue:bool=~
    Array Element_bSetRetargetSourceAsset:bool=~
    Array Element_Interpolation:EAnimInterpolationType Enum
    Array Element_CurveInterpolation:ERichCurveInterpMode Enum
    Array Element_bRecordInWorldSpace:bool=~
    Array Element_bEvaluateAllSkeletalMeshComponents:bool=~
    Array Element_IncludeAnimationNames:Array of Strings
    Array Element_ExcludeAnimationNames:Array of Strings
    Array Element_bRemoveExcludedCurves:bool=~
    Array Element_WarmUpFrames:Frame Number Structure
    Array Element_DelayBeforeStart:Frame Number Structure
    Array Element_bUseCustomTimeRange:bool=~
    Array Element_CustomStartFrame:Frame Number Structure
    Array Element_CustomEndFrame:Frame Number Structure
    Array Element_CustomDisplayRate:Frame Rate Structure
    Array Element_bUseCustomFrameRate:bool=~
    Array Element_CustomFrameRate:Frame Rate Structure
    Array Index
    Completed
   )
   > n10 = FindBindingById!(execute, Sequence:Movie Scene Sequence Object Reference=Get Selected LS(), BindingId:Guid Structure=@n9) -> (then, ReturnValue_BindingID:Guid Structure, ReturnValue_Sequence:Movie Scene Sequence Object Reference)
   n11 = Get Anim_Movie_Binding() -> (Anim_Movie_Binding:Movie Scene Binding Proxy Structure)
   > Branch(Condition="Equal (Guid)"(
     A:"Guid Structure (by ref)"=@n10
     B:"Guid Structure (by ref)"=Break Movie Scene Binding Proxy!(MovieSceneBindingProxy:"Movie Scene Binding Proxy Structure (by ref)"=@n11) -> (BindingID:Guid Structure, Sequence:Movie Scene Sequence Object Reference) {is=title K2Node_BreakStruct}
    ) -> (ReturnValue:bool=~) {is=title K2Node_PromotableOperator})
   > n12 = LoadAsset(AssetPath=BreakSoftObjectPath(InSoftObjectPath=@n9.Array Element_PathToAnimSequence))
   > n13 = Cast To AnimSequence!(execute, Object=@n12) -> (then, CastFailed, AsAnim Sequence:Anim Sequence Object Reference)
   > SyncBrowserToObjects(AssetPaths=Make Array("[0]"=Get Object Path String(Object=@n13)))
  ]}
 Callback Populate AC Files from Project Selection {nodes [
   n1 = Callback Populate AC Files from Project Selection!() -> (then, Context Item:Editor Utility Widget Object Reference, Slot Widget:Editor Utility Widget Object Reference) {is=title K2Node_FunctionEntry}
   > n2 = Cast To EUW_ContextMenu_Icon!(execute, Object=@n1.Slot Widget) -> (then, CastFailed, AsEUW Context Menu Icon:EUW Context Menu Icon Object Reference) {is=title K2Node_DynamicCast}
   n3 = Get ButtonText!(self:EUW Context Menu Icon Object Reference=@n2) -> (ButtonText:Text Object Reference)
   > Set Selected Project Name(Selected Project Name:String="To String (Text)"(InText=Get Text(self:Text Object Reference=@n3))) -> (Output_Get:String)
   > n4 = GetAssetsByClass!(
    execute
    self:Asset Registry Interface=GetAssetRegistry()
    ClassPathName:Top Level Asset Path Structure=GetSoftClassTopLevelAssetPath!(SoftClassReference:Object Soft Class Reference=/AnimInEdTools/AnimationBrowser/ProjectSettings/BP_AnimationBrowser_Project.BP_AnimationBrowser_Project_C) -> (ReturnValue:Top Level Asset Path Structure)
    bSearchSubClasses:bool=false
   ) -> (then, OutAssetData:Array of Asset Data Structures, ReturnValue:bool)
   > n5 = For Each Loop with Break(Array:Array of Asset Data Structures=@n4, Break=@n9.then) -> (Array Element:Asset Data Structure)
   n6 = Break Asset Data(AssetData=@n5)
   > n7 = LoadAsset(AssetPath=GetBaseFilename(InPath="To String (Name)"(InName=@n6), bRemovePath=false))
   > n8 = Cast To BP_AnimationBrowser_Project!(execute, Object=@n7) -> (then, CastFailed, AsBP Animation Browser Project:BP Animation Browser Project Object Reference) {is=title K2Node_DynamicCast}
   > Branch(Condition="Equal Exactly (String)"(
     A=Get Selected Project Name() -> (Selected Project Name:String)
     B=Get Name!(self:BP Animation Browser Project Object Reference=@n8) -> (Name:String)
    ))
   > n9 = Set Project!(execute, self, Project:BP Animation Browser Project Object Reference=@n8, Save Settings:bool=true) -> (then) {is=title K2Node_CallFunction}
  ]}
 Checked out by another user warning {nodes [
   n1 = Checked out by another user warning!() -> (then, Checked out by:String, Dipslay aset name:String) {is=title K2Node_FunctionEntry}
   > Show Message Dialog(
    Title=MakeLiteralText(Value="LOCTEXT(\"Cannot Edit File\")")
    Message="To Text (String)"(InString=Append!(A="The seleted file  ", B=@n1.Dipslay aset name, C:String=" is currently checked out by ", D:String=@n1) -> (ReturnValue))
    MessageType=Ok
    DefaultValue=Ok
    MessageCategory=Error
   )
   > Return Node!(execute, Was it done:bool=false)
  ]}
 Display Modal Dialog {nodes [
   n1 = Display Modal Dialog!() -> (then, Class:User Widget Class Reference) {is=title K2Node_FunctionEntry}
   > n2 = Create Widget!(
    execute:Exec
    WorldContextObject:Object Reference
    Class:User Widget Class Reference=/AnimInEdTools/AnimationBrowser/EUW_ModalDialog.EUW_ModalDialog_C
    OwningPlayer:Player Controller Object Reference
   ) -> (then:Exec, ReturnValue:EUW Modal Dialog Object Reference)
   > n3 = Create Widget(Class=@n1) -> (ReturnValue:User Widget Object Reference)
   > AddChild(self=Get NamedSlot!(self:EUW Modal Dialog Object Reference=@n2) -> (NamedSlot:Named Slot Object Reference), Content=@n3)
   > AddChild(self=Get ModalLayer() -> (ModalLayer:Named Slot Object Reference), Content=@n2)
   > Return Node!(execute, Modal Widget:EUW Modal Dialog Object Reference=@n2, New Widget:User Widget Object Reference=@n3)
  ]}
 ED_AssetHasChanged {nodes [ED_AssetHasChanged!() -> (then, NewAsset:String, Current Project:BP Animation Browser Project Object Reference) {is=title K2Node_FunctionEntry}]}
 ED_ListViewItemSelected {nodes [ED_ListViewItemSelected!() -> (then, ItemSelected:bool=~) {is=title K2Node_FunctionEntry}]}
 EventGraph {nodes [
   n1 = "On Item Double Clicked (ListView)"!() -> (OutputDelegate, then, Item:Object Reference) {is=title K2Node_ComponentBoundEvent}
   > n2 = Cast To Object_LevelSequenceBrowser(Object=@n1.Item)
   n3 = To Soft Object Path(SoftObjectReference=Get(self=@n2))
   > n4 = LoadAsset_Blocking(Asset=To Soft Object Reference(SoftObjectPath=MakeSoftClassPath(PathString=@n3)))
   > n5 = Cast To LevelSequence(Object=@n4)
   > OpenLevelSequence!(execute, LevelSequence:Level Sequence Object Reference=@n5) -> (then, ReturnValue:bool)
   OpenEditorForAssets!(
    execute=@n5.CastFailed
    self:Asset Editor Subsystem Object Reference=Get AssetEditorSubsystem()
    Assets:Array of Object References=Make Array("[0]":Object Reference=@n4) -> (Array:Array of Object References)
    OpenedMethod:EAssetTypeActivationOpenedMethod Enum=Edit
   ) -> (then, ReturnValue:bool)
   "On Item Clicked (ListView)"!() -> (OutputDelegate, then, Item:Object Reference) {is=title K2Node_ComponentBoundEvent}
   > Call ED List View Item Selected!(execute:Exec, self:EUW Level Sequence Browser Object Reference, ItemSelected:bool=true) -> (then:Exec) {is=title K2Node_CallDelegate}
   > Show!(
    execute
    self:EUW Context Menu Object Reference=Get EUW_ContextMenu() -> (EUW_ContextMenu:EUW Context Menu Object Reference)
    Show:bool=false
    Mouse Position:Vector 2D Structure
    Name:String
   ) -> (then)
   "On Opening (Asset_Searchable_ComboBox)"() {is=title K2Node_ComponentBoundEvent}
   > Update Combo Box!(
    execute
    self
    Combo Box:"Searchable Combo Box (String) Object Reference"=Get Asset_Searchable_ComboBox() -> (Asset_Searchable_ComboBox:"Searchable Combo Box (String) Object Reference")
    Items:Array of Strings=AnimatableCreature Get Assets!(AnimatableCreatures:Array of Animatable Creature Object References=Get Animatable Creatures() -> (Animatable Creatures:Array of Animatable Creature Object References)) -> (Categories:Array of Strings) {is=title K2Node_CallFunction}
    Default Item:String
    Add Blank:bool=false
   ) -> (then)
   > Show(self=Get EUW_ContextMenu(), Show=false)
   "On Opening (Plugin_Searchable_ComboBox)"() {is=title K2Node_ComponentBoundEvent}
   > Update Combo Box(
    Combo Box=Get Plugin_Searchable_ComboBox() -> (Plugin_Searchable_ComboBox:"Searchable Combo Box (String) Object Reference")
    Items=Animatable Creature Get Variations!(
     AnimatableCreatures:Array of Animatable Creature Object References=Get Animatable Creatures()
     Asset Name:String=Get Selected_Asset() -> (Selected_Asset:String)
     Add All Option:bool=true
    ) -> (Variations:Array of Strings) {is=title K2Node_CallFunction}
    Add Blank=false
   )
   > Show(self=Get EUW_ContextMenu(), Show=false)
   "On Opening (AnimationType_CombBox)"() {is=title K2Node_ComponentBoundEvent}
   > Update Combo Box(
    Combo Box=Get AnimationType_CombBox() -> (AnimationType_CombBox:"Searchable Combo Box (String) Object Reference")
    Items=Get Subfolders!() -> (Subfolders:Array of Strings) {is=title K2Node_CallFunction}
    Default Item=Get Animation Type Default Item() -> (Animation Type Default Item:String)
    Add Blank=false
   )
   > Show(self=Get EUW_ContextMenu(), Show=false)
   "On Clicked (ClearSearchBox)"() {is=title K2Node_ComponentBoundEvent}
   > Set Text(self:Text Box Object Reference=Get SearchBar() -> (SearchBar:Editor Utility Editable Text Box Object Reference))
   > n6 = Set SearchBar_Filter(SearchBar_Filter:String) -> (Output_Get:String)
   RetriggerableDelay!(execute=[@n34.then, @n6.then, @n26.then], WorldContextObject:Object Reference, Duration=0.2) -> (then)
   > Update List View!(
    execute
    self
    Filter:String=Get SearchBar_Filter() -> (SearchBar_Filter:String)
    Use Cache:bool=true
    Current Plugin:String=Get Selected Plugin() -> (Selected Plugin:String)
    Current Subfolder:String="Get Selected Subfolder "() -> ("Selected Subfolder ":String)
    Current Asset:String=Get Selected_Asset()
   ) -> (then)
   > Save Settings()
   "On Clicked (SettingsButton)"() {is=title K2Node_ComponentBoundEvent}
   n7 = Get EUW_ContextMenu()
   n8 = Get Name(self:EUW Context Menu Object Reference=@n7)
   > n9 = Branch(Condition=AND Boolean(A=IsVisible(self=@n7), B="Equal Exactly (String)"(A=@n8, B=Settings)))
   > Show(self=Get EUW_ContextMenu(), Show=false)
   Clear Context Menu(execute=@n9.else, self:EUW Context Menu Object Reference=Get EUW_ContextMenu())
   > n10 = Sequence!(execute) -> (then_0, then_1, then_2:Exec, then_3:Exec)
   > Add Icon Item!(
    execute
    self:EUW Context Menu Object Reference=Get EUW_ContextMenu()
    Event:"Delegate (by ref)"=Create Event()
    Category:String
    Name:String=Refresh Assets
    Font Size:double=10
    Icon:Texture 2D Object Reference=/AnimInEdTools/AnimationBrowser/Icons/RefreshSpinner
   ) -> (then, Context Menu Icon:EUW Context Menu Icon Object Reference)
   > Add Icon Item(self=Get EUW_ContextMenu(), Event=Create Event(), Category=Developer, Name=Browse to AC file, Font Size=10)
   n11 = For Each Loop!(
    Exec=@n10.then_1
    Array:Array of BP Animation Browser Project Object References=Get Projects!() -> (Projects:Array of BP Animation Browser Project Object References) {is=title K2Node_CallFunction}
   ) -> (LoopBody, Array Element:BP Animation Browser Project Object Reference, Array Index, Completed)
   > Add Icon Item(
    self=Get EUW_ContextMenu()
    Event=Create Event()
    Category=Projects
    Name=Get Name(self:BP Animation Browser Project Object Reference=@n11)
    Font Size=10
    Icon=/AnimInEdTools/ContextMenu/Icon/Icon_Edit
   )
   n12 = GetCachedGeometry!(self:Widget Object Reference=Get SettingsButton() -> (SettingsButton:Button Object Reference)) -> (ReturnValue:"Geometry Structure (by ref)")
   n13 = GetLocalSize!(Geometry:"Geometry Structure (by ref)"=@n12) -> (ReturnValue_X:double, ReturnValue_Y:double)
   Show(
    execute=@n10.then_3
    self=Get EUW_ContextMenu()
    Show=true
    Mouse Position="Local to Absolute (Coordinates)"!(Geometry:"Geometry Structure (by ref)"=@n12, LocalCoordinate_X:double=@n13, LocalCoordinate_Y:double=@n13.ReturnValue_Y) -> (ReturnValue:Vector 2D Structure)
    Name=Settings
   )
   n14 = "On Selection Changed (Asset_Searchable_ComboBox)"!() -> (OutputDelegate, then, SelectedItem:String, SelectionType:ESelectInfo Enum) {is=title K2Node_ComponentBoundEvent}
   > n15 = Sequence!(execute) -> (then_0, then_1)
   > n16 = Set Selected_Asset(Selected_Asset:String=@n14.SelectedItem) -> (Output_Get:String)
   n17 = Animatable Creature Get Variations(AnimatableCreatures=Get Animatable Creatures(), Asset Name=@n16, Add All Option=true)
   > Update Combo Box(Combo Box=Get Plugin_Searchable_ComboBox(), Items=@n17, Add Blank=false)
   n18 = Get AnimationType_CombBox()
   > Update Combo Box(Combo Box=@n18, Items=Get Subfolders(), Default Item=Get Animation Type Default Item(), Add Blank=false)
   > SetSelectedOption!(execute, self:"Searchable Combo Box (String) Object Reference"=@n18, Option:String) -> (then)
   > Branch(Condition="integer > integer"(A=Length(TargetArray:Array of Strings=@n17)))
   > SetSelectedOption(self=Get Plugin_Searchable_ComboBox(), Option="Get (a copy)"(Array=@n17))
   > Update List View From UI()
   Call ED Asset Has Changed!(
    execute:Exec=@n15.then_1
    self:EUW Level Sequence Browser Object Reference
    NewAsset:String=Get Selected_Asset()
    Current Project:BP Animation Browser Project Object Reference=Get Current Project() -> (Current Project:BP Animation Browser Project Object Reference)
   ) -> (then:Exec) {is=title K2Node_CallDelegate}
   > n19 = Switch on ESelectInfo!(execute, Selection:ESelectInfo Enum=@n14.SelectionType) -> (OnKeyPress:Exec, OnNavigation:Exec, OnMouseClick:Exec, Direct:Exec) {is=title K2Node_SwitchEnum}
   Set Asset!(
    execute=[@n19.OnKeyPress, @n19.OnMouseClick]
    Asset:String=Get Selected_Asset()
    self:SG Level Sequence Browser Object Reference=Get Save Game() -> (Save Game:SG Level Sequence Browser Object Reference)
   ) -> (then, Output_Get:String)
   > Save Settings()
   n20 = "On Selection Changed (Plugin_Searchable_ComboBox)"!() -> (OutputDelegate, then, SelectedItem:String, SelectionType:ESelectInfo Enum) {is=title K2Node_ComponentBoundEvent}
   > n21 = Switch on ESelectInfo(Selection=@n20.SelectionType)
   n22 = Sequence(execute=[@n21.OnKeyPress, @n21.OnMouseClick])
   > n23 = Branch(Condition="Equal Exactly (String)"(A=@n20.SelectedItem, B=Get Selected Plugin()))
   Update List View(
    execute=[@n23.else, @n21.Direct, @n21.OnNavigation]
    Use Cache=false
    Current Plugin=@n20.SelectedItem
    Current Subfolder="Get Selected Subfolder "()
    Current Asset=Get Selected_Asset()
   )
   n24 = Set Selected Plugin(execute=@n22.then_1, Selected Plugin:String=@n20.SelectedItem) -> (Output_Get:String)
   > Set Variant!(execute, Variant:String=@n24, self:SG Level Sequence Browser Object Reference=Get Save Game()) -> (then, Output_Get:String)
   > Save Settings()
   n25 = "On Text Changed (SearchBar)"!() -> (OutputDelegate, then, Text:"Text (by ref)") {is=title K2Node_ComponentBoundEvent}
   > n26 = Set SearchBar_Filter(SearchBar_Filter="To String (Text)"(InText=@n25.Text))
   n27 = "On Selection Changed (AnimationType_CombBox)"!() -> (OutputDelegate, then, SelectedItem:String, SelectionType:ESelectInfo Enum) {is=title K2Node_ComponentBoundEvent}
   > n28 = Sequence()
   n29 = GetSelectedOption!(self:"Searchable Combo Box (String) Object Reference"=Get AnimationType_CombBox()) -> (ReturnValue:String)
   n30 = "Get Selected Subfolder "()
   > n31 = Branch(Condition=OR Boolean(A="Equal Exactly (String)"(A=@n30, B=@n29), B=IsEmpty(InString=@n27.SelectedItem)))
   Update List View(
    execute=@n31.else
    Use Cache=false
    Current Plugin=Get Selected Plugin()
    Current Subfolder=@n27.SelectedItem
    Current Asset=Get Selected_Asset()
   )
   n32 = Switch on ESelectInfo(execute=@n28.then_1, Selection=OnKeyPress)
   Set Subfolder!(
    execute=[@n32.OnKeyPress, @n32.OnMouseClick, @n32.OnNavigation]
    Subfolder:String=@n27.SelectedItem
    self:SG Level Sequence Browser Object Reference=Get Save Game()
   ) -> (then, Output_Get:String)
   > Save Settings()
   n33 = "On Text Committed (SearchBar)"!() -> (OutputDelegate, then, Text:"Text (by ref)", CommitMethod:ETextCommit Enum=Default) {is=title K2Node_ComponentBoundEvent}
   > n34 = Set SearchBar_Filter(SearchBar_Filter="To String (Text)"(InText=@n33.Text))
   n35 = CustomEvent_2!() -> (OutputDelegate, then, Context Item:Editor Utility Widget Object Reference, Slot Widget:Editor Utility Widget Object Reference) {is=title K2Node_CustomEvent}
   > OpenAnimSeq()
   n36 = "On Item Right Clicked (List View)"!() -> (OutputDelegate, then, Item:Object Level Sequence Browser Object Reference, Mouse Event:Pointer Event Structure) {is=title K2Node_CustomEvent}
   > Set Selected Item!(execute, self:List View Object Reference=Get ListView(), Item:Object Reference=@n36.Item) -> (then) {is=title K2Node_CallFunction}
   > Clear Context Menu(self=Get EUW_ContextMenu())
   > n37 = Sequence!(execute) -> (then_0, then_1, then_2:Exec, then_3:Exec)
   > Add Icon Item(self=Get EUW_ContextMenu(), Event=Create Event(), Category=Explore, Name=Browse To LevelSequence, Font Size=10)
   Add Icon Item(execute=@n37.then_1, self=Get EUW_ContextMenu(), Event=Create Event(), Category=Explore, Name=Browse To AnimSequence, Font Size=10)
   Show(
    execute=@n37.then_3
    self=Get EUW_ContextMenu()
    Show=true
    Mouse Position=Get Screen Space Position!(Input:"Pointer Event Structure (by ref)"=@n36.Mouse Event) -> (ReturnValue:Vector 2D Structure) {is=title K2Node_CallFunction}
   )
   Add Icon Item(execute=@n37.then_2, self=Get EUW_ContextMenu(), Event=@n35, Category=Open, Name=Open AnimSequence, Font Size=10)
   Event Construct()
   > n38 = Branch(Condition=IsLoadingAssets!(self:Asset Registry Interface=GetAssetRegistry()) -> (ReturnValue:bool))
   > n39 = Set Timer by Function Name!(
    execute
    Object:Object Reference
    FunctionName:String=Try Setup
    Time:float=0.2
    bLooping:bool=true
    bMaxOncePerFrame:bool=false
    InitialStartDelay:float=0
    InitialStartDelayVariance:float=0
   ) -> (then, ReturnValue:Timer Handle Structure) {is=title K2Node_CallFunction}
   > Set SetupTimerHandle(SetupTimerHandle:Timer Handle Structure=@n39) -> (Output_Get:Timer Handle Structure)
   Load From Save(execute=@n38.else)
  ]}
 Find Project {nodes [
   Find Project!() -> (then, Project Name:String) {is=title K2Node_FunctionEntry}
   > n1 = GetAssetsByClass(
    self=GetAssetRegistry()
    ClassPathName=GetSoftClassTopLevelAssetPath(SoftClassReference=&BP_AnimationBrowser_Project_C)
    bSearchSubClasses=false
   )
   > n2 = For Each Loop(Array:Array of Asset Data Structures=@n1) -> (Array Element:Asset Data Structure)
   Return Node!(execute=@n2.Completed, AsBP Animation Browser Project:BP Animation Browser Project Object Reference)
   n3 = Break Asset Data(AssetData=@n2)
   > n4 = LoadAsset(AssetPath=GetBaseFilename(InPath="To String (Name)"(InName=@n3), bRemovePath=false))
   > n5 = Cast To BP_AnimationBrowser_Project(Object=@n4)
   > Branch(Condition="Equal Exactly (String)"(A=Get Name(self=@n5), B=Get Project Name() -> (Project Name:String)))
   > Return Node(AsBP Animation Browser Project=@n5)
  ]}
 Get Animation Directories for Asset {nodes [
   Get Animation Directories for Asset!() -> (then, Asset Name:String, Plugin:String, Animation Type:String) {is=title K2Node_FunctionEntry}
   > Clear(TargetArray:Array of Names=Get Local Directories() -> (Local Directories:Array of Names))
   > n1 = For Each Loop(Array:Array of Animatable Creature Object References=Get Animatable Creatures()) -> (Array Element:Animatable Creature Object Reference)
   Return Node!(execute=@n1.Completed, Animation Directories:Array of Names=Get Local Directories())
   n2 = Get Asset Name()
   n3 = Get Variant(self=@n1)
   n4 = Get Asset(self=@n1)
   n5 = "Equal Exactly (String)"(A=Get Asset(self=@n1), B=Get Asset Name())
   n6 = OR Boolean(A="Equal Exactly (String)"(A=Get Variant(self=@n1), B=Get Selected Plugin()), B="Equal Exactly (String)"(A=Get Selected Plugin(), B=All))
   > Branch(Condition=Select(Option 0:bool=AND Boolean(A=@n6, B=@n5), Option 1:bool="Equal Exactly (String)"(A=@n3, B=@n2), Index=IsEmpty(InString=@n4)) -> (ReturnValue:bool))
   n7 = Get Populate Subfolders From Disk!(self:Animatable Creature Object Reference=@n1) -> (Populate Subfolders From Disk:bool)
   > n8 = Branch(Condition=@n7)
   > n9 = Get Subfolders From Disk!(execute, self, Animatable Creature:Animatable Creature Object Reference=@n1) -> (then, Subfolders:Array of Strings) {is=title K2Node_CallFunction}
   n10 = Get Animation Type Default Item()
   n11 = Get Animation Type() -> (Animation Type:String)
   n12 = Get Subfolders!(self:Animatable Creature Object Reference=@n1) -> (Subfolders)
   n13 = For Each Loop(
    Exec=[@n9.then, @n8.else]
    Array:Array of Strings=Select(
     Option 0:Array of Strings=Make Array("[0]":String=@n11) -> (Array:Array of Strings)
     Option 1:Array of Strings=Select(Option 0=@n12, Option 1=@n9, Index=@n7)
     Index="Equal Exactly (String)"(A=@n11, B=@n10)
    ) -> (ReturnValue:Array of Strings)
   ) -> (Array Element:String)
   n14 = Get Path!(
    self:EU Animation Naming Convention Object Reference=Get Naming Convention!(self:Animatable Creature Object Reference=@n1) -> (ReturnValue:EU Animation Naming Convention Object Reference) {is=title K2Node_CallFunction}
    Token Map:Map of Strings to Strings=Get Path Parts Map!(
     Asset:String=Get Asset(self=@n1)
     Animation Set:String
     Animation Name:String
     Descriptor:String
     Direction:String
     Subfolder:String=@n13
     Variation:String
     Plugin:String=Get Variant(self=@n1)
     __WorldContext:Object Reference
    ) -> (Output Path Parts Map:Map of Strings to Strings) {is=title K2Node_CallFunction}
    Path Pattern:String=Get Animation Path!(self:Animatable Creature Object Reference=@n1) -> (Animation Path:String)
    Remove Spaces On Name:bool=Get RemoveSpacesOnName!(self:Animatable Creature Object Reference=@n1) -> (RemoveSpacesOnName:bool)
   ) -> (File Path:String, Folder Path:String, File Name:String) {is=title K2Node_CallFunction}
   > n15 = DoesDirectoryExist!(execute, DirectoryPath:String=@n14.Folder Path) -> (then, ReturnValue:bool)
   > Branch(Condition=@n15)
   > Add(TargetArray:Array of Names=Get Local Directories(), NewItem:"Name (by ref)"=String To Name(InString=@n14.Folder Path))
  ]}
 Get Projects {nodes [
   Get Projects() {is=title K2Node_FunctionEntry}
   > Clear(TargetArray:Array of BP Animation Browser Project Object References=Get Local Projects() -> (Local Projects:Array of BP Animation Browser Project Object References))
   > n1 = GetAssetsByClass(
    self=GetAssetRegistry()
    ClassPathName=GetSoftClassTopLevelAssetPath(SoftClassReference=&BP_AnimationBrowser_Project_C)
    bSearchSubClasses=false
   )
   > n2 = For Each Loop(Array:Array of Asset Data Structures=@n1) -> (Array Element:Asset Data Structure)
   Return Node!(execute=@n2.Completed, Projects:Array of BP Animation Browser Project Object References=Get Local Projects())
   n3 = Break Asset Data(AssetData=@n2)
   > n4 = LoadAsset(AssetPath=GetBaseFilename(InPath="To String (Name)"(InName=@n3), bRemovePath=false))
   > n5 = Cast To BP_AnimationBrowser_Project(Object=@n4)
   > Add(
    TargetArray:Array of BP Animation Browser Project Object References=Get Local Projects()
    NewItem:"BP Animation Browser Project Object Reference (by ref)"=@n5
   )
  ]}
 Get Subfolders {nodes [
   Get Subfolders() {is=title K2Node_FunctionEntry}
   > Clear(TargetArray:Array of Strings=Get NewLocalVar() -> (NewLocalVar:Array of Strings))
   n1 = Get Animatable Creatures()
   n2 = Get Selected_Asset()
   n3 = Get Selected Plugin()
   n4 = Get Selected Plugin()
   > n5 = For Each Loop(Array:Array of Strings=Select(
     Option 0=Make Array("[0]"=@n4)
     Option 1=Animatable Creature Get Variations(AnimatableCreatures=@n1, Asset Name=@n2, Add All Option=true)
     Index="Equal, Case Insensitive (String)"(A=@n3, B=All)
    )) -> (Array Element:String)
   Return Node!(execute=@n5.Completed, Subfolders:Array of Strings=Get NewLocalVar())
   n6 = AnimatableCreature Find!(
    AnimatableCreatures:Array of Animatable Creature Object References=Get Animatable Creatures()
    Asset:String=Get Selected_Asset()
    Variant:String=@n5
   ) -> (AnimatableCreature:Animatable Creature Object Reference, Exact Match:bool) {is=title K2Node_CallFunction}
   > Is Valid(InputObject=@n6)
   > n7 = For Each Loop(Array=Get Subfolders(self=@n6) {is=title K2Node_CallFunction})
   > Add Unique(TargetArray=Get NewLocalVar(), NewItem=@n7)
  ]}
 Get Subfolders From Disk {nodes [
   n1 = Get Subfolders From Disk!() -> (then, Animatable Creature:Animatable Creature Object Reference) {is=title K2Node_FunctionEntry}
   > n2 = Sequence!(execute) -> (then_0, then_1)
   > n3 = Get Subfolder Root!(execute, self:EU Animation Naming Convention Object Reference=Get Naming Convention(self=@n1), Path:String=Get Animation Path(self=@n1)) -> (then, Root Path:String) {is=title K2Node_CallFunction}
   > n4 = ListAssets!(execute, DirectoryPath:String=@n3, bRecursive:bool=false, bIncludeFolder:bool=true) -> (then, ReturnValue:Array of Strings)
   > n5 = For Each Loop(Array=@n4)
   > n6 = DoesDirectoryExist(DirectoryPath=@n5)
   > Branch(Condition=@n6)
   > Add Unique(
    TargetArray=Get Subfolders() -> (Subfolders:Array of Strings) {is=title K2Node_VariableGet}
    NewItem=GetCleanFilename!(InPath:String=Split(SourceString=@n5, InStr=/, SearchDir=FromEnd)) -> (ReturnValue:String)
   )
   Return Node(execute=@n2.then_1, Subfolders=Get Subfolders())
  ]}
 Get Subfolders From Disk_0 {nodes [
   n1 = Get Subfolders From Disk_0!() -> (then, Animatable Creature:Animatable Creature Object Reference) {is=title K2Node_FunctionEntry}
   > n2 = Sequence()
   > n3 = Get Subfolder Root(self=Get Naming Convention(self=@n1), Path=Get Animation Path(self=@n1))
   > n4 = ListAssets(DirectoryPath=@n3, bRecursive=false, bIncludeFolder=true)
   > n5 = For Each Loop(Array=@n4)
   > n6 = DoesDirectoryExist(DirectoryPath=@n5)
   > Branch(Condition=@n6)
   > Add Unique(TargetArray=Get Subfolders(), NewItem=GetCleanFilename(InPath=Split(SourceString=@n5, InStr=/, SearchDir=FromEnd)))
   Return Node(execute=@n2.then_1, Subfolders=Get Subfolders())
  ]}
 Get Subfolders_0 {nodes [
   Get Subfolders_0() {is=title K2Node_FunctionEntry}
   > n1 = Sequence()
   > Clear(TargetArray=Get NewLocalVar())
   n2 = Get Selected_Asset()
   n3 = Get Animatable Creatures()
   n4 = Get Selected Plugin()
   n5 = Get Selected Plugin()
   > n6 = For Each Loop(Array=Select(
     Option 0=Make Array("[0]"=@n5)
     Option 1=Animatable Creature Get Variations(AnimatableCreatures=@n3, Asset Name=@n2, Add All Option=false)
     Index="Equal, Case Insensitive (String)"(A=@n4, B=All)
    ))
   n7 = AnimatableCreature Find(AnimatableCreatures=Get Animatable Creatures(), Asset=Get Selected_Asset(), Variant=@n6)
   > Is Valid(InputObject=@n7)
   > n8 = Branch(Condition=Get Populate Subfolders From Disk(self=@n7))
   > n9 = Get Subfolder Root(self=Get Naming Convention(self=@n7), Path=Get Animation Path(self=@n7))
   > n10 = ListAssets(DirectoryPath=@n9, bRecursive=false, bIncludeFolder=true)
   > n11 = For Each Loop(Array=@n10)
   > n12 = DoesDirectoryExist(DirectoryPath=@n11)
   > Branch(Condition=@n12)
   > Add Unique(TargetArray=Get NewLocalVar(), NewItem=GetCleanFilename(InPath=Split(SourceString=@n11, InStr=/, SearchDir=FromEnd)))
   n13 = For Each Loop(Exec=@n8.else, Array=Get Subfolders(self=@n7) {is=title K2Node_CallFunction})
   > Add Unique(TargetArray=Get NewLocalVar(), NewItem=@n13)
   Return Node(execute=@n1.then_1, Subfolders=Get NewLocalVar())
  ]}
 Load From Save {nodes [
   Load From Save() {is=title K2Node_FunctionEntry}
   > n1 = LoadGameFromSlot!(execute, SlotName:String=Get Slot Name() -> (Slot Name:String), UserIndex:int=0) -> (then, ReturnValue:Save Game Object Reference)
   > n2 = Is Valid(InputObject=@n1)
   ClearChildren(execute=@n2.Is Not Valid, self=Get ModalLayer())
   > n3 = Display Modal Dialog!(execute, self, Class:User Widget Class Reference=/AnimInEdTools/AnimationBrowser/EUW_SelectProjectDialog.EUW_SelectProjectDialog_C) -> (then, Modal Widget:EUW Modal Dialog Object Reference, New Widget:User Widget Object Reference) {is=title K2Node_CallFunction}
   > n4 = Cast To EUW_SelectProjectDialog!(execute, Object=@n3.New Widget) -> (then, CastFailed, AsEUW Select Project Dialog:EUW Select Project Dialog Object Reference) {is=title K2Node_DynamicCast}
   > Set Projects!(
    execute
    self:EUW Select Project Dialog Object Reference=@n4
    Projects:Array of BP Animation Browser Project Object References=Get Projects() {is=title K2Node_CallFunction}
   ) -> (then) {is=title K2Node_CallFunction}
   > Bind Event to Project Selected(self:EUW Select Project Dialog Object Reference=@n4, Delegate=Create Event()) {is=title K2Node_AddDelegate}
   > n5 = CreateSaveGameObject!(execute, SaveGameClass:Save Game Class Reference=&/SG_LevelSequenceBrowser.SG_LevelSequenceBrowser_C) -> (then, ReturnValue:SG Level Sequence Browser Object Reference)
   > n6 = Set Save Game(Save Game:SG Level Sequence Browser Object Reference=@n5) -> (Output_Get:SG Level Sequence Browser Object Reference)
   > SaveGameToSlot!(execute, SaveGameObject:Save Game Object Reference=@n6, SlotName:String=Get Slot Name(), UserIndex:int=0) -> (then, ReturnValue:bool)
   n7 = Cast To SG_LevelSequenceBrowser!(execute=@n2.Is Valid, Object=@n1) -> (then, CastFailed, AsSG Level Sequence Browser:SG Level Sequence Browser Object Reference) {is=title K2Node_DynamicCast}
   > Set Save Game(Save Game=@n7)
   > n8 = LoadAsset_Blocking(Asset=Get Project!(self:SG Level Sequence Browser Object Reference=@n7) -> (Project:BP Animation Browser Project Soft Object Reference))
   > n9 = Cast To BP_AnimationBrowser_Project(Object=@n8)
   > Is Valid(InputObject=@n9)
   > Set Current Project(Current Project:BP Animation Browser Project Object Reference=@n9) -> (Output_Get:BP Animation Browser Project Object Reference)
   > Refresh Assets()
   > Update Combo Box(
    Combo Box=Get Asset_Searchable_ComboBox()
    Items=AnimatableCreature Get Categories!(AnimatableCreatures:Array of Animatable Creature Object References=Get Animatable Creatures()) -> (Categories:Array of Strings) {is=title K2Node_CallFunction}
    Default Item=Get Asset(self:SG Level Sequence Browser Object Reference=Get Save Game())
    Add Blank=false
   )
   > Updating Image()
   > SetSelectedOption(self=Get Asset_Searchable_ComboBox(), Option=Get Asset(self=Get Save Game()))
   > Set Selected_Asset(Selected_Asset=Get Asset(self=Get Save Game()))
   > Update Combo Box(
    Combo Box=Get AnimationType_CombBox()
    Items=Get Subfolders!() -> (Subfolders)
    Default Item=Get Animation Type Default Item()
    Add Blank=true
   )
   > SetSelectedOption(self=Get Plugin_Searchable_ComboBox(), Option=Get Variant(self:SG Level Sequence Browser Object Reference=Get Save Game()))
   > SetSelectedOption(
    self=Get AnimationType_CombBox()
    Option=Get Subfolder!(self:SG Level Sequence Browser Object Reference=Get Save Game()) -> (Subfolder:String)
   )
  ]}
 Not Current Warning {nodes [
   n1 = Not Current Warning!() -> (then, Type of Asset:String, Display Asset Name:String) {is=title K2Node_FunctionEntry}
   > Show Message Dialog(
    Title=Get Popup_Title() -> (Popup_Title:Text)
    Message="To Text (String)"(InString=Append!(A=@n1, B=@n1.Display Asset Name, C=" is not the latest. Please get the latest for this asset and try again.") -> (ReturnValue))
    MessageType=Ok
    DefaultValue=No
    MessageCategory=Error
   )
   > Return Node!(execute, Could it be done:bool=false)
   GetDisplayName(Object=Get Selected_LS() -> (Selected_LS:Level Sequence Object Reference))
  ]}
 OnMouseButtonDown {nodes [
   OnMouseButtonDown!() -> (then, MyGeometry:Geometry Structure, MouseEvent:"Pointer Event Structure (by ref)") {is=title K2Node_FunctionEntry}
   > Show(self=Get EUW_ContextMenu(), Show=false)
   > Return Node!(execute, ReturnValue:Event Reply Structure=Unhandled())
  ]}
 "Open, Save, Close Level Sequence" {nodes [
   n1 = "Open, Save, Close Level Sequence"!() -> (then, List Objects selected:Array of Object References) {is=title K2Node_FunctionEntry}
   > n2 = For Each Loop(Array:Array of Object References=@n1) -> (Array Element:Object Reference)
   Return Node!(execute=@n2.Completed, WasItDone:bool=Get Was the procedure done() -> (Was the procedure done:bool))
   > n3 = Cast To Object_LevelSequenceBrowser(Object=@n2)
   n4 = To Soft Object Path(SoftObjectReference=Get(self=@n3))
   > n5 = LoadAsset_Blocking(Asset=To Soft Object Reference(SoftObjectPath=MakeSoftClassPath(PathString=@n4)))
   > n6 = Cast To LevelSequence(Object=@n5)
   > Set Selected_LS(Selected_LS:Level Sequence Object Reference=@n6) -> (Output_Get:Level Sequence Object Reference)
   > n7 = QueryFileState!(
    execute
    InFile:String=GetSystemPath!(Object:Object Reference=Get Selected_LS()) -> (ReturnValue:String)
    bSilent:bool=false
    bUseSourceControlStateCache:bool=false
   ) -> (
    then
    ReturnValue_Filename:String
    ReturnValue_bIsValid:bool=~
    ReturnValue_bIsUnknown:bool=~
    ReturnValue_bCanCheckIn:bool=~
    ReturnValue_bCanCheckOut:bool=~
    ReturnValue_bIsCheckedOut:bool=~
    ReturnValue_bIsCurrent:bool=~
    ReturnValue_bIsSourceControlled:bool=~
    ReturnValue_bIsAdded:bool=~
    ReturnValue_bIsDeleted:bool=~
    ReturnValue_bIsIgnored:bool=~
    ReturnValue_bCanEdit:bool=~
    ReturnValue_bCanDelete:bool=~
    ReturnValue_bIsModified:bool=~
    ReturnValue_bCanAdd:bool=~
    ReturnValue_bIsConflicted:bool=~
    ReturnValue_bCanRevert:bool=~
    ReturnValue_bIsCheckedOutOther:bool=~
    ReturnValue_CheckedOutOther:String
    ReturnValue_bIsCheckedOutInOtherBranch:bool=~
    ReturnValue_bIsModifiedInOtherBranch:bool=~
    ReturnValue_PreviousUser:String
   )
   > n8 = Branch(Condition=@n7.ReturnValue_bIsCurrent)
   > n9 = Branch(Condition=@n7.ReturnValue_bIsCheckedOutOther)
   > n10 = Checked out by another user warning!(execute, self, Checked out by:String=@n7.ReturnValue_CheckedOutOther, Dipslay aset name:String=GetDisplayName(Object=Get Selected_LS())) -> (then, Was it done:bool) {is=title K2Node_CallFunction}
   > Set Was the procedure done(Was the procedure done:bool=@n10) -> (Output_Get:bool)
   n11 = GetBindings(execute=@n9.else, Sequence=Get Selected_LS())
   > n12 = GetAnimSequenceLinkFromLevelSequence(InLevelSequence=Get Selected_LS())
   > n13 = For Each Loop(Array:Array of Movie Scene Binding Proxy Structures=@n11) -> (Array Element:Movie Scene Binding Proxy Structure)
   > Set Anim_Movie_Binding(Anim_Movie_Binding=@n13)
   > n14 = For Each Loop!(Exec, Array:Array of Level Sequence Anim Sequence Link Item Structures=Get AnimSequenceLinks(self=@n12)) -> (
    LoopBody
    Array Element_SkelTrackGuid:Guid Structure
    Array Element_PathToAnimSequence:Soft Object Path Structure
    Array Element_bAutoBake:bool=~
    Array Element_bExportTransforms:bool=~
    Array Element_bExportMorphTargets:bool=~
    Array Element_bExportAttributeCurves:bool=~
    Array Element_bExportMaterialCurves:bool=~
    Array Element_bSkipCurvesWithZeroValue:bool=~
    Array Element_bSetRetargetSourceAsset:bool=~
    Array Element_Interpolation:EAnimInterpolationType Enum
    Array Element_CurveInterpolation:ERichCurveInterpMode Enum
    Array Element_bRecordInWorldSpace:bool=~
    Array Element_bEvaluateAllSkeletalMeshComponents:bool=~
    Array Element_IncludeAnimationNames:Array of Strings
    Array Element_ExcludeAnimationNames:Array of Strings
    Array Element_bRemoveExcludedCurves:bool=~
    Array Element_WarmUpFrames:Frame Number Structure
    Array Element_DelayBeforeStart:Frame Number Structure
    Array Element_bUseCustomTimeRange:bool=~
    Array Element_CustomStartFrame:Frame Number Structure
    Array Element_CustomEndFrame:Frame Number Structure
    Array Element_CustomDisplayRate:Frame Rate Structure
    Array Element_bUseCustomFrameRate:bool=~
    Array Element_CustomFrameRate:Frame Rate Structure
    Array Index
    Completed
   )
   > n15 = FindBindingById(Sequence=Get Selected_LS(), BindingId=@n14)
   n16 = Get Anim_Movie_Binding()
   > Branch(Condition="Equal (Guid)"(A=@n15, B=Break Movie Scene Binding Proxy(MovieSceneBindingProxy=@n16)) -> (ReturnValue=~))
   > Set Start Frame(Start Frame:int=FrameNumber to Integer!(InFrameNumber:"Frame Number Structure (by ref)"=@n14.Array Element_CustomStartFrame) -> (ReturnValue:int=0)) -> (Output_Get:int=0)
   > Set End Frame(End Frame:int=FrameNumber to Integer(InFrameNumber=@n14.Array Element_CustomEndFrame) -> (ReturnValue=0)) -> (Output_Get:int=0)
   > Set CustomeTimeFrame(CustomeTimeFrame:bool=@n14.Array Element_bUseCustomTimeRange) -> (Output_Get:bool)
   > n17 = LoadAsset(AssetPath=BreakSoftObjectPath(InSoftObjectPath=@n14.Array Element_PathToAnimSequence))
   > n18 = Cast To AnimSequence(Object=@n17)
   > n19 = Set Anim Sequence(Anim Sequence:Anim Sequence Object Reference=@n18) -> (Output_Get:Anim Sequence Object Reference)
   > n20 = QueryFileState(InFile=GetSystemPath(Object=@n19), bSilent=false, bUseSourceControlStateCache=false) -> (
    ReturnValue_bIsValid=~
    ReturnValue_bIsUnknown=~
    ReturnValue_bCanCheckIn=~
    ReturnValue_bCanCheckOut=~
    ReturnValue_bIsCheckedOut=~
    ReturnValue_bIsCurrent=~
    ReturnValue_bIsSourceControlled=~
    ReturnValue_bIsAdded=~
    ReturnValue_bIsDeleted=~
    ReturnValue_bIsIgnored=~
    ReturnValue_bCanEdit=~
    ReturnValue_bCanDelete=~
    ReturnValue_bIsModified=~
    ReturnValue_bCanAdd=~
    ReturnValue_bIsConflicted=~
    ReturnValue_bCanRevert=~
    ReturnValue_bIsCheckedOutOther=~
    ReturnValue_bIsCheckedOutInOtherBranch=~
    ReturnValue_bIsModifiedInOtherBranch=~
   )
   > n21 = Branch(Condition=@n20.ReturnValue_bIsCurrent)
   > n22 = Branch(Condition=@n20.ReturnValue_bIsCheckedOutOther)
   > n23 = Checked out by another user warning(
    Checked out by=@n20.ReturnValue_CheckedOutOther
    Dipslay aset name=GetDisplayName(Object=Get Anim Sequence() -> (Anim Sequence:Anim Sequence Object Reference))
   )
   > Set Was the procedure done(Was the procedure done=@n23)
   CheckoutAsset!(execute=@n22.else, AssetToCheckout:String=GetSoftObjectPath!(Object=Get Selected_LS()) -> (ReturnValue_PathString:String)) -> (then, ReturnValue:bool)
   > OpenLevelSequence(LevelSequence=Get Selected_LS())
   > SaveLoadedAssets(AssetsToSave=Make Array("[0]":Object Reference=Get Selected_LS()) -> (Array:Array of Object References), bOnlyIfIsDirty=false)
   > Execute Python Script!(
    execute
    PythonScript="## Get the current level sequence\r\nimport unreal\r\nlevel_sequence = selected_ls\r\n\r\n## Get the asset tools\r\nasset_tools = unreal.AssetToolsHelpers.get_asset_tools()\r\n\r\n## Get the SkeletaMeshActor binding called SK Mannequin\r\nbinding = binding\r\n\r\n## Grab the Level Editor World\r\neditor_subsystem = unreal.get_editor_subsystem(unreal.UnrealEditorSubsystem)\r\nworld = editor_subsystem.get_editor_world()\r\n\r\n## Create animation sequence export options\r\nanim_seq_export_options = unreal.AnimSeqExportOption()\r\nanim_seq_export_options.export_transforms = True\r\nanim_seq_export_options.export_morph_targets = True\r\n\r\nanim_seq_export_options.use_custom_time_range = custom_time\r\n\r\nif custom_time == True:\r\n\tanim_seq_export_options.custom_start_frame = unreal.FrameNumber(start_frame)\r\n\tanim_seq_export_options.custom_end_frame = unreal.FrameNumber(end_frame)\r\n\r\n\r\n## Bake to the created AnimSequence\r\nunreal.SequencerTools.export_anim_sequence(world, level_sequence, anim, anim_seq_export_options, binding, False)"
    Selected_LS:Level Sequence Object Reference=Get Selected_LS()
    binding:Movie Scene Binding Proxy Structure=Get Anim_Movie_Binding()
    anim:Anim Sequence Object Reference=Get Anim Sequence()
    start_frame:int=Get Start Frame() -> (Start Frame:int=0)
    end_frame:int=Get End Frame() -> (End Frame:int=0)
    custom_time:bool=Get CustomeTimeFrame() -> (CustomeTimeFrame:bool)
   ) -> (then, ReturnValue) {inputs [Selected_LS,binding,anim,start_frame,end_frame,custom_time]}
   > Apply All Animimation Modifiers!(execute, AnimSequence:Anim Sequence Object Reference=Get Anim Sequence()) -> (then)
   > SaveLoadedAssets(AssetsToSave=Make Array!("[0]"=Get Anim Sequence(), "[1]":Object Reference=Get Selected_LS()) -> (Array), bOnlyIfIsDirty=false)
   > CloseLevelSequence!(execute) -> (then)
   > n24 = Execute Python Script!(
    execute
    PythonScript="import unreal\r\n\r\nLS = selected_ls\r\n\r\nunreal.log(\"\\n------------------------------------------------------\")\r\nunreal.log(\"{} Anim Sequence has been updated\".format(LS))\r\n\r\nunreal.log(\"\\n-------------------------------------------------------\")"
    selected_ls:String=GetDisplayName(Object=Get Anim Sequence())
   ) -> (then, ReturnValue) {inputs [selected_ls]}
   > Set Was the procedure done(Was the procedure done=@n24)
   > SaveLoadedAsset(AssetToSave=Get Selected_LS(), bOnlyIfIsDirty=true)
   n25 = Not Current Warning!(execute=@n21.else, self, Type of Asset:String=Animation Sequence, Display Asset Name:String=GetDisplayName(Object=Get Anim Sequence())) -> (then, Could it be done:bool) {is=title K2Node_CallFunction}
   > Set Was the procedure done(Was the procedure done=@n25)
   n26 = Not Current Warning(execute=@n8.else, Type of Asset=Level Sequence, Display Asset Name=GetDisplayName(Object=Get Selected_LS()))
   > Set Was the procedure done(Was the procedure done=@n26)
  ]}
 OpenAnimSeq {nodes [
   OpenAnimSeq() {is=title K2Node_FunctionEntry}
   > n1 = Cast To Object_LevelSequenceBrowser(Object=Get Selected Item(self=Get ListView()))
   n2 = To Soft Object Path(SoftObjectReference=Get(self=@n1))
   > n3 = LoadAsset_Blocking(Asset=To Soft Object Reference(SoftObjectPath=MakeSoftClassPath(PathString=@n2)))
   > n4 = Cast To LevelSequence(Object=@n3)
   > n5 = GetBindings(Sequence=@n4)
   > n6 = GetAnimSequenceLinkFromLevelSequence(InLevelSequence=@n4)
   > n7 = For Each Loop!(Exec, Array:Array of Movie Scene Binding Proxy Structures=@n5) -> (LoopBody, Array Element:Movie Scene Binding Proxy Structure, Array Index, Completed)
   > n8 = For Each Loop!(Exec, Array:Array of Level Sequence Anim Sequence Link Item Structures=Get AnimSequenceLinks(self=@n6)) -> (
    LoopBody
    Array Element_SkelTrackGuid:Guid Structure
    Array Element_PathToAnimSequence:Soft Object Path Structure
    Array Element_bAutoBake:bool=~
    Array Element_bExportTransforms:bool=~
    Array Element_bExportMorphTargets:bool=~
    Array Element_bExportAttributeCurves:bool=~
    Array Element_bExportMaterialCurves:bool=~
    Array Element_bSkipCurvesWithZeroValue:bool=~
    Array Element_bSetRetargetSourceAsset:bool=~
    Array Element_Interpolation:EAnimInterpolationType Enum
    Array Element_CurveInterpolation:ERichCurveInterpMode Enum
    Array Element_bRecordInWorldSpace:bool=~
    Array Element_bEvaluateAllSkeletalMeshComponents:bool=~
    Array Element_IncludeAnimationNames:Array of Strings
    Array Element_ExcludeAnimationNames:Array of Strings
    Array Element_bRemoveExcludedCurves:bool=~
    Array Element_WarmUpFrames:Frame Number Structure
    Array Element_DelayBeforeStart:Frame Number Structure
    Array Element_bUseCustomTimeRange:bool=~
    Array Element_CustomStartFrame:Frame Number Structure
    Array Element_CustomEndFrame:Frame Number Structure
    Array Element_CustomDisplayRate:Frame Rate Structure
    Array Element_bUseCustomFrameRate:bool=~
    Array Element_CustomFrameRate:Frame Rate Structure
    Array Index
    Completed
   )
   > n9 = FindBindingById(Sequence=@n4, BindingId=@n8)
   > Branch(Condition="Equal (Guid)"(A=@n9, B=Break Movie Scene Binding Proxy(MovieSceneBindingProxy=@n7)) -> (ReturnValue=~))
   > n10 = LoadAsset(AssetPath=BreakSoftObjectPath(InSoftObjectPath=@n8.Array Element_PathToAnimSequence))
   > n11 = Cast To AnimSequence(Object=@n10)
   > OpenEditorForAssets(self=Get AssetEditorSubsystem(), Assets=Make Array!("[0]"=@n11) -> (Array), OpenedMethod=Edit)
  ]}
 Populate ComboBox {nodes [
   Populate ComboBox!() -> (then, Combo Box:"Searchable Combo Box (String) Object Reference", items:Array of Strings, Default Item:String) {is=title K2Node_FunctionEntry}
   > ClearOptions(self:"Searchable Combo Box (String) Object Reference"=Get Combo Box() -> (Combo Box:"Searchable Combo Box (String) Object Reference"))
   > n1 = Sequence!(execute) -> (then_0, then_1, then_2:Exec)
   > n2 = Branch(Condition=IsEmpty(InString=Get Default Item() -> (Default Item:String)))
   AddOption!(execute=@n2.else, self:"Searchable Combo Box (String) Object Reference"=Get Combo Box(), Option:String=Get Default Item()) -> (then)
   n3 = For Each Loop!(Exec=@n1.then_1, Array:Array of Strings=Get items() -> (items:Array of Strings)) -> (LoopBody, Array Element:String, Array Index, Completed)
   Set Selected(
    execute=@n3.Completed
    Selected:String=GetOptionAtIndex!(self:"Searchable Combo Box (String) Object Reference"=Get Combo Box(), Index:int=0) -> (ReturnValue:String)
   ) -> (Output_Get:String)
   > SetSelectedIndex!(execute, self:"Searchable Combo Box (String) Object Reference"=Get Combo Box(), Index:int=0) -> (then)
   AddOption(execute=@n3.LoopBody, self=Get Combo Box(), Option=@n3)
   Return Node!(execute=@n1.then_2, Selected Item:String=Get Selected() -> (Selected:String))
  ]}
 Populate_Asset_CombBox {nodes [
   n1 = Populate_Asset_CombBox!() -> (then, ComboBox:"Searchable Combo Box (String) Object Reference", Items:Array of Strings) {is=title K2Node_FunctionEntry}
   > n2 = Set Current Index(Current Index:int=GetSelectedIndex!(self:"Searchable Combo Box (String) Object Reference"=@n1) -> (ReturnValue:int=0)) -> (Output_Get:int=0)
   > Set Selected_Asset(Selected_Asset=GetOptionAtIndex(self=Get ComboBox() -> (ComboBox:"Searchable Combo Box (String) Object Reference"), Index=@n2))
   > n3 = Sequence!(execute) -> (then_0, then_1, then_2, then_3:Exec)
   > ClearOptions(self=Get Asset_Searchable_ComboBox())
   > Clear(TargetArray=Get Asset Grp() -> (Asset Grp:Array of Strings))
   n4 = For Each Loop(Exec=@n3.then_1, Array=Get Items() -> (Items:Array of Strings))
   > Add Unique(TargetArray=Get Asset Grp(), NewItem=@n4)
   n5 = For Each Loop(Exec=@n3.then_2, Array=Get Asset Grp())
   n6 = Set Selected_Asset(execute=@n5.Completed, Selected_Asset=GetOptionAtIndex(self=Get Asset_Searchable_ComboBox(), Index=0))
   > SetSelectedOption(self=Get Asset_Searchable_ComboBox(), Option=@n6)
   AddOption(execute=@n5.LoopBody, self=Get Asset_Searchable_ComboBox(), Option=@n5)
  ]}
 Refresh Assets {nodes [
   Refresh Assets() {is=title K2Node_FunctionEntry}
   > n1 = Branch(Condition=IsLoadingAssets(self=GetAssetRegistry()))
   Set Animatable Creatures(
    execute=@n1.else
    Animatable Creatures:Array of Animatable Creature Object References=Get Animatable Creatures!(Project:BP Animation Browser Project Object Reference=Get Current Project(), bSearchSubClasses:bool=false, __WorldContext:Object Reference) -> (Animatable Creatures:Array of Animatable Creature Object References) {is=title K2Node_CallFunction}
   ) -> (Output_Get:Array of Animatable Creature Object References)
  ]}
 Save {nodes [
   Save() {is=title K2Node_FunctionEntry}
   > SaveGameToSlot(SaveGameObject=Get Save Game(), SlotName=Get Slot Name(), UserIndex=0)
  ]}
 Save Settings {nodes [
   Save Settings() {is=title K2Node_FunctionEntry}
   > SaveGameToSlot(SaveGameObject=Get Save Game(), SlotName=Get Slot Name(), UserIndex=0)
  ]}
 SaveUIState {nodes [
   SaveUIState() {is=title K2Node_FunctionEntry}
   > n1 = CreateSaveGameObject(SaveGameClass=&SG_LevelSequenceBrowser_C)
   > n2 = Set Save Game(Save Game=@n1)
   > Set Asset(Asset=GetSelectedOption(self=Get Asset_Searchable_ComboBox()), self=@n2)
   n3 = Get Save Game()
   > Set Variant(Variant=GetSelectedOption(self=Get Plugin_Searchable_ComboBox()), self=@n3)
   > Set Subfolder(Subfolder=GetSelectedOption(self=Get AnimationType_CombBox()), self=@n3)
   n4 = Get Save Game()
   > Set Project!(
    execute
    Project:BP Animation Browser Project Soft Object Reference=Make Soft Reference!(Input:BP Animation Browser Project Object Reference=Get Current Project()) -> (Output:BP Animation Browser Project Soft Object Reference)
    self:SG Level Sequence Browser Object Reference=@n4
   ) -> (then, Output_Get:BP Animation Browser Project Soft Object Reference) {is=title K2Node_VariableSet}
   > SaveGameToSlot(SaveGameObject=@n4, SlotName=Get Slot Name(), UserIndex=0)
  ]}
 Set Project {nodes [
   n1 = Set Project!() -> (then, Project:BP Animation Browser Project Object Reference, Save Settings:bool=true) {is=title K2Node_FunctionEntry}
   > Set Current Project(Current Project=@n1)
   > Refresh Assets() {is=title K2Node_CallFunction}
   > Update Combo Box(
    Combo Box=Get Asset_Searchable_ComboBox()
    Items=AnimatableCreature Get Categories(AnimatableCreatures=Get Animatable Creatures() {is=title K2Node_VariableGet})
    Add Blank=false
   )
   > n2 = Sequence!(execute) -> (then_0, then_1, then_2)
   > Updating Image()
   n3 = Get Asset_Searchable_ComboBox()
   > Branch(Condition="integer > integer"(A=GetOptionCount!(self:"Searchable Combo Box (String) Object Reference"=@n3) -> (ReturnValue:int=0)))
   n4 = Get Asset_Searchable_ComboBox()
   > SetSelectedIndex(self=@n4, Index=0)
   > Set Selected_Asset(Selected_Asset=GetSelectedOption(self=@n4))
   Update Combo Box(
    execute=@n2.then_1
    Combo Box=Get AnimationType_CombBox()
    Items=Get Subfolders()
    Default Item=Get Animation Type Default Item()
    Add Blank=false
   )
   Branch(execute=@n2.then_2, Condition=Get Save Settings() -> (Save Settings:bool))
   n5 = Get Save Game()
   > Set Asset(Asset=GetSelectedOption(self=Get Asset_Searchable_ComboBox()), self=@n5)
   > Set Variant(Variant=GetSelectedOption(self=Get Plugin_Searchable_ComboBox()), self=@n5)
   > Set Subfolder(Subfolder=GetSelectedOption(self=Get AnimationType_CombBox()), self=@n5)
   > Set Project(Project=Make Soft Reference(Input=Get Current Project()), self=Get Save Game()) {is=title K2Node_VariableSet}
   > Save Settings() {is=title K2Node_CallFunction}
  ]}
 SetProject_ClearChildren_Function {nodes [
   n1 = SetProject_ClearChildren_Function!() -> (then, Project:BP Animation Browser Project Object Reference) {is=title K2Node_FunctionEntry}
   > Set Project(Project=@n1, Save Settings=true) {is=title K2Node_CallFunction}
   > ClearChildren(self=Get ModalLayer())
  ]}
 Settings Menu Browse to AC {nodes [
   Settings Menu Browse to AC!() -> (then, Context Item:Editor Utility Widget Object Reference, Slot Widget:Editor Utility Widget Object Reference) {is=title K2Node_FunctionEntry}
   n1 = AnimatableCreature Find(AnimatableCreatures=Get Animatable Creatures(), Asset=Get Selected_Asset(), Variant=Get Selected Plugin())
   > Is Valid(InputObject=@n1)
   > SyncBrowserToObjects(AssetPaths=Make Array("[0]":String=GetSoftObjectPath(Object=@n1)) -> (Array:Array of Strings))
  ]}
 Settings Menu Refresh {nodes [
   Settings Menu Refresh!() -> (then, Context Item:Editor Utility Widget Object Reference, Slot Widget:Editor Utility Widget Object Reference) {is=title K2Node_FunctionEntry}
   > Refresh Assets()
   > Update List View(Use Cache=false)
  ]}
 Success Menu {nodes [
   Success Menu() {is=title K2Node_FunctionEntry}
   > Show Message Dialog(
    Title=Get Popup_Title()
    Message="To Text (String)"(InString=The selected Level Sequences and their Animations have been updated)
    MessageType=Ok
    DefaultValue=No
    MessageCategory=Success
   )
  ]}
 Try Setup {nodes [
   Try Setup() {is=title K2Node_FunctionEntry}
   > n1 = Branch(Condition=IsLoadingAssets(self=GetAssetRegistry()))
   > Return Node!(execute)
   Clear and Invalidate Timer by Handle!(execute=@n1.else, WorldContextObject:Object Reference, Handle=Get SetupTimerHandle() -> (SetupTimerHandle:Timer Handle Structure)) -> (then)
   > Load From Save() {is=title K2Node_CallFunction}
  ]}
 Update Combo Box {nodes [
   Update Combo Box!() -> (then, Combo Box:"Searchable Combo Box (String) Object Reference", Items:Array of Strings, Default Item:String, Add Blank:bool) {is=title K2Node_FunctionEntry}
   n1 = Get Combo Box()
   > n2 = For Loop(LastIndex=int - int(A=GetOptionCount(self=@n1) -> (ReturnValue=0)))
   n3 = Branch(
    execute=@n2.Completed
    Condition=Identical!(ArrayA:Array of Strings=Get Items(), ArrayB:Array of Strings=Get Old Items() -> (Old Items:Array of Strings)) -> (ReturnValue:bool) {is=title K2Node_CallArrayFunction}
   )
   ClearOptions(execute=@n3.else, self=Get Combo Box())
   > n4 = Sequence!(execute) -> (then_0, then_1, then_2, then_3:Exec)
   > Branch(Condition=Get Add Blank() -> (Add Blank:bool))
   > AddOption(self=Get Combo Box())
   n5 = For Each Loop(Exec=@n4.then_1, Array=Get Items())
   > AddOption(self=Get Combo Box(), Option=@n5)
   AddOption(execute=@n4.then_2, self=Get Combo Box(), Option=Get Default Item())
   n6 = Branch(execute=@n4.then_3, Condition=IsEmpty(InString=Get Default Item()))
   > ClearSelection(self:"Searchable Combo Box (String) Object Reference"=Get Combo Box())
   > SetSelectedOption(self=Get Combo Box(), Option="Get (a ref)"(Array:Array of Strings=Get Items(), Dimension 1=0) -> (Output:"String (by ref)"))
   SetSelectedOption(execute=@n6.else, self=Get Combo Box(), Option=Get Default Item())
   n7 = GetOptionAtIndex(self=Get Combo Box(), Index=@n2)
   n8 = Branch(execute=@n2.LoopBody, Condition="Equal Exactly (String)"(A=@n7, B=Get Default Item()))
   Add(execute=@n8.else, TargetArray:Array of Strings=Get Old Items(), NewItem:"String (by ref)"=@n7)
  ]}
 Update Level Sequence Array {nodes [
   Update Level Sequence Array!() -> (then, Directories:Array of Names) {is=title K2Node_FunctionEntry}
   > n1 = GetAssetsByPaths!(
    execute
    self:Asset Registry Interface=GetAssetRegistry()
    PackagePaths:Array of Names=Get Directories() -> (Directories:Array of Names)
    bRecursive:bool=true
    bIncludeOnlyOnDiskAssets:bool=false
   ) -> (then, OutAssetData:Array of Asset Data Structures, ReturnValue:bool)
   > n2 = For Each Loop(Array:Array of Asset Data Structures=@n1) -> (Array Element:Asset Data Structure)
   Return Node!(execute=@n2.Completed, LS_Name_Array:Array of Strings=Get LS Name Cache() -> (LS Name Cache:Array of Strings))
   > Add(
    TargetArray=Get LS Name Cache()
    NewItem=BreakSoftObjectPath(InSoftObjectPath=ToSoftObjectPath!(InAssetData:"Asset Data Structure (by ref)"=@n2) -> (ReturnValue:Soft Object Path Structure))
   )
  ]}
 Update List View {nodes [
   Update List View!() -> (then, Filter:String, Use Cache:bool=~, Current Plugin:String, Current Subfolder:String, Current Asset:String) {is=title K2Node_FunctionEntry}
   > n1 = Sequence!(execute) -> (then_0, then_1, then_2, then_3, then_4:Exec)
   > ClearListItems(self=Get ListView())
   > Set Selected Plugin(Selected Plugin=Get Current Plugin() -> (Current Plugin:String))
   > "Set Selected Subfolder "("Selected Subfolder ":String=Get Current Subfolder() -> (Current Subfolder:String)) -> (Output_Get:String)
   > Clear(TargetArray=Get SearchSplitString() -> (SearchSplitString:Array of Strings))
   > Set Selected_Asset(Selected_Asset=Get Current Asset() -> (Current Asset:String))
   > Clear(TargetArray:Array of Object Level Sequence Browser Object References=Get LevelSequenceArray() -> (LevelSequenceArray:Array of Object Level Sequence Browser Object References))
   > n2 = Branch(Condition=Get Use Cache() -> (Use Cache:bool))
   Clear(execute=@n2.else, TargetArray:Array of Strings=Get NameOfLS() -> (NameOfLS:Array of Strings))
   n3 = For Each Loop(Exec=@n1.then_1, Array:Array of Strings=ParseIntoArray(SourceString=Get Filter() -> (Filter:String), Delimiter=" ")) -> (Array Element:String)
   > Append Array!(execute, TargetArray=Get SearchSplitString(), SourceArray:Array of Strings=ParseIntoArray(SourceString=@n3, Delimiter=_)) -> (then)
   n4 = Get NameOfLS()
   n5 = Branch(execute=@n1.then_2, Condition=AND Boolean(A=Get Use Cache(), B=Is Not Empty!(TargetArray=@n4) -> (ReturnValue:bool)))
   n6 = For Each Loop(Exec=[@n5.then, @n16.then], Array=Get NameOfLS())
   > Set Current LSLocal(Current LSLocal:String=@n6) -> (Output_Get:String)
   > n7 = Branch(Condition=IsEmpty(InString=Get Filter()))
   n8 = Construct Object from Class(
    execute=[@n7.then, @n11.then]
    Class=&/Object_LevelSequenceBrowser.Object_LevelSequenceBrowser_C
    self=Self-Reference()
   ) -> (ReturnValue:Object Level Sequence Browser Object Reference)
   n9 = Get Current LSLocal() -> (Current LSLocal:String)
   > Set Data!(
    execute
    Data:Struct Level Sequence Browser Item Structure=Make Struct Level Sequence Browser Item!(LS_58_#1:Soft Object Reference=To Soft Object Reference!(SoftObjectPath_PathString:String=@n9) -> (ReturnValue)) -> (Struct_LevelSequenceBrowser_Item:Struct Level Sequence Browser Item Structure) {is=title K2Node_MakeStruct}
    self:Object Level Sequence Browser Object Reference=@n8
   ) -> (then, Output_Get:Struct Level Sequence Browser Item Structure)
   > Add(
    TargetArray:Array of Object Level Sequence Browser Object References=Get LevelSequenceArray()
    NewItem:"Object Level Sequence Browser Object Reference (by ref)"=@n8
   )
   Set SearchStringContains(execute=@n7.else, SearchStringContains:bool=true) -> (Output_Get:bool)
   > n10 = For Each Loop(Array=Get SearchSplitString())
   n11 = Branch(execute=@n10.Completed, Condition=Get SearchStringContains() -> (SearchStringContains:bool))
   n12 = Branch(execute=@n10.LoopBody, Condition=Contains(SearchIn=Get Current LSLocal(), Substring=@n10))
   Set SearchStringContains(execute=@n12.else, SearchStringContains=false)
   n13 = Get Animation Directories for Asset!(
    execute=@n5.else
    self
    Asset Name:String=Get Selected_Asset()
    Plugin:String=Get Selected Plugin()
    Animation Type:String="Get Selected Subfolder "()
   ) -> (then, Animation Directories:Array of Names) {is=title K2Node_CallFunction}
   > n14 = Update Level Sequence Array!(execute, self, Directories:Array of Names=@n13) -> (then, LS_Name_Array:Array of Strings) {is=title K2Node_CallFunction}
   > n15 = Execute Python Script!(
    execute
    PythonScript="import os\r\nsorted_files = sorted(files_to_sort, key=lambda x:os.path.basename(x))"
    files_to_sort:Array of Strings=@n14
   ) -> (then, ReturnValue, sorted_files:Array of Strings) {inputs [files_to_sort], outputs [sorted_files]}
   > n16 = Set NameOfLS(NameOfLS:Array of Strings=@n15.sorted_files) -> (Output_Get:Array of Strings)
   n17 = For Each Loop(Exec=@n1.then_3, Array:Array of Object Level Sequence Browser Object References=Get LevelSequenceArray()) -> (Array Element:Object Level Sequence Browser Object Reference)
   > AddItem!(execute, self:List View Object Reference=Get ListView(), Item:Object Reference=@n17) -> (then)
   Return Node!(execute=@n1.then_4)
  ]}
 Update List View From UI {nodes [
   Update List View From UI() {is=title K2Node_FunctionEntry}
   > Update List View(
    Use Cache=false
    Current Plugin=GetSelectedOption(self=Get Plugin_Searchable_ComboBox())
    Current Subfolder=GetSelectedOption(self=Get AnimationType_CombBox())
    Current Asset=GetSelectedOption(self=Get Asset_Searchable_ComboBox())
   ) {is=title K2Node_CallFunction}
  ]}
 Updating Image {nodes [
   Updating Image() {is=title K2Node_FunctionEntry}
   n1 = Get Current Project()
   > n2 = Branch(Condition=Contains(SearchIn=GetDisplayName(Object=@n1), Substring=Companions, bSearchFromEnd=true))
   > SetBrushFromTexture!(
    execute
    self:Image Object Reference=Get Project_Image() -> (Project_Image:Image Object Reference)
    Texture:Texture 2D Object Reference=Get Companions() -> (Companions:Texture 2D Object Reference)
    bMatchSize:bool=false
   ) -> (then)
   n3 = Get Current Project()
   n4 = Branch(execute=@n2.else, Condition=Contains(SearchIn=GetDisplayName(Object=@n3), Substring=Rhino, bSearchFromEnd=true))
   > SetBrushFromTexture(self=Get Project_Image(), Texture=Get Car() -> (Car:Texture 2D Object Reference), bMatchSize=false)
   n5 = Get Current Project()
   n6 = Branch(execute=@n4.else, Condition=Contains(SearchIn=GetDisplayName(Object=@n5), Substring=EpicControlRig, bSearchFromEnd=true))
   > SetBrushFromTexture(self=Get Project_Image(), Texture=Get Bone() -> (Bone:Texture 2D Object Reference), bMatchSize=false)
   n7 = GetDisplayName(Object=Get Current Project())
   n8 = GetDisplayName(Object=Get Current Project())
   n9 = Branch(
    execute=@n6.else
    Condition=OR Boolean(A=Contains(SearchIn=@n7, Substring=Juno, bSearchFromEnd=true), B=Contains(SearchIn=@n8, Substring=Sprout, bSearchFromEnd=true))
   )
   > SetBrushFromTexture(self=Get Project_Image(), Texture=Get Lego() -> (Lego:Texture 2D Object Reference), bMatchSize=false)
   n10 = Get Current Project()
   n11 = Branch(execute=@n9.else, Condition=Contains(SearchIn=GetDisplayName(Object=@n10), Substring=Feral, bSearchFromEnd=true))
   > SetBrushFromTexture(self=Get Project_Image(), Texture=Get Feral() -> (Feral:Texture 2D Object Reference), bMatchSize=false)
   n12 = Get Current Project()
   n13 = Branch(execute=@n11.else, Condition=Contains(SearchIn=GetDisplayName(Object=@n12), Substring=BR, bSearchFromEnd=true))
   > SetBrushFromTexture(self=Get Project_Image(), Texture=Get Llama() -> (Llama:Texture 2D Object Reference), bMatchSize=false)
   SetBrushFromTexture(execute=@n13.else, self=Get Project_Image(), Texture=Get Default() -> (Default:Texture 2D Object Reference), bMatchSize=false)
  ]}
]
variables [
 Animatable Creatures {type="array<AnimatableCreature_C>", default []}
 Animation Type Default Item {type=string, default=All}
 Bone {type=Texture2D}
 Car {type=Texture2D, default=/Game/Athena/UI/Challenges/Art/ChallengeTypeIcons/UI_ChallengeIcon_Car}
 Companions {type=Texture2D}
 Current Project {type=BP_AnimationBrowser_Project_C, default=null}
 Default {type=Texture2D, default=/FortniteCoreUI/UI/Login/Art/T-EpicGames-Logo}
 ED_AssetHasChanged {default=()}
 ED_ListViewItemSelected {default=()}
 Feral {type=Texture2D, default=/Game/Athena/HUD/Art/BacchusHudIcons/Scope_Normal}
 Lego {type=Texture2D, default=/FortUILibrary/Textures/Icons/T_UI_Icon_UIKit_LegoSproutQuests}
 LevelSeqeunce_ListItem {type=LevelSequenceBrowser_ListItem_C, default=null}
 Llama {type=Texture2D}
 Popup_Title {type=string, default=Level Sequence Batch Update}
 Save Game {type=SG_LevelSequenceBrowser_C, default=null}
 SearchBar_Filter {type=string, default=/CompanionsAnimInEd/Assets/}
 Selected Plugin {type=string, default=""}
 Selected Project Name {type=string, default=""}
 "Selected Subfolder " {type=string, default=""}
 Selected_Asset {type=string, default=""}
 Selected_LS {type=LevelSequence, default=null}
 SetupTimerHandle {type=TimerHandle}
 Slot Name {type=string, default=LevelSequenceBrowser}
]
widgets {
 info {rootWidgetClass=&CanvasPanel, widgetCount=23}
 widgets [
  CanvasPanel_21 {widgetClassPath=&CanvasPanel, widget=CanvasPanel_21}
  VerticalBox_29 {
   widgetClassPath=&VerticalBox
   parent=CanvasPanel_21
   widget=VerticalBox_29
   slot {class=CanvasPanelSlot, properties {layoutData {alignment=<0 0>, anchors {maximum=<1 1>, minimum=<0 0>}, offsets=<ltrb 0 0 0 0>}}}
  }
  HeaderBorder {
   widgetClassPath=&Border
   parent=VerticalBox_29
   widget=HeaderBorder
   properties {brushColor=<rgba 0.0194 0.0194 0.0194 1>, padding=<ltrb 2 2 2 2>}
   slot {class=VerticalBoxSlot, properties {padding=<ltrb 2 5 2 0>}}
  }
  VerticalBox_1 {
   widgetClassPath=&VerticalBox
   parent=HeaderBorder
   widget=VerticalBox_1
   slot {class=BorderSlot, properties {padding=<ltrb 2 2 2 2>}}
  }
  HorizontalBox_49 {widgetClassPath=&HorizontalBox, parent=VerticalBox_1, widget=HorizontalBox_49, slot {class=VerticalBoxSlot}}
  Project_Image {
   widgetClassPath=&Image
   bIsVariable=true
   parent=HorizontalBox_49
   widget=Project_Image
   properties {brush {
     bIsDynamicallyLoaded=false
     drawAs=Image
     imageType=FullColor
     mirroring=NoMirror
     resourceName=None
     resourceObject=&T-EpicGames-Logo
     tiling=NoTile
     imageSize=<28 28>
     margin=<ltrb 0 0 0 0>
     outlineSettings {
      bUseBrushTransparency=false
      roundingType=HalfHeightRadius
      width=0
      color {colorUseRule=UseColor_Specified, specifiedColor=<rgba 0 0 0 0>}
      cornerRadii=<0 0 0 0>
     }
     tintColor {colorUseRule=UseColor_Specified, specifiedColor=<rgba 1 1 1 1>}
     uVRegion {bIsValid=false, max=<0 0>, min=<0 0>}
    }}
   slot {class=HorizontalBoxSlot, properties {verticalAlignment=VAlign_Center, padding=<ltrb 5 0 0 0>}}
  }
  Asset_Searchable_ComboBox {
   widgetClassPath=/Script/FortniteEditor.SearchableStringComboBox
   bIsVariable=true
   parent=HorizontalBox_49
   widget=Asset_Searchable_ComboBox
   properties {
    selectedOption=Asset
    toolTipText=asset drop down
    defaultOptions [Asset]
    foregroundColor {colorUseRule=UseColor_Specified, specifiedColor=<rgba 0.5271 0.5271 0.5271 1>}
    itemStyle {
     bUseParentRowBrush=false
     activeBrush {
      bIsDynamicallyLoaded=false
      drawAs=Image
      imageType=NoImage
      mirroring=NoMirror
      resourceName=None
      resourceObject=null
      tiling=NoTile
      imageSize=<0 0>
      margin=<ltrb 0 0 0 0>
      outlineSettings {
       bUseBrushTransparency=false
       roundingType=HalfHeightRadius
       width=0
       color {colorUseRule=UseColor_Specified, specifiedColor=<rgba 0 0 0 0>}
       cornerRadii=<0 0 0 0>
      }
      tintColor {colorUseRule=UseColor_Specified, specifiedColor=<rgba 0 0.162 0.7454 1>}
      uVRegion {bIsValid=false, max=<0 0>, min=<0 0>}
     }
     activeHighlightedBrush {
      bIsDynamicallyLoaded=false
      drawAs=Image
      imageType=NoImage
      mirroring=NoMirror
      resourceName=None
      resourceObject=null
      tiling=NoTile
      imageSize=<0 0>
      margin=<ltrb 0 0 0 0>
      outlineSettings {
       bUseBrushTransparency=false
       roundingType=HalfHeightRadius
       width=0
       color {colorUseRule=UseColor_Specified, specifiedColor=<rgba 0 0 0 0>}
       cornerRadii=<0 0 0 0>
      }
      tintColor {colorUseRule=UseColor_Specified, specifiedColor=<rgba 0.0252 0.0319 0.0423 1>}
      uVRegion {bIsValid=false, max=<0 0>, min=<0 0>}
     }
     activeHoveredBrush {
      bIsDynamicallyLoaded=false
      drawAs=Image
      imageType=NoImage
      mirroring=NoMirror
      resourceName=None
      resourceObject=null
      tiling=NoTile
      imageSize=<0 0>
      margin=<ltrb 0 0 0 0>
      outlineSettings {
       bUseBrushTransparency=false
       roundingType=HalfHeightRadius
       width=0
       color {colorUseRule=UseColor_Specified, specifiedColor=<rgba 0 0 0 0>}
       cornerRadii=<0 0 0 0>
      }
      tintColor {colorUseRule=UseColor_Specified, specifiedColor=<rgba 0 0.162 0.7454 1>}
      uVRegion {bIsValid=false, max=<0 0>, min=<0 0>}
     }
     dropIndicator_Above {
      bIsDynamicallyLoaded=false
      drawAs=Box
      imageType=FullColor
      mirroring=NoMirror
      resourceName=../../../Engine/Content/Slate/Common/DropZoneIndicator_Above.png
      resourceObject=null
      tiling=NoTile
      imageSize=<0 0>
      margin=<ltrb 0.625 0.625 0 0>
      outlineSettings {
       bUseBrushTransparency=false
       roundingType=HalfHeightRadius
       width=0
       color {colorUseRule=UseColor_Specified, specifiedColor=<rgba 0 0 0 0>}
       cornerRadii=<0 0 0 0>
      }
      tintColor {colorUseRule=UseColor_Specified, specifiedColor=<rgba 0.0091 0.3515 0.9647 1>}
      uVRegion {bIsValid=false, max=<0 0>, min=<0 0>}
     }
     dropIndicator_Below {
      bIsDynamicallyLoaded=false
      drawAs=Box
      imageType=FullColor
      mirroring=NoMirror
      resourceName=../../../Engine/Content/Slate/Common/DropZoneIndicator_Below.png
      resourceObject=null
      tiling=NoTile
      imageSize=<0 0>
      margin=<ltrb 0.625 0 0 0.625>
      outlineSettings {
       bUseBrushTransparency=false
       roundingType=HalfHeightRadius
       width=0
       color {colorUseRule=UseColor_Specified, specifiedColor=<rgba 0 0 0 0>}
       cornerRadii=<0 0 0 0>
      }
      tintColor {colorUseRule=UseColor_Specified, specifiedColor=<rgba 0.0091 0.3515 0.9647 1>}
      uVRegion {bIsValid=false, max=<0 0>, min=<0 0>}
     }
     dropIndicator_Onto {
      bIsDynamicallyLoaded=false
      drawAs=Box
      imageType=FullColor
      mirroring=NoMirror
      resourceName=../../../Engine/Content/Slate/Common/DropZoneIndicator_Onto.png
      resourceObject=null
      tiling=NoTile
      imageSize=<0 0>
      margin=<ltrb 0.25 0.25 0.25 0.25>
      outlineSettings {
       bUseBrushTransparency=false
       roundingType=HalfHeightRadius
       width=0
       color {colorUseRule=UseColor_Specified, specifiedColor=<rgba 0 0 0 0>}
       cornerRadii=<0 0 0 0>
      }
      tintColor {colorUseRule=UseColor_Specified, specifiedColor=<rgba 0.0091 0.3515 0.9647 1>}
      uVRegion {bIsValid=false, max=<0 0>, min=<0 0>}
     }
     evenRowBackgroundBrush {
      bIsDynamicallyLoaded=false
      drawAs=Image
      imageType=NoImage
      mirroring=NoMirror
      resourceName=None
      resourceObject=null
      tiling=NoTile
      imageSize=<0 0>
      margin=<ltrb 0 0 0 0>
      outlineSettings {
       bUseBrushTransparency=false
       roundingType=HalfHeightRadius
       width=0
       color {colorUseRule=UseColor_Specified, specifiedColor=<rgba 0 0 0 0>}
       cornerRadii=<0 0 0 0>
      }
      tintColor {colorUseRule=UseColor_Specified, specifiedColor=<rgba 0.0103 0.0103 0.0103 1>}
      uVRegion {bIsValid=false, max=<0 0>, min=<0 0>}
     }
     evenRowBackgroundHoveredBrush {
      bIsDynamicallyLoaded=false
      drawAs=Image
      imageType=NoImage
      mirroring=NoMirror
      resourceName=None
      resourceObject=null
      tiling=NoTile
      imageSize=<0 0>
      margin=<ltrb 0 0 0 0>
      outlineSettings {
       bUseBrushTransparency=false
       roundingType=HalfHeightRadius
       width=0
       color {colorUseRule=UseColor_Specified, specifiedColor=<rgba 0 0 0 0>}
       cornerRadii=<0 0 0 0>
      }
      tintColor {colorUseRule=UseColor_Specified, specifiedColor=<rgba 0.0176 0.0176 0.0176 1>}
      uVRegion {bIsValid=false, max=<0 0>, min=<0 0>}
     }
     inactiveBrush {
      bIsDynamicallyLoaded=false
      drawAs=Image
      imageType=NoImage
      mirroring=NoMirror
      resourceName=None
      resourceObject=null
      tiling=NoTile
      imageSize=<0 0>
      margin=<ltrb 0 0 0 0>
      outlineSettings {
       bUseBrushTransparency=false
       roundingType=HalfHeightRadius
       width=0
       color {colorUseRule=UseColor_Specified, specifiedColor=<rgba 0 0 0 0>}
       cornerRadii=<0 0 0 0>
      }
      tintColor {colorUseRule=UseColor_Specified, specifiedColor=<rgba 0.0513 0.0953 0.159 1>}
      uVRegion {bIsValid=false, max=<0 0>, min=<0 0>}
     }
     inactiveHighlightedBrush {
      bIsDynamicallyLoaded=false
      drawAs=Image
      imageType=NoImage
      mirroring=NoMirror
      resourceName=None
      resourceObject=null
      tiling=NoTile
      imageSize=<0 0>
      margin=<ltrb 0 0 0 0>
      outlineSettings {
       bUseBrushTransparency=false
       roundingType=HalfHeightRadius
       width=0
       color {colorUseRule=UseColor_Specified, specifiedColor=<rgba 0 0 0 0>}
       cornerRadii=<0 0 0 0>
      }
      tintColor {colorUseRule=UseColor_Specified, specifiedColor=<rgba 0.0252 0.0319 0.0423 1>}
      uVRegion {bIsValid=false, max=<0 0>, min=<0 0>}
     }
     inactiveHoveredBrush {
      bIsDynamicallyLoaded=false
      drawAs=Image
      imageType=NoImage
      mirroring=NoMirror
      resourceName=None
      resourceObject=null
      tiling=NoTile
      imageSize=<0 0>
      margin=<ltrb 0 0 0 0>
      outlineSettings {
       bUseBrushTransparency=false
       roundingType=HalfHeightRadius
       width=0
       color {colorUseRule=UseColor_Specified, specifiedColor=<rgba 0 0 0 0>}
       cornerRadii=<0 0 0 0>
      }
      tintColor {colorUseRule=UseColor_Specified, specifiedColor=<rgba 0.0513 0.0953 0.159 1>}
      uVRegion {bIsValid=false, max=<0 0>, min=<0 0>}
     }
     oddRowBackgroundBrush {
      bIsDynamicallyLoaded=false
      drawAs=Image
      imageType=NoImage
      mirroring=NoMirror
      resourceName=None
      resourceObject=null
      tiling=NoTile
      imageSize=<0 0>
      margin=<ltrb 0 0 0 0>
      outlineSettings {
       bUseBrushTransparency=false
       roundingType=HalfHeightRadius
       width=0
       color {colorUseRule=UseColor_Specified, specifiedColor=<rgba 0 0 0 0>}
       cornerRadii=<0 0 0 0>
      }
      tintColor {colorUseRule=UseColor_Specified, specifiedColor=<rgba 0.0103 0.0103 0.0103 1>}
      uVRegion {bIsValid=false, max=<0 0>, min=<0 0>}
     }
     oddRowBackgroundHoveredBrush {
      bIsDynamicallyLoaded=false
      drawAs=Image
      imageType=NoImage
      mirroring=NoMirror
      resourceName=None
      resourceObject=null
      tiling=NoTile
      imageSize=<0 0>
      margin=<ltrb 0 0 0 0>
      outlineSettings {
       bUseBrushTransparency=false
       roundingType=HalfHeightRadius
       width=0
       color {colorUseRule=UseColor_Specified, specifiedColor=<rgba 0 0 0 0>}
       cornerRadii=<0 0 0 0>
      }
      tintColor {colorUseRule=UseColor_Specified, specifiedColor=<rgba 0.0176 0.0176 0.0176 1>}
      uVRegion {bIsValid=false, max=<0 0>, min=<0 0>}
     }
     parentRowBackgroundBrush {
      bIsDynamicallyLoaded=false
      drawAs=Image
      imageType=NoImage
      mirroring=NoMirror
      resourceName=None
      resourceObject=null
      tiling=NoTile
      imageSize=<32 32>
      margin=<ltrb 0 0 0 0>
      outlineSettings {
       bUseBrushTransparency=false
       roundingType=HalfHeightRadius
       width=0
       color {colorUseRule=UseColor_Specified, specifiedColor=<rgba 0 0 0 0>}
       cornerRadii=<0 0 0 0>
      }
      tintColor {colorUseRule=UseColor_Specified, specifiedColor=<rgba 1 1 1 1>}
      uVRegion {bIsValid=false, max=<0 0>, min=<0 0>}
     }
     parentRowBackgroundHoveredBrush {
      bIsDynamicallyLoaded=false
      drawAs=Image
      imageType=NoImage
      mirroring=NoMirror
      resourceName=None
      resourceObject=null
      tiling=NoTile
      imageSize=<32 32>
      margin=<ltrb 0 0 0 0>
      outlineSettings {
       bUseBrushTransparency=false
       roundingType=HalfHeightRadius
       width=0
       color {colorUseRule=UseColor_Specified, specifiedColor=<rgba 0 0 0 0>}
       cornerRadii=<0 0 0 0>
      }
      tintColor {colorUseRule=UseColor_Specified, specifiedColor=<rgba 1 1 1 1>}
      uVRegion {bIsValid=false, max=<0 0>, min=<0 0>}
     }
     selectedTextColor {colorUseRule=UseColor_Specified, specifiedColor=<rgba 1 1 1 1>}
     selectorFocusedBrush {
      bIsDynamicallyLoaded=false
      drawAs=Border
      imageType=FullColor
      mirroring=NoMirror
      resourceName=../../../Engine/Content/Slate/Common/Selector.png
      resourceObject=null
      tiling=Both
      imageSize=<0 0>
      margin=<ltrb 0.25 0.25 0.25 0.25>
      outlineSettings {
       bUseBrushTransparency=false
       roundingType=HalfHeightRadius
       width=0
       color {colorUseRule=UseColor_Specified, specifiedColor=<rgba 0 0 0 0>}
       cornerRadii=<0 0 0 0>
      }
      tintColor {colorUseRule=UseColor_Specified, specifiedColor=<rgba 0.701 0.225 0.003 1>}
      uVRegion {bIsValid=false, max=<0 0>, min=<0 0>}
     }
     textColor {colorUseRule=UseColor_Specified, specifiedColor=<rgba 0.5271 0.5271 0.5271 1>}
    }
    widgetStyle {
     comboButtonStyle {
      downArrowAlign=VAlign_Center
      buttonStyle {
       clickedSlateSound {resourceObject=null}
       disabled {
        bIsDynamicallyLoaded=false
        drawAs=RoundedBox
        imageType=NoImage
        mirroring=NoMirror
        resourceName=None
        resourceObject=null
        tiling=NoTile
        imageSize=<0 0>
        margin=<ltrb 0 0 0 0>
        outlineSettings {
         bUseBrushTransparency=false
         roundingType=FixedRadius
         width=1
         color {colorUseRule=UseColor_ColorTable, specifiedColor=<rgba 0 0 0 0>}
         cornerRadii=<4 4 4 4>
        }
        tintColor {colorUseRule=UseColor_Specified, specifiedColor=<rgba 0.0395 0.0395 0.0395 1>}
        uVRegion {bIsValid=false, max=<0 0>, min=<0 0>}
       }
       disabledForeground {colorUseRule=UseColor_ColorTable, specifiedColor=<rgba 0 0 0 0>}
       hovered {
        bIsDynamicallyLoaded=false
        drawAs=RoundedBox
        imageType=NoImage
        mirroring=NoMirror
        resourceName=None
        resourceObject=null
        tiling=NoTile
        imageSize=<0 0>
        margin=<ltrb 0 0 0 0>
        outlineSettings {
         bUseBrushTransparency=false
         roundingType=FixedRadius
         width=1
         color {colorUseRule=UseColor_ColorTable, specifiedColor=<rgba 0 0 0 0>}
         cornerRadii=<4 4 4 4>
        }
        tintColor {colorUseRule=UseColor_Specified, specifiedColor=<rgba 0.0048 0.0048 0.0048 1>}
        uVRegion {bIsValid=false, max=<0 0>, min=<0 0>}
       }
       hoveredForeground {colorUseRule=UseColor_ColorTable, specifiedColor=<rgba 0 0 0 0>}
       hoveredSlateSound {resourceObject=null}
       normal {
        bIsDynamicallyLoaded=false
        drawAs=RoundedBox
        imageType=NoImage
        mirroring=NoMirror
        resourceName=None
        resourceObject=null
        tiling=NoTile
        imageSize=<0 0>
        margin=<ltrb 0 0 0 0>
        outlineSettings {
         bUseBrushTransparency=false
         roundingType=FixedRadius
         width=1
         color {colorUseRule=UseColor_ColorTable, specifiedColor=<rgba 0 0 0 0>}
         cornerRadii=<4 4 4 4>
        }
        tintColor {colorUseRule=UseColor_Specified, specifiedColor=<rgba 0.0048 0.0048 0.0048 1>}
        uVRegion {bIsValid=false, max=<0 0>, min=<0 0>}
       }
       normalForeground {colorUseRule=UseColor_ColorTable, specifiedColor=<rgba 0 0 0 0>}
       normalPadding=<ltrb 8 1 8 1>
       pressed {
        bIsDynamicallyLoaded=false
        drawAs=RoundedBox
        imageType=NoImage
        mirroring=NoMirror
        resourceName=None
        resourceObject=null
        tiling=NoTile
        imageSize=<0 0>
        margin=<ltrb 0 0 0 0>
        outlineSettings {
         bUseBrushTransparency=false
         roundingType=FixedRadius
         width=1
         color {colorUseRule=UseColor_ColorTable, specifiedColor=<rgba 0 0 0 0>}
         cornerRadii=<4 4 4 4>
        }
        tintColor {colorUseRule=UseColor_Specified, specifiedColor=<rgba 0.0048 0.0048 0.0048 1>}
        uVRegion {bIsValid=false, max=<0 0>, min=<0 0>}
       }
       pressedForeground {colorUseRule=UseColor_ColorTable, specifiedColor=<rgba 0 0 0 0>}
       pressedPadding=<ltrb 8 2 8 0>
       pressedSlateSound {resourceObject=null}
      }
      contentPadding=<ltrb 0 0 0 0>
      downArrowImage {
       bIsDynamicallyLoaded=false
       drawAs=Image
       imageType=Vector
       mirroring=NoMirror
       resourceName=../../../Engine/Content/Slate/Starship/CoreWidgets/ComboBox/wide-chevron-down.svg
       resourceObject=null
       tiling=NoTile
       imageSize=<16 16>
       margin=<ltrb 0 0 0 0>
       outlineSettings {
        bUseBrushTransparency=false
        roundingType=HalfHeightRadius
        width=0
        color {colorUseRule=UseColor_Specified, specifiedColor=<rgba 0 0 0 0>}
        cornerRadii=<0 0 0 0>
       }
       tintColor {colorUseRule=UseColor_Specified, specifiedColor=<rgba 1 1 1 1>}
       uVRegion {bIsValid=false, max=<0 0>, min=<0 0>}
      }
      downArrowPadding=<ltrb 2 2 2 2>
      menuBorderBrush {
       bIsDynamicallyLoaded=false
       drawAs=RoundedBox
       imageType=NoImage
       mirroring=NoMirror
       resourceName=None
       resourceObject=null
       tiling=NoTile
       imageSize=<0 0>
       margin=<ltrb 0 0 0 0>
       outlineSettings {
        bUseBrushTransparency=false
        roundingType=FixedRadius
        width=1
        color {colorUseRule=UseColor_ColorTable, specifiedColor=<rgba 0 0 0 0>}
        cornerRadii=<0 0 0 0>
       }
       tintColor {colorUseRule=UseColor_Specified, specifiedColor=<rgba 0.0395 0.0395 0.0395 1>}
       uVRegion {bIsValid=false, max=<0 0>, min=<0 0>}
      }
      menuBorderPadding=<ltrb 0 0 0 0>
      shadowColorAndOpacity=<rgba 0 0 0 1>
      shadowOffset=<0 0>
     }
     contentPadding=<ltrb 0 0 0 0>
     menuRowPadding=<ltrb 8 3 8 3>
     pressedSlateSound {resourceObject=null}
     selectionChangeSlateSound {resourceObject=null}
    }
   }
   slot {class=HorizontalBoxSlot, properties {padding=<ltrb 5 0 5 0>, size {sizeRule=Fill, value=1}}}
  }
  PluginImage {
   widgetClassPath=&Image
   bIsVariable=true
   parent=HorizontalBox_49
   widget=PluginImage
   properties {brush {
     bIsDynamicallyLoaded=false
     drawAs=Image
     imageType=FullColor
     mirroring=NoMirror
     resourceName=None
     resourceObject=/AnimInEdTools/AnimationBrowser/Icons/Icon_Plugin
     tiling=NoTile
     imageSize=<18 28>
     margin=<ltrb 0 0 0 0>
     outlineSettings {
      bUseBrushTransparency=false
      roundingType=HalfHeightRadius
      width=0
      color {colorUseRule=UseColor_Specified, specifiedColor=<rgba 0 0 0 0>}
      cornerRadii=<0 0 0 0>
     }
     tintColor {colorUseRule=UseColor_Specified, specifiedColor=<rgba 1 1 1 1>}
     uVRegion {bIsValid=false, max=<0 0>, min=<0 0>}
    }}
   slot {class=HorizontalBoxSlot, properties {padding=<ltrb 0 0 5 0>}}
  }
  Plugin_Searchable_ComboBox {
   widgetClassPath=&SearchableStringComboBox
   bIsVariable=true
   parent=HorizontalBox_49
   widget=Plugin_Searchable_ComboBox
   properties {
    selectedOption=Plugin
    toolTipText=plugin type dropdown
    defaultOptions [Plugin]
    foregroundColor {colorUseRule=UseColor_Specified, specifiedColor=<rgba 0.5271 0.5271 0.5271 1>}
    itemStyle {
     bUseParentRowBrush=false
     activeBrush {
      bIsDynamicallyLoaded=false
      drawAs=Image
      imageType=NoImage
      mirroring=NoMirror
      resourceName=None
      resourceObject=null
      tiling=NoTile
      imageSize=<0 0>
      margin=<ltrb 0 0 0 0>
      outlineSettings {
       bUseBrushTransparency=false
       roundingType=HalfHeightRadius
       width=0
       color {colorUseRule=UseColor_Specified, specifiedColor=<rgba 0 0 0 0>}
       cornerRadii=<0 0 0 0>
      }
      tintColor {colorUseRule=UseColor_Specified, specifiedColor=<rgba 0 0.162 0.7454 1>}
      uVRegion {bIsValid=false, max=<0 0>, min=<0 0>}
     }
     activeHighlightedBrush {
      bIsDynamicallyLoaded=false
      drawAs=Image
      imageType=NoImage
      mirroring=NoMirror
      resourceName=None
      resourceObject=null
      tiling=NoTile
      imageSize=<0 0>
      margin=<ltrb 0 0 0 0>
      outlineSettings {
       bUseBrushTransparency=false
       roundingType=HalfHeightRadius
       width=0
       color {colorUseRule=UseColor_Specified, specifiedColor=<rgba 0 0 0 0>}
       cornerRadii=<0 0 0 0>
      }
      tintColor {colorUseRule=UseColor_Specified, specifiedColor=<rgba 0.0252 0.0319 0.0423 1>}
      uVRegion {bIsValid=false, max=<0 0>, min=<0 0>}
     }
     activeHoveredBrush {
      bIsDynamicallyLoaded=false
      drawAs=Image
      imageType=NoImage
      mirroring=NoMirror
      resourceName=None
      resourceObject=null
      tiling=NoTile
      imageSize=<0 0>
      margin=<ltrb 0 0 0 0>
      outlineSettings {
       bUseBrushTransparency=false
       roundingType=HalfHeightRadius
       width=0
       color {colorUseRule=UseColor_Specified, specifiedColor=<rgba 0 0 0 0>}
       cornerRadii=<0 0 0 0>
      }
      tintColor {colorUseRule=UseColor_Specified, specifiedColor=<rgba 0 0.162 0.7454 1>}
      uVRegion {bIsValid=false, max=<0 0>, min=<0 0>}
     }
     dropIndicator_Above {
      bIsDynamicallyLoaded=false
      drawAs=Box
      imageType=FullColor
      mirroring=NoMirror
      resourceName=../../../Engine/Content/Slate/Common/DropZoneIndicator_Above.png
      resourceObject=null
      tiling=NoTile
      imageSize=<0 0>
      margin=<ltrb 0.625 0.625 0 0>
      outlineSettings {
       bUseBrushTransparency=false
       roundingType=HalfHeightRadius
       width=0
       color {colorUseRule=UseColor_Specified, specifiedColor=<rgba 0 0 0 0>}
       cornerRadii=<0 0 0 0>
      }
      tintColor {colorUseRule=UseColor_Specified, specifiedColor=<rgba 0.0091 0.3515 0.9647 1>}
      uVRegion {bIsValid=false, max=<0 0>, min=<0 0>}
     }
     dropIndicator_Below {
      bIsDynamicallyLoaded=false
      drawAs=Box
      imageType=FullColor
      mirroring=NoMirror
      resourceName=../../../Engine/Content/Slate/Common/DropZoneIndicator_Below.png
      resourceObject=null
      tiling=NoTile
      imageSize=<0 0>
      margin=<ltrb 0.625 0 0 0.625>
      outlineSettings {
       bUseBrushTransparency=false
       roundingType=HalfHeightRadius
       width=0
       color {colorUseRule=UseColor_Specified, specifiedColor=<rgba 0 0 0 0>}
       cornerRadii=<0 0 0 0>
      }
      tintColor {colorUseRule=UseColor_Specified, specifiedColor=<rgba 0.0091 0.3515 0.9647 1>}
      uVRegion {bIsValid=false, max=<0 0>, min=<0 0>}
     }
     dropIndicator_Onto {
      bIsDynamicallyLoaded=false
      drawAs=Box
      imageType=FullColor
      mirroring=NoMirror
      resourceName=../../../Engine/Content/Slate/Common/DropZoneIndicator_Onto.png
      resourceObject=null
      tiling=NoTile
      imageSize=<0 0>
      margin=<ltrb 0.25 0.25 0.25 0.25>
      outlineSettings {
       bUseBrushTransparency=false
       roundingType=HalfHeightRadius
       width=0
       color {colorUseRule=UseColor_Specified, specifiedColor=<rgba 0 0 0 0>}
       cornerRadii=<0 0 0 0>
      }
      tintColor {colorUseRule=UseColor_Specified, specifiedColor=<rgba 0.0091 0.3515 0.9647 1>}
      uVRegion {bIsValid=false, max=<0 0>, min=<0 0>}
     }
     evenRowBackgroundBrush {
      bIsDynamicallyLoaded=false
      drawAs=Image
      imageType=NoImage
      mirroring=NoMirror
      resourceName=None
      resourceObject=null
      tiling=NoTile
      imageSize=<0 0>
      margin=<ltrb 0 0 0 0>
      outlineSettings {
       bUseBrushTransparency=false
       roundingType=HalfHeightRadius
       width=0
       color {colorUseRule=UseColor_Specified, specifiedColor=<rgba 0 0 0 0>}
       cornerRadii=<0 0 0 0>
      }
      tintColor {colorUseRule=UseColor_Specified, specifiedColor=<rgba 0.0103 0.0103 0.0103 1>}
      uVRegion {bIsValid=false, max=<0 0>, min=<0 0>}
     }
     evenRowBackgroundHoveredBrush {
      bIsDynamicallyLoaded=false
      drawAs=Image
      imageType=NoImage
      mirroring=NoMirror
      resourceName=None
      resourceObject=null
      tiling=NoTile
      imageSize=<0 0>
      margin=<ltrb 0 0 0 0>
      outlineSettings {
       bUseBrushTransparency=false
       roundingType=HalfHeightRadius
       width=0
       color {colorUseRule=UseColor_Specified, specifiedColor=<rgba 0 0 0 0>}
       cornerRadii=<0 0 0 0>
      }
      tintColor {colorUseRule=UseColor_Specified, specifiedColor=<rgba 0.0176 0.0176 0.0176 1>}
      uVRegion {bIsValid=false, max=<0 0>, min=<0 0>}
     }
     inactiveBrush {
      bIsDynamicallyLoaded=false
      drawAs=Image
      imageType=NoImage
      mirroring=NoMirror
      resourceName=None
      resourceObject=null
      tiling=NoTile
      imageSize=<0 0>
      margin=<ltrb 0 0 0 0>
      outlineSettings {
       bUseBrushTransparency=false
       roundingType=HalfHeightRadius
       width=0
       color {colorUseRule=UseColor_Specified, specifiedColor=<rgba 0 0 0 0>}
       cornerRadii=<0 0 0 0>
      }
      tintColor {colorUseRule=UseColor_Specified, specifiedColor=<rgba 0.0513 0.0953 0.159 1>}
      uVRegion {bIsValid=false, max=<0 0>, min=<0 0>}
     }
     inactiveHighlightedBrush {
      bIsDynamicallyLoaded=false
      drawAs=Image
      imageType=NoImage
      mirroring=NoMirror
      resourceName=None
      resourceObject=null
      tiling=NoTile
      imageSize=<0 0>
      margin=<ltrb 0 0 0 0>
      outlineSettings {
       bUseBrushTransparency=false
       roundingType=HalfHeightRadius
       width=0
       color {colorUseRule=UseColor_Specified, specifiedColor=<rgba 0 0 0 0>}
       cornerRadii=<0 0 0 0>
      }
      tintColor {colorUseRule=UseColor_Specified, specifiedColor=<rgba 0.0252 0.0319 0.0423 1>}
      uVRegion {bIsValid=false, max=<0 0>, min=<0 0>}
     }
     inactiveHoveredBrush {
      bIsDynamicallyLoaded=false
      drawAs=Image
      imageType=NoImage
      mirroring=NoMirror
      resourceName=None
      resourceObject=null
      tiling=NoTile
      imageSize=<0 0>
      margin=<ltrb 0 0 0 0>
      outlineSettings {
       bUseBrushTransparency=false
       roundingType=HalfHeightRadius
       width=0
       color {colorUseRule=UseColor_Specified, specifiedColor=<rgba 0 0 0 0>}
       cornerRadii=<0 0 0 0>
      }
      tintColor {colorUseRule=UseColor_Specified, specifiedColor=<rgba 0.0513 0.0953 0.159 1>}
      uVRegion {bIsValid=false, max=<0 0>, min=<0 0>}
     }
     oddRowBackgroundBrush {
      bIsDynamicallyLoaded=false
      drawAs=Image
      imageType=NoImage
      mirroring=NoMirror
      resourceName=None
      resourceObject=null
      tiling=NoTile
      imageSize=<0 0>
      margin=<ltrb 0 0 0 0>
      outlineSettings {
       bUseBrushTransparency=false
       roundingType=HalfHeightRadius
       width=0
       color {colorUseRule=UseColor_Specified, specifiedColor=<rgba 0 0 0 0>}
       cornerRadii=<0 0 0 0>
      }
      tintColor {colorUseRule=UseColor_Specified, specifiedColor=<rgba 0.0103 0.0103 0.0103 1>}
      uVRegion {bIsValid=false, max=<0 0>, min=<0 0>}
     }
     oddRowBackgroundHoveredBrush {
      bIsDynamicallyLoaded=false
      drawAs=Image
      imageType=NoImage
      mirroring=NoMirror
      resourceName=None
      resourceObject=null
      tiling=NoTile
      imageSize=<0 0>
      margin=<ltrb 0 0 0 0>
      outlineSettings {
       bUseBrushTransparency=false
       roundingType=HalfHeightRadius
       width=0
       color {colorUseRule=UseColor_Specified, specifiedColor=<rgba 0 0 0 0>}
       cornerRadii=<0 0 0 0>
      }
      tintColor {colorUseRule=UseColor_Specified, specifiedColor=<rgba 0.0176 0.0176 0.0176 1>}
      uVRegion {bIsValid=false, max=<0 0>, min=<0 0>}
     }
     parentRowBackgroundBrush {
      bIsDynamicallyLoaded=false
      drawAs=Image
      imageType=NoImage
      mirroring=NoMirror
      resourceName=None
      resourceObject=null
      tiling=NoTile
      imageSize=<32 32>
      margin=<ltrb 0 0 0 0>
      outlineSettings {
       bUseBrushTransparency=false
       roundingType=HalfHeightRadius
       width=0
       color {colorUseRule=UseColor_Specified, specifiedColor=<rgba 0 0 0 0>}
       cornerRadii=<0 0 0 0>
      }
      tintColor {colorUseRule=UseColor_Specified, specifiedColor=<rgba 1 1 1 1>}
      uVRegion {bIsValid=false, max=<0 0>, min=<0 0>}
     }
     parentRowBackgroundHoveredBrush {
      bIsDynamicallyLoaded=false
      drawAs=Image
      imageType=NoImage
      mirroring=NoMirror
      resourceName=None
      resourceObject=null
      tiling=NoTile
      imageSize=<32 32>
      margin=<ltrb 0 0 0 0>
      outlineSettings {
       bUseBrushTransparency=false
       roundingType=HalfHeightRadius
       width=0
       color {colorUseRule=UseColor_Specified, specifiedColor=<rgba 0 0 0 0>}
       cornerRadii=<0 0 0 0>
      }
      tintColor {colorUseRule=UseColor_Specified, specifiedColor=<rgba 1 1 1 1>}
      uVRegion {bIsValid=false, max=<0 0>, min=<0 0>}
     }
     selectedTextColor {colorUseRule=UseColor_Specified, specifiedColor=<rgba 1 1 1 1>}
     selectorFocusedBrush {
      bIsDynamicallyLoaded=false
      drawAs=Border
      imageType=FullColor
      mirroring=NoMirror
      resourceName=../../../Engine/Content/Slate/Common/Selector.png
      resourceObject=null
      tiling=Both
      imageSize=<0 0>
      margin=<ltrb 0.25 0.25 0.25 0.25>
      outlineSettings {
       bUseBrushTransparency=false
       roundingType=HalfHeightRadius
       width=0
       color {colorUseRule=UseColor_Specified, specifiedColor=<rgba 0 0 0 0>}
       cornerRadii=<0 0 0 0>
      }
      tintColor {colorUseRule=UseColor_Specified, specifiedColor=<rgba 0.701 0.225 0.003 1>}
      uVRegion {bIsValid=false, max=<0 0>, min=<0 0>}
     }
     textColor {colorUseRule=UseColor_Specified, specifiedColor=<rgba 0.5271 0.5271 0.5271 1>}
    }
    widgetStyle {
     comboButtonStyle {
      downArrowAlign=VAlign_Center
      buttonStyle {
       clickedSlateSound {resourceObject=null}
       disabled {
        bIsDynamicallyLoaded=false
        drawAs=RoundedBox
        imageType=NoImage
        mirroring=NoMirror
        resourceName=None
        resourceObject=null
        tiling=NoTile
        imageSize=<0 0>
        margin=<ltrb 0 0 0 0>
        outlineSettings {
         bUseBrushTransparency=false
         roundingType=FixedRadius
         width=1
         color {colorUseRule=UseColor_ColorTable, specifiedColor=<rgba 0 0 0 0>}
         cornerRadii=<4 4 4 4>
        }
        tintColor {colorUseRule=UseColor_Specified, specifiedColor=<rgba 0.0395 0.0395 0.0395 1>}
        uVRegion {bIsValid=false, max=<0 0>, min=<0 0>}
       }
       disabledForeground {colorUseRule=UseColor_ColorTable, specifiedColor=<rgba 0 0 0 0>}
       hovered {
        bIsDynamicallyLoaded=false
        drawAs=RoundedBox
        imageType=NoImage
        mirroring=NoMirror
        resourceName=None
        resourceObject=null
        tiling=NoTile
        imageSize=<0 0>
        margin=<ltrb 0 0 0 0>
        outlineSettings {
         bUseBrushTransparency=false
         roundingType=FixedRadius
         width=1
         color {colorUseRule=UseColor_ColorTable, specifiedColor=<rgba 0 0 0 0>}
         cornerRadii=<4 4 4 4>
        }
        tintColor {colorUseRule=UseColor_Specified, specifiedColor=<rgba 0.0048 0.0048 0.0048 1>}
        uVRegion {bIsValid=false, max=<0 0>, min=<0 0>}
       }
       hoveredForeground {colorUseRule=UseColor_ColorTable, specifiedColor=<rgba 0 0 0 0>}
       hoveredSlateSound {resourceObject=null}
       normal {
        bIsDynamicallyLoaded=false
        drawAs=RoundedBox
        imageType=NoImage
        mirroring=NoMirror
        resourceName=None
        resourceObject=null
        tiling=NoTile
        imageSize=<0 0>
        margin=<ltrb 0 0 0 0>
        outlineSettings {
         bUseBrushTransparency=false
         roundingType=FixedRadius
         width=1
         color {colorUseRule=UseColor_ColorTable, specifiedColor=<rgba 0 0 0 0>}
         cornerRadii=<4 4 4 4>
        }
        tintColor {colorUseRule=UseColor_Specified, specifiedColor=<rgba 0.0048 0.0048 0.0048 1>}
        uVRegion {bIsValid=false, max=<0 0>, min=<0 0>}
       }
       normalForeground {colorUseRule=UseColor_ColorTable, specifiedColor=<rgba 0 0 0 0>}
       normalPadding=<ltrb 8 1 8 1>
       pressed {
        bIsDynamicallyLoaded=false
        drawAs=RoundedBox
        imageType=NoImage
        mirroring=NoMirror
        resourceName=None
        resourceObject=null
        tiling=NoTile
        imageSize=<0 0>
        margin=<ltrb 0 0 0 0>
        outlineSettings {
         bUseBrushTransparency=false
         roundingType=FixedRadius
         width=1
         color {colorUseRule=UseColor_ColorTable, specifiedColor=<rgba 0 0 0 0>}
         cornerRadii=<4 4 4 4>
        }
        tintColor {colorUseRule=UseColor_Specified, specifiedColor=<rgba 0.0048 0.0048 0.0048 1>}
        uVRegion {bIsValid=false, max=<0 0>, min=<0 0>}
       }
       pressedForeground {colorUseRule=UseColor_ColorTable, specifiedColor=<rgba 0 0 0 0>}
       pressedPadding=<ltrb 8 2 8 0>
       pressedSlateSound {resourceObject=null}
      }
      contentPadding=<ltrb 0 0 0 0>
      downArrowImage {
       bIsDynamicallyLoaded=false
       drawAs=Image
       imageType=Vector
       mirroring=NoMirror
       resourceName=../../../Engine/Content/Slate/Starship/CoreWidgets/ComboBox/wide-chevron-down.svg
       resourceObject=null
       tiling=NoTile
       imageSize=<16 16>
       margin=<ltrb 0 0 0 0>
       outlineSettings {
        bUseBrushTransparency=false
        roundingType=HalfHeightRadius
        width=0
        color {colorUseRule=UseColor_Specified, specifiedColor=<rgba 0 0 0 0>}
        cornerRadii=<0 0 0 0>
       }
       tintColor {colorUseRule=UseColor_Specified, specifiedColor=<rgba 1 1 1 1>}
       uVRegion {bIsValid=false, max=<0 0>, min=<0 0>}
      }
      downArrowPadding=<ltrb 2 2 2 2>
      menuBorderBrush {
       bIsDynamicallyLoaded=false
       drawAs=RoundedBox
       imageType=NoImage
       mirroring=NoMirror
       resourceName=None
       resourceObject=null
       tiling=NoTile
       imageSize=<0 0>
       margin=<ltrb 0 0 0 0>
       outlineSettings {
        bUseBrushTransparency=false
        roundingType=FixedRadius
        width=1
        color {colorUseRule=UseColor_ColorTable, specifiedColor=<rgba 0 0 0 0>}
        cornerRadii=<0 0 0 0>
       }
       tintColor {colorUseRule=UseColor_Specified, specifiedColor=<rgba 0.0395 0.0395 0.0395 1>}
       uVRegion {bIsValid=false, max=<0 0>, min=<0 0>}
      }
      menuBorderPadding=<ltrb 0 0 0 0>
      shadowColorAndOpacity=<rgba 0 0 0 1>
      shadowOffset=<0 0>
     }
     contentPadding=<ltrb 0 0 0 0>
     menuRowPadding=<ltrb 8 3 8 3>
     pressedSlateSound {resourceObject=null}
     selectionChangeSlateSound {resourceObject=null}
    }
   }
   slot {class=HorizontalBoxSlot, properties {padding=<ltrb 5 0 5 0>, size {sizeRule=Fill, value=1}}}
  }
  TypeImage {
   widgetClassPath=&Image
   bIsVariable=true
   parent=HorizontalBox_49
   widget=TypeImage
   properties {brush {
     bIsDynamicallyLoaded=false
     drawAs=Image
     imageType=FullColor
     mirroring=NoMirror
     resourceName=None
     resourceObject=/AnimInEdTools/AnimationBrowser/Icons/Icon_Categories
     tiling=NoTile
     imageSize=<18 28>
     margin=<ltrb 0 0 0 0>
     outlineSettings {
      bUseBrushTransparency=false
      roundingType=HalfHeightRadius
      width=0
      color {colorUseRule=UseColor_Specified, specifiedColor=<rgba 0 0 0 0>}
      cornerRadii=<0 0 0 0>
     }
     tintColor {colorUseRule=UseColor_Specified, specifiedColor=<rgba 1 1 1 1>}
     uVRegion {bIsValid=false, max=<0 0>, min=<0 0>}
    }}
   slot {class=HorizontalBoxSlot, properties {horizontalAlignment=HAlign_Center, verticalAlignment=VAlign_Center, padding=<ltrb 0 0 5 0>}}
  }
  AnimationType_CombBox {
   widgetClassPath=&SearchableStringComboBox
   bIsVariable=true
   parent=HorizontalBox_49
   widget=AnimationType_CombBox
   properties {
    selectedOption=Subfolder
    toolTipText=animation type dropdown
    defaultOptions [Subfolder]
    foregroundColor {colorUseRule=UseColor_Specified, specifiedColor=<rgba 0.5271 0.5271 0.5271 1>}
    itemStyle {
     bUseParentRowBrush=false
     activeBrush {
      bIsDynamicallyLoaded=false
      drawAs=Image
      imageType=NoImage
      mirroring=NoMirror
      resourceName=None
      resourceObject=null
      tiling=NoTile
      imageSize=<0 0>
      margin=<ltrb 0 0 0 0>
      outlineSettings {
       bUseBrushTransparency=false
       roundingType=HalfHeightRadius
       width=0
       color {colorUseRule=UseColor_Specified, specifiedColor=<rgba 0 0 0 0>}
       cornerRadii=<0 0 0 0>
      }
      tintColor {colorUseRule=UseColor_Specified, specifiedColor=<rgba 0 0.162 0.7454 1>}
      uVRegion {bIsValid=false, max=<0 0>, min=<0 0>}
     }
     activeHighlightedBrush {
      bIsDynamicallyLoaded=false
      drawAs=Image
      imageType=NoImage
      mirroring=NoMirror
      resourceName=None
      resourceObject=null
      tiling=NoTile
      imageSize=<0 0>
      margin=<ltrb 0 0 0 0>
      outlineSettings {
       bUseBrushTransparency=false
       roundingType=HalfHeightRadius
       width=0
       color {colorUseRule=UseColor_Specified, specifiedColor=<rgba 0 0 0 0>}
       cornerRadii=<0 0 0 0>
      }
      tintColor {colorUseRule=UseColor_Specified, specifiedColor=<rgba 0.0252 0.0319 0.0423 1>}
      uVRegion {bIsValid=false, max=<0 0>, min=<0 0>}
     }
     activeHoveredBrush {
      bIsDynamicallyLoaded=false
      drawAs=Image
      imageType=NoImage
      mirroring=NoMirror
      resourceName=None
      resourceObject=null
      tiling=NoTile
      imageSize=<0 0>
      margin=<ltrb 0 0 0 0>
      outlineSettings {
       bUseBrushTransparency=false
       roundingType=HalfHeightRadius
       width=0
       color {colorUseRule=UseColor_Specified, specifiedColor=<rgba 0 0 0 0>}
       cornerRadii=<0 0 0 0>
      }
      tintColor {colorUseRule=UseColor_Specified, specifiedColor=<rgba 0 0.162 0.7454 1>}
      uVRegion {bIsValid=false, max=<0 0>, min=<0 0>}
     }
     dropIndicator_Above {
      bIsDynamicallyLoaded=false
      drawAs=Box
      imageType=FullColor
      mirroring=NoMirror
      resourceName=../../../Engine/Content/Slate/Common/DropZoneIndicator_Above.png
      resourceObject=null
      tiling=NoTile
      imageSize=<0 0>
      margin=<ltrb 0.625 0.625 0 0>
      outlineSettings {
       bUseBrushTransparency=false
       roundingType=HalfHeightRadius
       width=0
       color {colorUseRule=UseColor_Specified, specifiedColor=<rgba 0 0 0 0>}
       cornerRadii=<0 0 0 0>
      }
      tintColor {colorUseRule=UseColor_Specified, specifiedColor=<rgba 0.0091 0.3515 0.9647 1>}
      uVRegion {bIsValid=false, max=<0 0>, min=<0 0>}
     }
     dropIndicator_Below {
      bIsDynamicallyLoaded=false
      drawAs=Box
      imageType=FullColor
      mirroring=NoMirror
      resourceName=../../../Engine/Content/Slate/Common/DropZoneIndicator_Below.png
      resourceObject=null
      tiling=NoTile
      imageSize=<0 0>
      margin=<ltrb 0.625 0 0 0.625>
      outlineSettings {
       bUseBrushTransparency=false
       roundingType=HalfHeightRadius
       width=0
       color {colorUseRule=UseColor_Specified, specifiedColor=<rgba 0 0 0 0>}
       cornerRadii=<0 0 0 0>
      }
      tintColor {colorUseRule=UseColor_Specified, specifiedColor=<rgba 0.0091 0.3515 0.9647 1>}
      uVRegion {bIsValid=false, max=<0 0>, min=<0 0>}
     }
     dropIndicator_Onto {
      bIsDynamicallyLoaded=false
      drawAs=Box
      imageType=FullColor
      mirroring=NoMirror
      resourceName=../../../Engine/Content/Slate/Common/DropZoneIndicator_Onto.png
      resourceObject=null
      tiling=NoTile
      imageSize=<0 0>
      margin=<ltrb 0.25 0.25 0.25 0.25>
      outlineSettings {
       bUseBrushTransparency=false
       roundingType=HalfHeightRadius
       width=0
       color {colorUseRule=UseColor_Specified, specifiedColor=<rgba 0 0 0 0>}
       cornerRadii=<0 0 0 0>
      }
      tintColor {colorUseRule=UseColor_Specified, specifiedColor=<rgba 0.0091 0.3515 0.9647 1>}
      uVRegion {bIsValid=false, max=<0 0>, min=<0 0>}
     }
     evenRowBackgroundBrush {
      bIsDynamicallyLoaded=false
      drawAs=Image
      imageType=NoImage
      mirroring=NoMirror
      resourceName=None
      resourceObject=null
      tiling=NoTile
      imageSize=<0 0>
      margin=<ltrb 0 0 0 0>
      outlineSettings {
       bUseBrushTransparency=false
       roundingType=HalfHeightRadius
       width=0
       color {colorUseRule=UseColor_Specified, specifiedColor=<rgba 0 0 0 0>}
       cornerRadii=<0 0 0 0>
      }
      tintColor {colorUseRule=UseColor_Specified, specifiedColor=<rgba 0.0103 0.0103 0.0103 1>}
      uVRegion {bIsValid=false, max=<0 0>, min=<0 0>}
     }
     evenRowBackgroundHoveredBrush {
      bIsDynamicallyLoaded=false
      drawAs=Image
      imageType=NoImage
      mirroring=NoMirror
      resourceName=None
      resourceObject=null
      tiling=NoTile
      imageSize=<0 0>
      margin=<ltrb 0 0 0 0>
      outlineSettings {
       bUseBrushTransparency=false
       roundingType=HalfHeightRadius
       width=0
       color {colorUseRule=UseColor_Specified, specifiedColor=<rgba 0 0 0 0>}
       cornerRadii=<0 0 0 0>
      }
      tintColor {colorUseRule=UseColor_Specified, specifiedColor=<rgba 0.0176 0.0176 0.0176 1>}
      uVRegion {bIsValid=false, max=<0 0>, min=<0 0>}
     }
     inactiveBrush {
      bIsDynamicallyLoaded=false
      drawAs=Image
      imageType=NoImage
      mirroring=NoMirror
      resourceName=None
      resourceObject=null
      tiling=NoTile
      imageSize=<0 0>
      margin=<ltrb 0 0 0 0>
      outlineSettings {
       bUseBrushTransparency=false
       roundingType=HalfHeightRadius
       width=0
       color {colorUseRule=UseColor_Specified, specifiedColor=<rgba 0 0 0 0>}
       cornerRadii=<0 0 0 0>
      }
      tintColor {colorUseRule=UseColor_Specified, specifiedColor=<rgba 0.0513 0.0953 0.159 1>}
      uVRegion {bIsValid=false, max=<0 0>, min=<0 0>}
     }
     inactiveHighlightedBrush {
      bIsDynamicallyLoaded=false
      drawAs=Image
      imageType=NoImage
      mirroring=NoMirror
      resourceName=None
      resourceObject=null
      tiling=NoTile
      imageSize=<0 0>
      margin=<ltrb 0 0 0 0>
      outlineSettings {
       bUseBrushTransparency=false
       roundingType=HalfHeightRadius
       width=0
       color {colorUseRule=UseColor_Specified, specifiedColor=<rgba 0 0 0 0>}
       cornerRadii=<0 0 0 0>
      }
      tintColor {colorUseRule=UseColor_Specified, specifiedColor=<rgba 0.0252 0.0319 0.0423 1>}
      uVRegion {bIsValid=false, max=<0 0>, min=<0 0>}
     }
     inactiveHoveredBrush {
      bIsDynamicallyLoaded=false
      drawAs=Image
      imageType=NoImage
      mirroring=NoMirror
      resourceName=None
      resourceObject=null
      tiling=NoTile
      imageSize=<0 0>
      margin=<ltrb 0 0 0 0>
      outlineSettings {
       bUseBrushTransparency=false
       roundingType=HalfHeightRadius
       width=0
       color {colorUseRule=UseColor_Specified, specifiedColor=<rgba 0 0 0 0>}
       cornerRadii=<0 0 0 0>
      }
      tintColor {colorUseRule=UseColor_Specified, specifiedColor=<rgba 0.0513 0.0953 0.159 1>}
      uVRegion {bIsValid=false, max=<0 0>, min=<0 0>}
     }
     oddRowBackgroundBrush {
      bIsDynamicallyLoaded=false
      drawAs=Image
      imageType=NoImage
      mirroring=NoMirror
      resourceName=None
      resourceObject=null
      tiling=NoTile
      imageSize=<0 0>
      margin=<ltrb 0 0 0 0>
      outlineSettings {
       bUseBrushTransparency=false
       roundingType=HalfHeightRadius
       width=0
       color {colorUseRule=UseColor_Specified, specifiedColor=<rgba 0 0 0 0>}
       cornerRadii=<0 0 0 0>
      }
      tintColor {colorUseRule=UseColor_Specified, specifiedColor=<rgba 0.0103 0.0103 0.0103 1>}
      uVRegion {bIsValid=false, max=<0 0>, min=<0 0>}
     }
     oddRowBackgroundHoveredBrush {
      bIsDynamicallyLoaded=false
      drawAs=Image
      imageType=NoImage
      mirroring=NoMirror
      resourceName=None
      resourceObject=null
      tiling=NoTile
      imageSize=<0 0>
      margin=<ltrb 0 0 0 0>
      outlineSettings {
       bUseBrushTransparency=false
       roundingType=HalfHeightRadius
       width=0
       color {colorUseRule=UseColor_Specified, specifiedColor=<rgba 0 0 0 0>}
       cornerRadii=<0 0 0 0>
      }
      tintColor {colorUseRule=UseColor_Specified, specifiedColor=<rgba 0.0176 0.0176 0.0176 1>}
      uVRegion {bIsValid=false, max=<0 0>, min=<0 0>}
     }
     parentRowBackgroundBrush {
      bIsDynamicallyLoaded=false
      drawAs=Image
      imageType=NoImage
      mirroring=NoMirror
      resourceName=None
      resourceObject=null
      tiling=NoTile
      imageSize=<32 32>
      margin=<ltrb 0 0 0 0>
      outlineSettings {
       bUseBrushTransparency=false
       roundingType=HalfHeightRadius
       width=0
       color {colorUseRule=UseColor_Specified, specifiedColor=<rgba 0 0 0 0>}
       cornerRadii=<0 0 0 0>
      }
      tintColor {colorUseRule=UseColor_Specified, specifiedColor=<rgba 1 1 1 1>}
      uVRegion {bIsValid=false, max=<0 0>, min=<0 0>}
     }
     parentRowBackgroundHoveredBrush {
      bIsDynamicallyLoaded=false
      drawAs=Image
      imageType=NoImage
      mirroring=NoMirror
      resourceName=None
      resourceObject=null
      tiling=NoTile
      imageSize=<32 32>
      margin=<ltrb 0 0 0 0>
      outlineSettings {
       bUseBrushTransparency=false
       roundingType=HalfHeightRadius
       width=0
       color {colorUseRule=UseColor_Specified, specifiedColor=<rgba 0 0 0 0>}
       cornerRadii=<0 0 0 0>
      }
      tintColor {colorUseRule=UseColor_Specified, specifiedColor=<rgba 1 1 1 1>}
      uVRegion {bIsValid=false, max=<0 0>, min=<0 0>}
     }
     selectedTextColor {colorUseRule=UseColor_Specified, specifiedColor=<rgba 1 1 1 1>}
     selectorFocusedBrush {
      bIsDynamicallyLoaded=false
      drawAs=Border
      imageType=FullColor
      mirroring=NoMirror
      resourceName=../../../Engine/Content/Slate/Common/Selector.png
      resourceObject=null
      tiling=Both
      imageSize=<0 0>
      margin=<ltrb 0.25 0.25 0.25 0.25>
      outlineSettings {
       bUseBrushTransparency=false
       roundingType=HalfHeightRadius
       width=0
       color {colorUseRule=UseColor_Specified, specifiedColor=<rgba 0 0 0 0>}
       cornerRadii=<0 0 0 0>
      }
      tintColor {colorUseRule=UseColor_Specified, specifiedColor=<rgba 0.701 0.225 0.003 1>}
      uVRegion {bIsValid=false, max=<0 0>, min=<0 0>}
     }
     textColor {colorUseRule=UseColor_Specified, specifiedColor=<rgba 0.5271 0.5271 0.5271 1>}
    }
    widgetStyle {
     comboButtonStyle {
      downArrowAlign=VAlign_Center
      buttonStyle {
       clickedSlateSound {resourceObject=null}
       disabled {
        bIsDynamicallyLoaded=false
        drawAs=RoundedBox
        imageType=NoImage
        mirroring=NoMirror
        resourceName=None
        resourceObject=null
        tiling=NoTile
        imageSize=<0 0>
        margin=<ltrb 0 0 0 0>
        outlineSettings {
         bUseBrushTransparency=false
         roundingType=FixedRadius
         width=1
         color {colorUseRule=UseColor_ColorTable, specifiedColor=<rgba 0 0 0 0>}
         cornerRadii=<4 4 4 4>
        }
        tintColor {colorUseRule=UseColor_Specified, specifiedColor=<rgba 0.0395 0.0395 0.0395 1>}
        uVRegion {bIsValid=false, max=<0 0>, min=<0 0>}
       }
       disabledForeground {colorUseRule=UseColor_ColorTable, specifiedColor=<rgba 0 0 0 0>}
       hovered {
        bIsDynamicallyLoaded=false
        drawAs=RoundedBox
        imageType=NoImage
        mirroring=NoMirror
        resourceName=None
        resourceObject=null
        tiling=NoTile
        imageSize=<0 0>
        margin=<ltrb 0 0 0 0>
        outlineSettings {
         bUseBrushTransparency=false
         roundingType=FixedRadius
         width=1
         color {colorUseRule=UseColor_ColorTable, specifiedColor=<rgba 0 0 0 0>}
         cornerRadii=<4 4 4 4>
        }
        tintColor {colorUseRule=UseColor_Specified, specifiedColor=<rgba 0.0048 0.0048 0.0048 1>}
        uVRegion {bIsValid=false, max=<0 0>, min=<0 0>}
       }
       hoveredForeground {colorUseRule=UseColor_ColorTable, specifiedColor=<rgba 0 0 0 0>}
       hoveredSlateSound {resourceObject=null}
       normal {
        bIsDynamicallyLoaded=false
        drawAs=RoundedBox
        imageType=NoImage
        mirroring=NoMirror
        resourceName=None
        resourceObject=null
        tiling=NoTile
        imageSize=<0 0>
        margin=<ltrb 0 0 0 0>
        outlineSettings {
         bUseBrushTransparency=false
         roundingType=FixedRadius
         width=1
         color {colorUseRule=UseColor_ColorTable, specifiedColor=<rgba 0 0 0 0>}
         cornerRadii=<4 4 4 4>
        }
        tintColor {colorUseRule=UseColor_Specified, specifiedColor=<rgba 0.0048 0.0048 0.0048 1>}
        uVRegion {bIsValid=false, max=<0 0>, min=<0 0>}
       }
       normalForeground {colorUseRule=UseColor_ColorTable, specifiedColor=<rgba 0 0 0 0>}
       normalPadding=<ltrb 8 1 8 1>
       pressed {
        bIsDynamicallyLoaded=false
        drawAs=RoundedBox
        imageType=NoImage
        mirroring=NoMirror
        resourceName=None
        resourceObject=null
        tiling=NoTile
        imageSize=<0 0>
        margin=<ltrb 0 0 0 0>
        outlineSettings {
         bUseBrushTransparency=false
         roundingType=FixedRadius
         width=1
         color {colorUseRule=UseColor_ColorTable, specifiedColor=<rgba 0 0 0 0>}
         cornerRadii=<4 4 4 4>
        }
        tintColor {colorUseRule=UseColor_Specified, specifiedColor=<rgba 0.0048 0.0048 0.0048 1>}
        uVRegion {bIsValid=false, max=<0 0>, min=<0 0>}
       }
       pressedForeground {colorUseRule=UseColor_ColorTable, specifiedColor=<rgba 0 0 0 0>}
       pressedPadding=<ltrb 8 2 8 0>
       pressedSlateSound {resourceObject=null}
      }
      contentPadding=<ltrb 0 0 0 0>
      downArrowImage {
       bIsDynamicallyLoaded=false
       drawAs=Image
       imageType=Vector
       mirroring=NoMirror
       resourceName=../../../Engine/Content/Slate/Starship/CoreWidgets/ComboBox/wide-chevron-down.svg
       resourceObject=null
       tiling=NoTile
       imageSize=<16 16>
       margin=<ltrb 0 0 0 0>
       outlineSettings {
        bUseBrushTransparency=false
        roundingType=HalfHeightRadius
        width=0
        color {colorUseRule=UseColor_Specified, specifiedColor=<rgba 0 0 0 0>}
        cornerRadii=<0 0 0 0>
       }
       tintColor {colorUseRule=UseColor_Specified, specifiedColor=<rgba 1 1 1 1>}
       uVRegion {bIsValid=false, max=<0 0>, min=<0 0>}
      }
      downArrowPadding=<ltrb 2 2 2 2>
      menuBorderBrush {
       bIsDynamicallyLoaded=false
       drawAs=RoundedBox
       imageType=NoImage
       mirroring=NoMirror
       resourceName=None
       resourceObject=null
       tiling=NoTile
       imageSize=<0 0>
       margin=<ltrb 0 0 0 0>
       outlineSettings {
        bUseBrushTransparency=false
        roundingType=FixedRadius
        width=1
        color {colorUseRule=UseColor_ColorTable, specifiedColor=<rgba 0 0 0 0>}
        cornerRadii=<0 0 0 0>
       }
       tintColor {colorUseRule=UseColor_Specified, specifiedColor=<rgba 0.0395 0.0395 0.0395 1>}
       uVRegion {bIsValid=false, max=<0 0>, min=<0 0>}
      }
      menuBorderPadding=<ltrb 0 0 0 0>
      shadowColorAndOpacity=<rgba 0 0 0 1>
      shadowOffset=<0 0>
     }
     contentPadding=<ltrb 0 0 0 0>
     menuRowPadding=<ltrb 8 3 8 3>
     pressedSlateSound {resourceObject=null}
     selectionChangeSlateSound {resourceObject=null}
    }
   }
   slot {class=HorizontalBoxSlot, properties {padding=<ltrb 5 0 5 0>, size {sizeRule=Fill, value=1}}}
  }
  SearchBar_HorizontalBox {
   widgetClassPath=&HorizontalBox
   parent=VerticalBox_1
   widget=SearchBar_HorizontalBox
   slot {class=VerticalBoxSlot, properties {padding=<ltrb 2 5 2 0>}}
  }
  SearchBar {
   widgetClassPath=&EditorUtilityEditableTextBox
   bIsVariable=true
   parent=SearchBar_HorizontalBox
   widget=SearchBar
   properties {hintText=Search}
   slot {class=HorizontalBoxSlot, properties {size {sizeRule=Fill, value=1}}}
  }
  ClearSearchBox {
   widgetClassPath=&Button
   parent=SearchBar_HorizontalBox
   widget=ClearSearchBox
   properties {widgetStyle {
     clickedSlateSound {resourceObject=null}
     disabled {
      bIsDynamicallyLoaded=false
      drawAs=RoundedBox
      imageType=NoImage
      mirroring=NoMirror
      resourceName=None
      resourceObject=null
      tiling=NoTile
      imageSize=<0 0>
      margin=<ltrb 0 0 0 0>
      outlineSettings {
       bUseBrushTransparency=false
       roundingType=FixedRadius
       width=1
       color {colorUseRule=UseColor_ColorTable, specifiedColor=<rgba 0 0 0 0>}
       cornerRadii=<4 4 4 4>
      }
      tintColor {colorUseRule=UseColor_Specified, specifiedColor=<rgba 0.0395 0.0395 0.0395 1>}
      uVRegion {bIsValid=false, max=<0 0>, min=<0 0>}
     }
     disabledForeground {colorUseRule=UseColor_ColorTable, specifiedColor=<rgba 0 0 0 0>}
     hovered {
      bIsDynamicallyLoaded=false
      drawAs=RoundedBox
      imageType=NoImage
      mirroring=NoMirror
      resourceName=None
      resourceObject=null
      tiling=NoTile
      imageSize=<32 32>
      margin=<ltrb 0 0 0 0>
      outlineSettings {
       bUseBrushTransparency=false
       roundingType=FixedRadius
       width=0
       color {colorUseRule=UseColor_ColorTable, specifiedColor=<rgba 0 0 0 0>}
       cornerRadii=<4 4 4 4>
      }
      tintColor {colorUseRule=UseColor_Specified, specifiedColor=<rgba 0.0953 0.0953 0.0953 1>}
      uVRegion {bIsValid=false, max=<0 0>, min=<0 0>}
     }
     hoveredForeground {colorUseRule=UseColor_ColorTable, specifiedColor=<rgba 0 0 0 0>}
     hoveredSlateSound {resourceObject=null}
     normal {
      bIsDynamicallyLoaded=false
      drawAs=NoDrawType
      imageType=FullColor
      mirroring=NoMirror
      resourceName=None
      resourceObject=null
      tiling=NoTile
      imageSize=<16 16>
      margin=<ltrb 0 0 0 0>
      outlineSettings {
       bUseBrushTransparency=false
       roundingType=FixedRadius
       width=0
       color {colorUseRule=UseColor_ColorTable, specifiedColor=<rgba 0 0 0 0>}
       cornerRadii=<4 4 4 4>
      }
      tintColor {colorUseRule=UseColor_Specified, specifiedColor=<rgba 0.0176 0.0176 0.0176 1>}
      uVRegion {bIsValid=false, max=<0 0>, min=<0 0>}
     }
     normalForeground {colorUseRule=UseColor_ColorTable, specifiedColor=<rgba 0 0 0 0>}
     normalPadding=<ltrb 6 1.5 6 1.5>
     pressed {
      bIsDynamicallyLoaded=false
      drawAs=RoundedBox
      imageType=NoImage
      mirroring=NoMirror
      resourceName=None
      resourceObject=null
      tiling=NoTile
      imageSize=<32 32>
      margin=<ltrb 0 0 0 0>
      outlineSettings {
       bUseBrushTransparency=false
       roundingType=FixedRadius
       width=0
       color {colorUseRule=UseColor_ColorTable, specifiedColor=<rgba 0 0 0 0>}
       cornerRadii=<4 4 4 4>
      }
      tintColor {colorUseRule=UseColor_Specified, specifiedColor=<rgba 0.0284 0.0284 0.0284 1>}
      uVRegion {bIsValid=false, max=<0 0>, min=<0 0>}
     }
     pressedForeground {colorUseRule=UseColor_ColorTable, specifiedColor=<rgba 0 0 0 0>}
     pressedPadding=<ltrb 6 2.5 6 0.5>
     pressedSlateSound {resourceObject=null}
    }}
   slot {class=HorizontalBoxSlot, properties {horizontalAlignment=HAlign_Center}}
  }
  X {
   widgetClassPath=&TextBlock
   parent=ClearSearchBox
   widget=X
   properties {
    autoWrapText=true
    justification=Center
    text=x
    colorAndOpacity {colorUseRule=UseColor_Specified, specifiedColor=<rgba 0.5255 0.5255 0.5255 1>}
    font {
     bForceMonospaced=false
     bMaterialIsStencil=false
     fontMaterial=null
     fontObject=&Roboto
     letterSpacing=0
     monospacedWidth=1
     size=10
     skewAmount=0
     typefaceFontName=Bold
     outlineSettings {
      bApplyOutlineToDropShadows=false
      bMiteredCorners=false
      bSeparateFillAlpha=false
      outlineMaterial=null
      outlineSize=0
      outlineColor=<rgba 0 0 0 1>
     }
    }
   }
   slot {class=ButtonSlot}
  }
  SettingsButton {
   widgetClassPath=&Button
   parent=SearchBar_HorizontalBox
   widget=SettingsButton
   properties {
    toolTipText=Settings Menu
    widgetStyle {
     clickedSlateSound {resourceObject=null}
     disabled {
      bIsDynamicallyLoaded=false
      drawAs=RoundedBox
      imageType=NoImage
      mirroring=NoMirror
      resourceName=None
      resourceObject=null
      tiling=NoTile
      imageSize=<0 0>
      margin=<ltrb 0 0 0 0>
      outlineSettings {
       bUseBrushTransparency=false
       roundingType=FixedRadius
       width=1
       color {colorUseRule=UseColor_ColorTable, specifiedColor=<rgba 0 0 0 0>}
       cornerRadii=<4 4 4 4>
      }
      tintColor {colorUseRule=UseColor_Specified, specifiedColor=<rgba 0.0395 0.0395 0.0395 1>}
      uVRegion {bIsValid=false, max=<0 0>, min=<0 0>}
     }
     disabledForeground {colorUseRule=UseColor_ColorTable, specifiedColor=<rgba 0 0 0 0>}
     hovered {
      bIsDynamicallyLoaded=false
      drawAs=RoundedBox
      imageType=FullColor
      mirroring=NoMirror
      resourceName=None
      resourceObject=null
      tiling=NoTile
      imageSize=<28 28>
      margin=<ltrb 0 0 0 0>
      outlineSettings {
       bUseBrushTransparency=false
       roundingType=FixedRadius
       width=0
       color {colorUseRule=UseColor_ColorTable, specifiedColor=<rgba 0 0 0 0>}
       cornerRadii=<4 4 4 4>
      }
      tintColor {colorUseRule=UseColor_Specified, specifiedColor=<rgba 0.0941 0.0941 0.0941 1>}
      uVRegion {bIsValid=false, max=<0 0>, min=<0 0>}
     }
     hoveredForeground {colorUseRule=UseColor_ColorTable, specifiedColor=<rgba 0 0 0 0>}
     hoveredSlateSound {resourceObject=null}
     normal {
      bIsDynamicallyLoaded=false
      drawAs=NoDrawType
      imageType=FullColor
      mirroring=NoMirror
      resourceName=None
      resourceObject=null
      tiling=NoTile
      imageSize=<28 28>
      margin=<ltrb 0 0 0 0>
      outlineSettings {
       bUseBrushTransparency=false
       roundingType=FixedRadius
       width=0
       color {colorUseRule=UseColor_ColorTable, specifiedColor=<rgba 0 0 0 0>}
       cornerRadii=<4 4 4 4>
      }
      tintColor {colorUseRule=UseColor_Specified, specifiedColor=<rgba 0.0176 0.0176 0.0176 1>}
      uVRegion {bIsValid=false, max=<0 0>, min=<0 0>}
     }
     normalForeground {colorUseRule=UseColor_ColorTable, specifiedColor=<rgba 0 0 0 0>}
     normalPadding=<ltrb 0 0 0 0>
     pressed {
      bIsDynamicallyLoaded=false
      drawAs=RoundedBox
      imageType=FullColor
      mirroring=NoMirror
      resourceName=None
      resourceObject=null
      tiling=NoTile
      imageSize=<28 28>
      margin=<ltrb 0 0 0 0>
      outlineSettings {
       bUseBrushTransparency=false
       roundingType=FixedRadius
       width=0
       color {colorUseRule=UseColor_ColorTable, specifiedColor=<rgba 0 0 0 0>}
       cornerRadii=<4 4 4 4>
      }
      tintColor {colorUseRule=UseColor_Specified, specifiedColor=<rgba 0.0392 0.0392 0.0392 1>}
      uVRegion {bIsValid=false, max=<0 0>, min=<0 0>}
     }
     pressedForeground {colorUseRule=UseColor_ColorTable, specifiedColor=<rgba 0 0 0 0>}
     pressedPadding=<ltrb 0 0 0 0>
     pressedSlateSound {resourceObject=null}
    }
   }
   slot {class=HorizontalBoxSlot, properties {horizontalAlignment=HAlign_Center, verticalAlignment=VAlign_Center}}
  }
  HorizontalBox_5 {
   widgetClassPath=&HorizontalBox
   parent=SettingsButton
   widget=HorizontalBox_5
   slot {class=ButtonSlot, properties {padding=<ltrb 0 0 0 0>}}
  }
  SettingsButtonImage {
   widgetClassPath=&Image
   bIsVariable=true
   parent=HorizontalBox_5
   widget=SettingsButtonImage
   properties {brush {
     bIsDynamicallyLoaded=false
     drawAs=Image
     imageType=FullColor
     mirroring=NoMirror
     resourceName=None
     resourceObject=/AnimInEdTools/AnimationBrowser/Icons/Icon_Settings
     tiling=NoTile
     imageSize=<28 28>
     margin=<ltrb 0 0 0 0>
     outlineSettings {
      bUseBrushTransparency=false
      roundingType=HalfHeightRadius
      width=0
      color {colorUseRule=UseColor_Specified, specifiedColor=<rgba 0 0 0 0>}
      cornerRadii=<0 0 0 0>
     }
     tintColor {colorUseRule=UseColor_Specified, specifiedColor=<rgba 1 1 1 1>}
     uVRegion {bIsValid=false, max=<0 0>, min=<0 0>}
    }}
   slot {class=HorizontalBoxSlot, properties {horizontalAlignment=HAlign_Center, verticalAlignment=VAlign_Center}}
  }
  ListViewBorder {
   widgetClassPath=&Border
   parent=VerticalBox_29
   widget=ListViewBorder
   properties {brushColor=<rgba 0.0048 0.0048 0.0048 1>, padding=<ltrb 0 0 0 0>}
   slot {class=VerticalBoxSlot, properties {padding=<ltrb 10 2 10 5>, size {sizeRule=Fill, value=1}}}
  }
  ListView {
   widgetClassPath=&CommonListView
   bIsVariable=true
   parent=ListViewBorder
   widget=ListView
   properties {
    allowOverscroll=false
    bAllowKeepPreselectedItems=false
    bEnableRightClickScrolling=false
    bEnableScrollAnimation=false
    consumeMouseWheel=Always
    entryWidgetClass=&/LevelSequenceBrowser_ListItem.LevelSequenceBrowser_ListItem_C
    selectionMode=Multi
    scrollBarStyle {
     thickness=8
     draggedThumbImage {
      bIsDynamicallyLoaded=false
      drawAs=RoundedBox
      imageType=NoImage
      mirroring=NoMirror
      resourceName=None
      resourceObject=null
      tiling=NoTile
      imageSize=<0 0>
      margin=<ltrb 0 0 0 0>
      outlineSettings {
       bUseBrushTransparency=false
       roundingType=FixedRadius
       width=0
       color {colorUseRule=UseColor_Specified, specifiedColor=<rgba 0 0 0 0>}
       cornerRadii=<4 4 4 4>
      }
      tintColor {colorUseRule=UseColor_Specified, specifiedColor=<rgba 0.2159 0.2159 0.2159 1>}
      uVRegion {bIsValid=false, max=<0 0>, min=<0 0>}
     }
     horizontalBackgroundImage {
      bIsDynamicallyLoaded=false
      drawAs=NoDrawType
      imageType=NoImage
      mirroring=NoMirror
      resourceName=None
      resourceObject=null
      tiling=NoTile
      imageSize=<0 0>
      margin=<ltrb 0 0 0 0>
      outlineSettings {
       bUseBrushTransparency=false
       roundingType=HalfHeightRadius
       width=0
       color {colorUseRule=UseColor_Specified, specifiedColor=<rgba 0 0 0 0>}
       cornerRadii=<0 0 0 0>
      }
      tintColor {colorUseRule=UseColor_Specified, specifiedColor=<rgba 1 1 1 1>}
      uVRegion {bIsValid=false, max=<0 0>, min=<0 0>}
     }
     horizontalBottomSlotImage {
      bIsDynamicallyLoaded=false
      drawAs=NoDrawType
      imageType=NoImage
      mirroring=NoMirror
      resourceName=None
      resourceObject=null
      tiling=NoTile
      imageSize=<0 0>
      margin=<ltrb 0 0 0 0>
      outlineSettings {
       bUseBrushTransparency=false
       roundingType=HalfHeightRadius
       width=0
       color {colorUseRule=UseColor_Specified, specifiedColor=<rgba 0 0 0 0>}
       cornerRadii=<0 0 0 0>
      }
      tintColor {colorUseRule=UseColor_Specified, specifiedColor=<rgba 1 1 1 1>}
      uVRegion {bIsValid=false, max=<0 0>, min=<0 0>}
     }
     horizontalTopSlotImage {
      bIsDynamicallyLoaded=false
      drawAs=NoDrawType
      imageType=NoImage
      mirroring=NoMirror
      resourceName=None
      resourceObject=null
      tiling=NoTile
      imageSize=<0 0>
      margin=<ltrb 0 0 0 0>
      outlineSettings {
       bUseBrushTransparency=false
       roundingType=HalfHeightRadius
       width=0
       color {colorUseRule=UseColor_Specified, specifiedColor=<rgba 0 0 0 0>}
       cornerRadii=<0 0 0 0>
      }
      tintColor {colorUseRule=UseColor_Specified, specifiedColor=<rgba 1 1 1 1>}
      uVRegion {bIsValid=false, max=<0 0>, min=<0 0>}
     }
     hoveredThumbImage {
      bIsDynamicallyLoaded=false
      drawAs=RoundedBox
      imageType=NoImage
      mirroring=NoMirror
      resourceName=None
      resourceObject=null
      tiling=NoTile
      imageSize=<0 0>
      margin=<ltrb 0 0 0 0>
      outlineSettings {
       bUseBrushTransparency=false
       roundingType=FixedRadius
       width=0
       color {colorUseRule=UseColor_Specified, specifiedColor=<rgba 0 0 0 0>}
       cornerRadii=<4 4 4 4>
      }
      tintColor {colorUseRule=UseColor_Specified, specifiedColor=<rgba 0.2159 0.2159 0.2159 1>}
      uVRegion {bIsValid=false, max=<0 0>, min=<0 0>}
     }
     normalThumbImage {
      bIsDynamicallyLoaded=false
      drawAs=RoundedBox
      imageType=NoImage
      mirroring=NoMirror
      resourceName=None
      resourceObject=null
      tiling=NoTile
      imageSize=<0 0>
      margin=<ltrb 0 0 0 0>
      outlineSettings {
       bUseBrushTransparency=false
       roundingType=FixedRadius
       width=0
       color {colorUseRule=UseColor_Specified, specifiedColor=<rgba 0 0 0 0>}
       cornerRadii=<4 4 4 4>
      }
      tintColor {colorUseRule=UseColor_Specified, specifiedColor=<rgba 0.0953 0.0953 0.0953 1>}
      uVRegion {bIsValid=false, max=<0 0>, min=<0 0>}
     }
     verticalBackgroundImage {
      bIsDynamicallyLoaded=false
      drawAs=NoDrawType
      imageType=NoImage
      mirroring=NoMirror
      resourceName=None
      resourceObject=null
      tiling=NoTile
      imageSize=<0 0>
      margin=<ltrb 0 0 0 0>
      outlineSettings {
       bUseBrushTransparency=false
       roundingType=HalfHeightRadius
       width=0
       color {colorUseRule=UseColor_Specified, specifiedColor=<rgba 0 0 0 0>}
       cornerRadii=<0 0 0 0>
      }
      tintColor {colorUseRule=UseColor_Specified, specifiedColor=<rgba 1 1 1 1>}
      uVRegion {bIsValid=false, max=<0 0>, min=<0 0>}
     }
     verticalBottomSlotImage {
      bIsDynamicallyLoaded=false
      drawAs=NoDrawType
      imageType=NoImage
      mirroring=NoMirror
      resourceName=None
      resourceObject=null
      tiling=NoTile
      imageSize=<0 0>
      margin=<ltrb 0 0 0 0>
      outlineSettings {
       bUseBrushTransparency=false
       roundingType=HalfHeightRadius
       width=0
       color {colorUseRule=UseColor_Specified, specifiedColor=<rgba 0 0 0 0>}
       cornerRadii=<0 0 0 0>
      }
      tintColor {colorUseRule=UseColor_Specified, specifiedColor=<rgba 1 1 1 1>}
      uVRegion {bIsValid=false, max=<0 0>, min=<0 0>}
     }
     verticalTopSlotImage {
      bIsDynamicallyLoaded=false
      drawAs=NoDrawType
      imageType=NoImage
      mirroring=NoMirror
      resourceName=None
      resourceObject=null
      tiling=NoTile
      imageSize=<0 0>
      margin=<ltrb 0 0 0 0>
      outlineSettings {
       bUseBrushTransparency=false
       roundingType=HalfHeightRadius
       width=0
       color {colorUseRule=UseColor_Specified, specifiedColor=<rgba 0 0 0 0>}
       cornerRadii=<0 0 0 0>
      }
      tintColor {colorUseRule=UseColor_Specified, specifiedColor=<rgba 1 1 1 1>}
      uVRegion {bIsValid=false, max=<0 0>, min=<0 0>}
     }
    }
    widgetStyle {backgroundBrush {
      bIsDynamicallyLoaded=false
      drawAs=NoDrawType
      imageType=NoImage
      mirroring=NoMirror
      resourceName=None
      resourceObject=null
      tiling=NoTile
      imageSize=<0 0>
      margin=<ltrb 0 0 0 0>
      outlineSettings {
       bUseBrushTransparency=false
       roundingType=HalfHeightRadius
       width=0
       color {colorUseRule=UseColor_Specified, specifiedColor=<rgba 0 0 0 0>}
       cornerRadii=<0 0 0 0>
      }
      tintColor {colorUseRule=UseColor_Specified, specifiedColor=<rgba 0 0 0 1>}
      uVRegion {bIsValid=false, max=<0 0>, min=<0 0>}
     }}
   }
   slot {class=BorderSlot, properties {padding=<ltrb 0 0 0 0>}}
  }
  Overlay_173 {
   widgetClassPath=&Overlay
   parent=CanvasPanel_21
   widget=Overlay_173
   slot {
    class=CanvasPanelSlot
    properties {zOrder=1, layoutData {alignment=<0 0>, anchors {maximum=<1 0.5>, minimum=<0 0.5>}, offsets=<ltrb 0 -540.5405 0 1081.0811>}}
   }
  }
  EUW_ContextMenu {
   widgetClassPath=/AnimInEdTools/ContextMenu/EUW_ContextMenu
   bIsVariable=true
   parent=Overlay_173
   widget=EUW_ContextMenu
   properties {visibility=Collapsed}
   slot {class=OverlaySlot}
  }
  ModalLayer {
   widgetClassPath=&NamedSlot
   bIsVariable=true
   parent=Overlay_173
   widget=ModalLayer
   properties {bExposeOnInstanceOnly=true}
   slot {class=OverlaySlot, properties {horizontalAlignment=HAlign_Fill, verticalAlignment=VAlign_Fill}}
  }
 ]
}
