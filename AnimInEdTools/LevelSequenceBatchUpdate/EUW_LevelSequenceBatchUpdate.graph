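# Editor utility widget that batch re-exports Anim Sequences linked to Level Sequences chosen in an embedded Level Sequence Browser.
# Update Selected or Update All first checks source control, warning about outdated or checked-out assets.
# Python scripts re-bake each linked animation with stored link settings, including excluded curves, under a progress bar.
# It then reapplies Animation Modifiers, saves, and reports Level Sequences flagged as missing modifiers.
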EditorUtilityWidgetBlueprint /AnimInEdTools/LevelSequenceBatchUpdate/EUW_LevelSequenceBatchUpdate
graphs [
 "Adding AnimSeq to Map, Query AnimSeq State" {nodes [
   "Adding AnimSeq to Map, Query AnimSeq State"() {is=title K2Node_FunctionEntry}
   > n1 = For Each Loop(Array:Array of Level Sequence Object References=Get LSArray() -> (LSArray:Array of Level Sequence Object References)) -> (Array Element:Level Sequence Object Reference)
   Return Node!(execute=@n1.Completed, Are AnimSeq Valid:bool=Get Was it Done() -> (Was it Done:bool))
   > n2 = Set Current LS(Current LS:Level Sequence Object Reference=@n1) -> (Output_Get:Level Sequence Object Reference)
   > n3 = GetBindings!(execute, Sequence:Movie Scene Sequence Object Reference=@n2) -> (then, ReturnValue:Array of Movie Scene Binding Proxy Structures)
   > n4 = GetAnimSequenceLinkFromLevelSequence!(execute, InLevelSequence:Level Sequence Object Reference=Get Current LS() -> (Current LS:Level Sequence Object Reference)) -> (then, ReturnValue:Level Sequence Anim Sequence Link Object Reference)
   > n5 = For Each Loop(Array:Array of Movie Scene Binding Proxy Structures=@n3) -> (Array Element:Movie Scene Binding Proxy Structure)
   > Set Anim_Movie_Binding(Anim_Movie_Binding:Movie Scene Binding Proxy Structure=@n5) -> (Output_Get:Movie Scene Binding Proxy Structure)
   > n6 = For Each Loop!(
    Exec
    Array:Array of Level Sequence Anim Sequence Link Item Structures=Get AnimSequenceLinks!(self:Level Sequence Anim Sequence Link Object Reference=@n4) -> (AnimSequenceLinks:Array of Level Sequence Anim Sequence Link Item Structures)
   ) -> (
    LoopBody
    Array Element_SkelTrackGuid:Guid Structure
    Array Element_PathToAnimSequence:Soft Object Path Structure
    Array Element_bAutoBake:bool=~
    Array Element_bExportTransforms:bool=~
    Array Element_bExportMorphTargets:bool=~
    Array Element_bExportAttributeCurves:bool=~
    Array Element_bExportMaterialCurves:bool=~
    Array Element_bSkipCurvesWithZeroValue:bool=~
    Array Element_bSetRetargetSourceAsset:bool=~
    Array Element_Interpolation:EAnimInterpolationType Enum
    Array Element_CurveInterpolation:ERichCurveInterpMode Enum
    Array Element_bRecordInWorldSpace:bool=~
    Array Element_bEvaluateAllSkeletalMeshComponents:bool=~
    Array Element_IncludeAnimationNames:Array of Strings
    Array Element_ExcludeAnimationNames:Array of Strings
    Array Element_bRemoveExcludedCurves:bool=~
    Array Element_WarmUpFrames:Frame Number Structure
    Array Element_DelayBeforeStart:Frame Number Structure
    Array Element_bUseCustomTimeRange:bool=~
    Array Element_CustomStartFrame:Frame Number Structure
    Array Element_CustomEndFrame:Frame Number Structure
    Array Element_CustomDisplayRate:Frame Rate Structure
    Array Element_bUseCustomFrameRate:bool=~
    Array Element_CustomFrameRate:Frame Rate Structure
    Array Index
    Completed
   )
   > n7 = FindBindingById!(execute, Sequence:Movie Scene Sequence Object Reference=Get Current LS(), BindingId:Guid Structure=@n6) -> (then, ReturnValue_BindingID:Guid Structure, ReturnValue_Sequence:Movie Scene Sequence Object Reference)
   n8 = Get Anim_Movie_Binding() -> (Anim_Movie_Binding:Movie Scene Binding Proxy Structure)
   > Branch(Condition="Equal (Guid)"(
     A:"Guid Structure (by ref)"=@n7
     B:"Guid Structure (by ref)"=Break Movie Scene Binding Proxy!(MovieSceneBindingProxy:"Movie Scene Binding Proxy Structure (by ref)"=@n8) -> (BindingID:Guid Structure, Sequence:Movie Scene Sequence Object Reference) {is=title K2Node_BreakStruct}
    ) -> (ReturnValue:bool=~) {is=title K2Node_PromotableOperator})
   > n9 = LoadAsset(AssetPath=BreakSoftObjectPath(InSoftObjectPath=@n6.Array Element_PathToAnimSequence))
   > n10 = Cast To AnimSequence!(execute, Object=@n9) -> (then, CastFailed, AsAnim Sequence:Anim Sequence Object Reference)
   > n11 = Query Anim Sequence!(
    execute
    self
    InFile:String=GetSystemPath!(Object:Object Reference=@n10) -> (ReturnValue:String)
    Anim Sequence:Anim Sequence Object Reference=@n10
   ) -> (then, Was it done:bool)
   > n12 = Set Was it Done(Was it Done:bool=@n11) -> (Output_Get:bool)
   > Branch(Condition=@n12)
   > n13 = Does Anim Modifier Exists!(execute, self, incoming AnimSeq:Interface Asset User Data Interface=@n10, incomgin LS:Level Sequence Object Reference=Get Current LS()) -> (then, Selected LS missing Modifier:Level Sequence Object Reference)
   > Add Unique(
    TargetArray:Array of Level Sequence Object References=Get Selected LS Missing Modifier array() -> (Selected LS Missing Modifier array:Array of Level Sequence Object References)
    NewItem:"Level Sequence Object Reference (by ref)"=@n13
   )
  ]}
 "Adding LS to Map, Query LS State" {nodes [
   "Adding LS to Map, Query LS State"!() -> (then, Selected Items:Array of Object References) {is=title K2Node_FunctionEntry}
   > Clear(TargetArray:Array of Level Sequence Object References=Get LSArray())
   > GetSelectedItems!(
    execute
    self:List View Object Reference=Get ListView!(self:EUW Level Sequence Browser Object Reference=Get LevelSeqBrowser() -> (LevelSeqBrowser:EUW Level Sequence Browser Object Reference)) -> (ListView:Common List View Object Reference)
   ) -> (then, Items:Array of Object References, ReturnValue:bool)
   > n1 = For Each Loop!(Exec, Array:Array of Object References=Get Selected Items() -> (Selected Items:Array of Object References)) -> (LoopBody, Array Element:Object Reference, Array Index, Completed)
   > n2 = Cast To Object_LevelSequenceBrowser!(execute, Object=@n1) -> (then, CastFailed, AsObject Level Sequence Browser:Object Level Sequence Browser Object Reference) {is=title K2Node_DynamicCast}
   n3 = To Soft Object Path!(SoftObjectReference:Soft Object Reference=Get!(self:Object Level Sequence Browser Object Reference=@n2) -> (Data_LS_58_#1:Soft Object Reference)) -> (ReturnValue_PathString:String)
   > n4 = LoadAsset_Blocking(Asset=To Soft Object Reference(SoftObjectPath=MakeSoftClassPath(PathString=@n3)))
   > n5 = Cast To LevelSequence!(execute, Object=@n4) -> (then, CastFailed, AsLevel Sequence:Level Sequence Object Reference) {is=title K2Node_DynamicCast}
   > n6 = Set Local LS(Local LS:Level Sequence Object Reference=@n5) -> (Output_Get:Level Sequence Object Reference)
   > n7 = QueryLS!(execute, self, InFile:String=GetSystemPath(Object=@n6), CurrentLS:Level Sequence Object Reference=@n6) -> (then, Was it done:bool)
   > Add Unique(TargetArray=Get LSArray(), NewItem=Get Local LS() -> (Local LS:Level Sequence Object Reference))
   Return Node!(execute=@n1.Completed, LS Valid?:bool=@n7)
  ]}
 Checked out by another user warning {nodes [
   n1 = Checked out by another user warning!() -> (then, Checked out by:String, Dipslay aset name:String) {is=title K2Node_FunctionEntry}
   > Show Message Dialog(
    Title=MakeLiteralText(Value="LOCTEXT(\"Cannot Edit File\")")
    Message="To Text (String)"(InString=Append!(A="The seleted file  ", B=@n1.Dipslay aset name, C:String=" is currently checked out by ", D:String=@n1) -> (ReturnValue))
    MessageType=Ok
    DefaultValue=Ok
    MessageCategory=Error
   )
   > Return Node!(execute, Was it done:bool=false)
  ]}
 Does Anim Modifier Exists {nodes [
   n1 = Does Anim Modifier Exists!() -> (then, incoming AnimSeq:Interface Asset User Data Interface, incomgin LS:Level Sequence Object Reference) {is=title K2Node_FunctionEntry}
   > n2 = GetAssetUserDataOfClass!(
    execute
    self:Interface Asset User Data Interface=@n1
    InUserDataClass:Asset User Data Class Reference=/Script/AnimationModifiers.AnimationModifiersAssetUserData
   ) -> (then, ReturnValue:Animation Modifiers Asset User Data Object Reference)
   > n3 = Branch(Condition=Is Empty!(TargetArray:Array of Animation Modifier Object References=Get AnimationModifierInstances!(self:Animation Modifiers Asset User Data Object Reference=@n2) -> (AnimationModifierInstances:Array of Animation Modifier Object References)) -> (ReturnValue:bool))
   > Return Node!(execute, Selected LS missing Modifier:Level Sequence Object Reference=@n1.incomgin LS)
   Return Node(execute=@n3.else)
  ]}
 EventGraph {nodes [
   "On Clicked (Update_LS_Btn)"() {is=title K2Node_ComponentBoundEvent}
   > Clear(TargetArray=Get Selected LS Missing Modifier array())
   > n1 = GetSelectedItems(self=Get ListView(self=Get LevelSeqBrowser()))
   > n2 = "Adding LS to Map, Query LS State"!(execute, self, Selected Items:Array of Object References=@n1) -> (then, LS Valid?:bool) {is=title K2Node_CallFunction}
   > Branch(Condition=@n2)
   > n3 = "Adding AnimSeq to Map, Query AnimSeq State"!(execute, self) -> (then, Are AnimSeq Valid:bool) {is=title K2Node_CallFunction}
   > Branch(Condition=@n3)
   > n4 = "Open, Save, Close Level Sequence"!(execute, self, List Objects selected:Array of Level Sequence Object References=Get LSArray()) -> (then, WasItDone:bool)
   > Branch(Condition=@n4)
   > RetriggerableDelay!(execute, WorldContextObject:Object Reference, Duration=0.3) -> (then)
   > Success Menu()
   "On Clicked (Update_All_LS_Btn)"() {is=title K2Node_ComponentBoundEvent}
   n5 = Get Selected_Asset() -> (Selected_Asset:String)
   > n6 = Show Message Dialog(
    Title=Get Popup_Title() -> (Popup_Title:Text)
    Message="To Text (String)"(InString=Append!(
      A="Are you sure you want to update all the Level Sequences for the "
      B=@n5
      C=" asset? This can take a while to complete.\r\n\r\nThis will check out all the Level and Anim Sequences for this asset and others will not be able to work on them as this is running. "
     ) -> (ReturnValue))
    MessageType=YesNo
    DefaultValue=No
    MessageCategory=Warning
   )
   > n7 = Branch(Condition="Equal (Enum)"(A:EAppReturnType Enum=@n6, B:EAppReturnType Enum=Yes))
   > Clear(TargetArray=Get Selected LS Missing Modifier array())
   n8 = Get LevelSeqBrowser()
   > n9 = "Adding LS to Map, Query LS State"(Selected Items=GetListItems!(self:List View Object Reference=Get ListView(self=@n8)) -> (ReturnValue:Array of Object References))
   > Branch(Condition=@n9)
   > n10 = "Adding AnimSeq to Map, Query AnimSeq State"()
   > Branch(Condition=@n10)
   > n11 = "Open, Save, Close Level Sequence"(List Objects selected=Get LSArray())
   > Branch(Condition=@n11)
   > RetriggerableDelay(Duration=0.3)
   > Success Menu()
   PrintString(execute=@n7.else, InString=Action Cancelled)
   Event Construct()
   > PopulatingSlot()
  ]}
 Not Current Warning {nodes [
   n1 = Not Current Warning!() -> (then, Type of Asset:String, Display Asset Name:String) {is=title K2Node_FunctionEntry}
   > Show Message Dialog(
    Title=Get Popup_Title()
    Message="To Text (String)"(InString=Append(A=@n1, B=@n1.Display Asset Name, C=" is not the latest. Please get the latest for this asset and try again."))
    MessageType=Ok
    DefaultValue=No
    MessageCategory=Error
   )
   > Return Node!(execute, Could it be done:bool=false)
   GetDisplayName(Object=Get Selected_LS() -> (Selected_LS:Level Sequence Object Reference))
  ]}
 OnMouseButtonDown {nodes [
   OnMouseButtonDown!() -> (then, MyGeometry:Geometry Structure, MouseEvent:"Pointer Event Structure (by ref)") {is=title K2Node_FunctionEntry}
   > Show!(
    execute
    self:EUW Context Menu Object Reference=Get EUW_ContextMenu() -> (EUW_ContextMenu:EUW Context Menu Object Reference)
    Show:bool=false
    Mouse Position:Vector 2D Structure
    Name:String
   ) -> (then)
   > Return Node!(execute, ReturnValue:Event Reply Structure=Unhandled())
  ]}
 "Open, Save, Close Level Sequence" {nodes [
   n1 = "Open, Save, Close Level Sequence"!() -> (then, List Objects selected:Array of Level Sequence Object References) {is=title K2Node_FunctionEntry}
   > Execute Python Script!(
    execute
    PythonScript="import unreal\r\nimport builtins\r\n\r\nbuiltins._progress = unreal.ScopedSlowTask(action_count, description_text)\r\nbuiltins._progress.__enter__()\r\nbuiltins._progress.make_dialog(False) ##no cancel button"
    action_count:int=Length(TargetArray:Array of Level Sequence Object References=@n1)
    description_text:String=Get ProgressBar_DescriptionText() -> (ProgressBar_DescriptionText:String)
   ) -> (then, ReturnValue) {inputs [action_count, description_text]}
   > n2 = For Each Loop(Array:Array of Level Sequence Object References=@n1) -> (Array Element:Level Sequence Object Reference)
   Execute Python Script!(
    execute=@n2.Completed
    PythonScript="import builtins\r\n\r\nif hasattr(builtins, '_progress'):\r\n    builtins._progress.__exit__(None, None, None)\r\n    del builtins._progress"
   ) -> (then, ReturnValue)
   > Return Node!(execute, WasItDone:bool=Get Was the procedure done() -> (Was the procedure done:bool))
   n3 = Set Current LS(execute=@n2.LoopBody, Current LS=@n2)
   > Execute Python Script!(
    execute
    PythonScript="import builtins\r\n\r\nif hasattr(builtins, '_progress'):\r\n    builtins._progress.enter_progress_frame(1, f\"Updating Level Sequence: {current_asset_name}\")"
    current_asset_name:String=GetDisplayName(Object=@n3)
   ) -> (then, ReturnValue) {inputs [current_asset_name]}
   > CheckoutAsset!(execute, AssetToCheckout:String=GetSoftObjectPath!(Object=@n3) -> (ReturnValue_PathString:String)) -> (then, ReturnValue:bool)
   > OpenLevelSequence!(execute, LevelSequence:Level Sequence Object Reference=Get Current LS()) -> (then, ReturnValue:bool)
   > SaveLoadedAsset(AssetToSave=Get Current LS(), bOnlyIfIsDirty=false)
   > n4 = GetBindings(Sequence=Get Current LS())
   > n5 = GetAnimSequenceLinkFromLevelSequence(InLevelSequence=Get Current LS())
   > n6 = For Each Loop(Array:Array of Movie Scene Binding Proxy Structures=@n4) -> (Array Element:Movie Scene Binding Proxy Structure)
   > Set Anim_Movie_Binding(Anim_Movie_Binding=@n6)
   > n7 = For Each Loop!(Exec, Array:Array of Level Sequence Anim Sequence Link Item Structures=Get AnimSequenceLinks(self=@n5)) -> (
    LoopBody
    Array Element_SkelTrackGuid:Guid Structure
    Array Element_PathToAnimSequence:Soft Object Path Structure
    Array Element_bAutoBake:bool=~
    Array Element_bExportTransforms:bool=~
    Array Element_bExportMorphTargets:bool=~
    Array Element_bExportAttributeCurves:bool=~
    Array Element_bExportMaterialCurves:bool=~
    Array Element_bSkipCurvesWithZeroValue:bool=~
    Array Element_bSetRetargetSourceAsset:bool=~
    Array Element_Interpolation:EAnimInterpolationType Enum
    Array Element_CurveInterpolation:ERichCurveInterpMode Enum
    Array Element_bRecordInWorldSpace:bool=~
    Array Element_bEvaluateAllSkeletalMeshComponents:bool=~
    Array Element_IncludeAnimationNames:Array of Strings
    Array Element_ExcludeAnimationNames:Array of Strings
    Array Element_bRemoveExcludedCurves:bool=~
    Array Element_WarmUpFrames:Frame Number Structure
    Array Element_DelayBeforeStart:Frame Number Structure
    Array Element_bUseCustomTimeRange:bool=~
    Array Element_CustomStartFrame:Frame Number Structure
    Array Element_CustomEndFrame:Frame Number Structure
    Array Element_CustomDisplayRate:Frame Rate Structure
    Array Element_bUseCustomFrameRate:bool=~
    Array Element_CustomFrameRate:Frame Rate Structure
    Array Index
    Completed
   )
   > n8 = FindBindingById(Sequence=Get Current LS(), BindingId=@n7)
   > Set CustomTimeFrame(CustomTimeFrame:bool=@n7.Array Element_bUseCustomTimeRange) -> (Output_Get:bool)
   n9 = Get Anim_Movie_Binding()
   > Branch(Condition="Equal (Guid)"(A=@n8, B=Break Movie Scene Binding Proxy(MovieSceneBindingProxy=@n9)) -> (ReturnValue=~))
   > Set StartFrame(StartFrame:Frame Number Structure=@n7.Array Element_CustomStartFrame) -> (Output_Get:Frame Number Structure)
   > Set End Frame(End Frame:Frame Number Structure=@n7.Array Element_CustomEndFrame) -> (Output_Get:Frame Number Structure)
   > n10 = LoadAsset(AssetPath=BreakSoftObjectPath(InSoftObjectPath=@n7.Array Element_PathToAnimSequence))
   > n11 = Cast To AnimSequence(Object=@n10)
   > Set Anim Sequence(Anim Sequence:Anim Sequence Object Reference=@n11) -> (Output_Get:Anim Sequence Object Reference)
   > Set CustomDisplayRate(CustomDisplayRate:Frame Rate Structure=@n7.Array Element_CustomDisplayRate) -> (Output_Get:Frame Rate Structure)
   > Execute Python Script!(
    execute
    PythonScript="import unreal\r\n\r\nasset_tools = unreal.AssetToolsHelpers.get_asset_tools()\r\neditor_subsystem = unreal.get_editor_subsystem(unreal.UnrealEditorSubsystem)\r\nworld = editor_subsystem.get_editor_world()\r\n\r\nanim_seq_export_options = unreal.AnimSeqExportOption()\r\nanim_seq_export_options.export_transforms = True\r\nanim_seq_export_options.export_morph_targets = True\r\nanim_seq_export_options.custom_display_rate = cdr\r\nanim_seq_export_options.use_custom_time_range = custom_time\r\nanim_seq_export_options.custom_start_frame = start_frame\r\nanim_seq_export_options.custom_end_frame = end_frame\r\n\r\n##FORT-1121516: re-apply export settings stored on the LS<->Anim link from the last\r\n##export -- most importantly excluded FACIAL curves. Without this the re-bake\r\n##re-introduces face curves deliberately stripped (CL 53535014).\r\nls_link = unreal.SequencerTools.get_anim_sequence_link_from_level_sequence(selected_ls)\r\nif ls_link:\r\n  for as_link in ls_link.anim_sequence_links:\r\n      if as_link.path_to_anim_sequence.to_tuple()[0] == anim.get_path_name():\r\n          anim_seq_export_options.skip_curves_with_zero_value = as_link.skip_curves_with_zero_value\r\n          anim_seq_export_options.set_retarget_source_asset = as_link.set_retarget_source_asset\r\n          excluded = as_link.exclude_animation_names\r\n          if excluded:\r\n              ue_array = unreal.Array(str)\r\n              for curve in excluded:\r\n                  ue_array.append(curve)\r\n              anim_seq_export_options.exclude_animation_names = ue_array\r\n              anim_seq_export_options.remove_excluded_curves = True\r\n          break\r\n\r\nunreal.LevelSequenceEditorBlueprintLibrary.force_update()\r\nunreal.ControlRigSequencerLibrary.export_anim_sequence_from_sequencer(anim, anim_seq_export_options, binding, False)"
    Selected_LS:Level Sequence Object Reference=Get Current LS()
    binding:Movie Scene Binding Proxy Structure=Get Anim_Movie_Binding()
    anim:Anim Sequence Object Reference=Get Anim Sequence() -> (Anim Sequence:Anim Sequence Object Reference)
    custom_time:bool=Get CustomTimeFrame() -> (CustomTimeFrame:bool)
    start_frame:Frame Number Structure=Get StartFrame() -> (StartFrame:Frame Number Structure)
    end_frame:Frame Number Structure=Get End Frame() -> (End Frame:Frame Number Structure)
    cdr:Frame Rate Structure=Get CustomDisplayRate() -> (CustomDisplayRate:Frame Rate Structure)
   ) -> (then, ReturnValue) {inputs [Selected_LS,binding,anim,custom_time,start_frame,end_frame,cdr]}
   > Apply All Animimation Modifiers!(execute, AnimSequence:Anim Sequence Object Reference=Get Anim Sequence()) -> (then)
   > SaveLoadedAssets(
    AssetsToSave=Make Array!("[0]":Object Reference=Get Anim Sequence(), "[1]":Object Reference=Get Current LS()) -> (Array:Array of Object References)
    bOnlyIfIsDirty=false
   )
   > CloseLevelSequence!(execute) -> (then)
   > n12 = Execute Python Script!(
    execute
    PythonScript="import unreal\r\n\r\nLS = selected_ls\r\n\r\nunreal.log(\"\\n------------------------------------------------------\")\r\nunreal.log(\"{} Anim Sequence has been updated\".format(LS))\r\n\r\nunreal.log(\"\\n-------------------------------------------------------\")"
    selected_ls:String=GetDisplayName(Object=Get Anim Sequence())
   ) -> (then, ReturnValue) {inputs [selected_ls]}
   > Set Was the procedure done(Was the procedure done:bool=@n12) -> (Output_Get:bool)
   > SaveLoadedAsset(AssetToSave=Get Current LS(), bOnlyIfIsDirty=true)
  ]}
 PopulatingSlot {nodes [
   PopulatingSlot() {is=title K2Node_FunctionEntry}
   > ClearChildren(self=Get LSB_Slot() -> (LSB_Slot:Named Slot Object Reference))
   > n1 = Create Widget!(
    execute:Exec
    WorldContextObject:Object Reference
    Class:User Widget Class Reference=/AnimInEdTools/LevelSequenceBrowser/EUW_LevelSequenceBrowser.EUW_LevelSequenceBrowser_C
    OwningPlayer:Player Controller Object Reference
   ) -> (then:Exec, ReturnValue:EUW Level Sequence Browser Object Reference)
   > Is Valid(InputObject=@n1)
   > Set LevelSeqBrowser(LevelSeqBrowser:EUW Level Sequence Browser Object Reference=@n1) -> (Output_Get:EUW Level Sequence Browser Object Reference)
   > AddChild(self=Get LSB_Slot(), Content=@n1)
  ]}
 Query Anim Sequence {nodes [
   n1 = Query Anim Sequence!() -> (then, InFile:String, Anim Sequence:Anim Sequence Object Reference) {is=title K2Node_FunctionEntry}
   > n2 = QueryFileState!(execute, InFile:String=@n1, bSilent:bool=false, bUseSourceControlStateCache:bool=false) -> (
    then
    ReturnValue_Filename:String
    ReturnValue_bIsValid:bool=~
    ReturnValue_bIsUnknown:bool=~
    ReturnValue_bCanCheckIn:bool=~
    ReturnValue_bCanCheckOut:bool=~
    ReturnValue_bIsCheckedOut:bool=~
    ReturnValue_bIsCurrent:bool=~
    ReturnValue_bIsSourceControlled:bool=~
    ReturnValue_bIsAdded:bool=~
    ReturnValue_bIsDeleted:bool=~
    ReturnValue_bIsIgnored:bool=~
    ReturnValue_bCanEdit:bool=~
    ReturnValue_bCanDelete:bool=~
    ReturnValue_bIsModified:bool=~
    ReturnValue_bCanAdd:bool=~
    ReturnValue_bIsConflicted:bool=~
    ReturnValue_bCanRevert:bool=~
    ReturnValue_bIsCheckedOutOther:bool=~
    ReturnValue_CheckedOutOther:String
    ReturnValue_bIsCheckedOutInOtherBranch:bool=~
    ReturnValue_bIsModifiedInOtherBranch:bool=~
    ReturnValue_PreviousUser:String
   )
   > n3 = Branch(Condition=@n2.ReturnValue_bIsCurrent)
   > n4 = Branch(Condition=@n2.ReturnValue_bIsCheckedOutOther)
   > n5 = Checked out by another user warning!(execute, self, Checked out by:String=@n2.ReturnValue_CheckedOutOther, Dipslay aset name:String=GetDisplayName(Object=Get Anim Sequence())) -> (then, Was it done:bool) {is=title K2Node_CallFunction}
   > Return Node!(execute, Was it done:bool=@n5)
   Return Node(execute=@n4.else, Was it done=true)
   n6 = Not Current Warning!(execute=@n3.else, self, Type of Asset:String=Animation Sequence, Display Asset Name:String=GetDisplayName(Object=Get Anim Sequence())) -> (then, Could it be done:bool) {is=title K2Node_CallFunction}
   > Return Node(Was it done=@n6)
  ]}
 QueryLS {nodes [
   n1 = QueryLS!() -> (then, InFile:String, CurrentLS:Level Sequence Object Reference) {is=title K2Node_FunctionEntry}
   > n2 = QueryFileState(InFile=@n1, bSilent=false, bUseSourceControlStateCache=false) -> (
    ReturnValue_bIsValid=~
    ReturnValue_bIsUnknown=~
    ReturnValue_bCanCheckIn=~
    ReturnValue_bCanCheckOut=~
    ReturnValue_bIsCheckedOut=~
    ReturnValue_bIsCurrent=~
    ReturnValue_bIsSourceControlled=~
    ReturnValue_bIsAdded=~
    ReturnValue_bIsDeleted=~
    ReturnValue_bIsIgnored=~
    ReturnValue_bCanEdit=~
    ReturnValue_bCanDelete=~
    ReturnValue_bIsModified=~
    ReturnValue_bCanAdd=~
    ReturnValue_bIsConflicted=~
    ReturnValue_bCanRevert=~
    ReturnValue_bIsCheckedOutOther=~
    ReturnValue_bIsCheckedOutInOtherBranch=~
    ReturnValue_bIsModifiedInOtherBranch=~
   )
   > n3 = Branch(Condition=@n2.ReturnValue_bIsCurrent)
   > n4 = Branch(Condition=@n2.ReturnValue_bIsCheckedOutOther)
   > n5 = Checked out by another user warning(
    Checked out by=@n2.ReturnValue_CheckedOutOther
    Dipslay aset name=GetDisplayName(Object=Get CurrentLS() -> (CurrentLS:Level Sequence Object Reference))
   )
   > Return Node(Was it done=@n5)
   Return Node(execute=@n4.else, Was it done=true)
   n6 = Not Current Warning(execute=@n3.else, Type of Asset=Level Sequence, Display Asset Name=GetDisplayName(Object=Get CurrentLS()))
   > Return Node(Was it done=@n6)
  ]}
 Success Menu {nodes [
   Success Menu() {is=title K2Node_FunctionEntry}
   > n1 = For Each Loop!(Exec, Array:Array of Level Sequence Object References=Get Selected LS Missing Modifier array()) -> (LoopBody, Array Element:Level Sequence Object Reference, Array Index, Completed)
   Set NewLocalVar(
    execute=@n1.Completed
    NewLocalVar:String=JoinStringArray(SourceArray=Get ListOfMissingMods() -> (ListOfMissingMods:Array of Strings), Separator=" , ")
   ) -> (Output_Get:String)
   > n2 = Branch(Condition=Get notEmpty() -> (notEmpty:bool))
   n3 = Get NewLocalVar() -> (NewLocalVar:String)
   > Show Message Dialog(
    Title=Get Popup_Title()
    Message="To Text (String)"(InString=Append(A="The selected Level Sequences and their Animations have been updated\r\n\r\n", B=@n3, C=" are missing Animation Data Modifiers!"))
    MessageType=Ok
    DefaultValue=No
    MessageCategory=Warning
   )
   Show Message Dialog(
    execute=@n2.else
    Title=Get Popup_Title()
    Message="To Text (String)"(InString=The selected Level Sequences and their Animations have been updated)
    MessageType=Ok
    DefaultValue=No
    MessageCategory=Success
   )
   n4 = GetDisplayName(Object=@n1)
   n5 = Branch(execute=@n1.LoopBody, Condition="Equal Exactly (String)"(A=@n4))
   Set notEmpty(execute=@n5.else, notEmpty:bool=true) -> (Output_Get:bool)
   > Add Unique(TargetArray:Array of Strings=Get ListOfMissingMods(), NewItem:"String (by ref)"=@n4)
  ]}
]
variables [
 LevelSeqBrowser {type=EUW_LevelSequenceBrowser_C, default=null}
 LSArray {type="array<LevelSequence>", default []}
 Popup_Title {type=string, default=Level Sequence Batch Update}
 Save Game {type=SG_LevelSequenceBrowser_C, default=null}
 Selected LS Missing Modifier array {type="array<LevelSequence>", default []}
 Selected_Asset {type=string, default=""}
 Selected_LS {type=LevelSequence, default=null}
]
widgets {
 info {rootWidgetClass=&CanvasPanel, widgetCount=14}
 widgets [
  CanvasPanel_21 {widgetClassPath=&CanvasPanel, widget=CanvasPanel_21}
  VerticalBox_29 {
   widgetClassPath=&VerticalBox
   parent=CanvasPanel_21
   widget=VerticalBox_29
   slot {class=CanvasPanelSlot, properties {layoutData {alignment=<0 0>, anchors {maximum=<1 1>, minimum=<0 0>}, offsets=<ltrb 0 0 0 0>}}}
  }
  HorizontalBox_0 {
   widgetClassPath=&HorizontalBox
   parent=VerticalBox_29
   widget=HorizontalBox_0
   slot {class=VerticalBoxSlot, properties {size {sizeRule=Fill, value=1}}}
  }
  SizeBox_0 {
   widgetClassPath=&SizeBox
   parent=HorizontalBox_0
   widget=SizeBox_0
   properties {bOverride_MinDesiredHeight=true, minDesiredHeight=500, minDesiredWidth=50}
   slot {class=HorizontalBoxSlot, properties {size {sizeRule=Fill, value=1}}}
  }
  LSB_Slot {widgetClassPath=&NamedSlot, bIsVariable=true, parent=SizeBox_0, widget=LSB_Slot, slot {class=SizeBoxSlot}}
  SizeBox {
   widgetClassPath=&SizeBox
   parent=VerticalBox_29
   widget=SizeBox
   properties {bOverride_MinDesiredWidth=true, maxDesiredHeight=400, minDesiredWidth=100}
   slot {class=VerticalBoxSlot, properties {horizontalAlignment=HAlign_Center, padding=<ltrb 0 0 0 10>}}
  }
  HorizontalBox_235 {
   widgetClassPath=&HorizontalBox
   parent=SizeBox
   widget=HorizontalBox_235
   slot {class=SizeBoxSlot, properties {horizontalAlignment=HAlign_Left, verticalAlignment=VAlign_Center}}
  }
  Update_LS_Btn {
   widgetClassPath=&EditorUtilityButton
   parent=HorizontalBox_235
   widget=Update_LS_Btn
   slot {class=HorizontalBoxSlot, properties {horizontalAlignment=HAlign_Left, padding=<ltrb 10 0 0 0>}}
  }
  TextBlock_58 {
   widgetClassPath=&TextBlock
   parent=Update_LS_Btn
   widget=TextBlock_58
   properties {
    text=Update Selected Level Sequences
    font {
     bForceMonospaced=false
     bMaterialIsStencil=false
     fontMaterial=null
     fontObject=&Roboto
     letterSpacing=0
     monospacedWidth=1
     size=12
     skewAmount=0
     typefaceFontName=Regular
     outlineSettings {
      bApplyOutlineToDropShadows=false
      bMiteredCorners=false
      bSeparateFillAlpha=false
      outlineMaterial=null
      outlineSize=0
      outlineColor=<rgba 0 0 0 1>
     }
    }
   }
   slot {class=ButtonSlot}
  }
  Update_All_LS_Btn {
   widgetClassPath=&EditorUtilityButton
   parent=HorizontalBox_235
   widget=Update_All_LS_Btn
   slot {class=HorizontalBoxSlot, properties {horizontalAlignment=HAlign_Left, padding=<ltrb 10 0 0 0>}}
  }
  TextBlock {
   widgetClassPath=&TextBlock
   parent=Update_All_LS_Btn
   widget=TextBlock
   properties {
    text=Update All Level Sequences
    font {
     bForceMonospaced=false
     bMaterialIsStencil=false
     fontMaterial=null
     fontObject=&Roboto
     letterSpacing=0
     monospacedWidth=1
     size=12
     skewAmount=0
     typefaceFontName=Regular
     outlineSettings {
      bApplyOutlineToDropShadows=false
      bMiteredCorners=false
      bSeparateFillAlpha=false
      outlineMaterial=null
      outlineSize=0
      outlineColor=<rgba 0 0 0 1>
     }
    }
   }
   slot {class=ButtonSlot}
  }
  Overlay_173 {
   widgetClassPath=&Overlay
   parent=CanvasPanel_21
   widget=Overlay_173
   slot {class=CanvasPanelSlot, properties {zOrder=1, layoutData {alignment=<0 0>, anchors {maximum=<1 1>, minimum=<0 0>}, offsets=<ltrb 0 0 0 0>}}}
  }
  EUW_ContextMenu {
   widgetClassPath=/AnimInEdTools/ContextMenu/EUW_ContextMenu
   bIsVariable=true
   parent=Overlay_173
   widget=EUW_ContextMenu
   properties {visibility=Collapsed}
   slot {class=OverlaySlot}
  }
  ModalLayer {
   widgetClassPath=&NamedSlot
   bIsVariable=true
   parent=Overlay_173
   widget=ModalLayer
   properties {bExposeOnInstanceOnly=true}
   slot {class=OverlaySlot, properties {horizontalAlignment=HAlign_Fill, verticalAlignment=VAlign_Fill}}
  }
 ]
}
